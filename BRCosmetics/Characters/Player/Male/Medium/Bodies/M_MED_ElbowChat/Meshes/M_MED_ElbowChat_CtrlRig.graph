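# Control Rig for the M_MED_ElbowChat body cosmetic, correcting collar, shoulder, front and back cloth, and foot bones to follow the character's pose.
# A DeformRig function runs first, then spherical pose readers, bone deltas and pelvis height produce pose weights, some smoothed with Kalman filters.
# Those weights drive ModifyTransforms offsets for arm, head, leg and sliding poses, with variable-driven branches or weights for skydive, crouch, DBNO, moving and targeting.
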ControlRigBlueprint /BRCosmetics/Characters/Player/Male/Medium/Bodies/M_MED_ElbowChat/Meshes/M_MED_ElbowChat_CtrlRig
graphs [
 RigVMFunctionLibrary {links [], nodes []}
 RigVMFunction_Sequence_1_ContainedGraph {nodes [
   n1 = RigVMFunctionEntryNode()
   RigVMFunction_Sequence = &RigVMFunction_Sequence(ExecuteContext=@n1.ExecuteContext)
   n2 = &RigVMFunction_Sequence(ExecuteContext=@RigVMFunction_Sequence.B)
   RigVMFunctionReturnNode(A=@RigVMFunction_Sequence.A, B=@n2.A, C=@n2.B)
  ]}
 RigVMModel {nodes [
   RigUnit_BeginExecution = &RigUnit_BeginExecution()
   FN_DeformRig_FNC_1 = RigVMFunctionReferenceNode(
    DeformRig_ Curve="(deformRig_l_upperarm_up_alpha_72_#1=False,deformRig_r_upperarm_up_alpha_73_#2=False,deformRig_l_upperarm_fr_alpha_74_#3=False,deformRig_r_upperarm_fr_alpha_75_CB51B8 ...(truncated)"
    DeformRig_Bool="(upperarm_1_#4=True,elbow_3_#5=True,wrist_7_#6=True,groin_8_#7=True,glute_17_#8=True,knee_9_40C408324CA261 ...(truncated)"
    DeformRig_Debug=(upperarm_debug_18_#4=False,wrist_debug_20_#6=False,lat_debug_26_#9=False,groinGlute_debug_27_#7=False)
    DeformRig_Value="(upperarm_up_18_#10=(Rotation=(X=0,Y=0,Z=0,W=1),Translation=(X=0,Y=0,Z=8),Scale3D=(X=0.7,Y=1,Z=1)),upperarm_fwd_fr_19_#11=(Rotation=(X=0,Y=0,Z=0,W=1),Translation=(X=0,Y=10,Z=0),Scale3D=(X=0.5,Y=1,Z= ...(truncated)"
    ExecuteContext=@RigUnit_BeginExecution.ExecutePin
   )
   > n1 = &RigUnit_SphericalPoseReader(
    ActiveRegionScaleFactors=(PositiveWidth=1,NegativeWidth=0.1,PositiveHeight=1,NegativeHeight=1)
    ActiveRegionSize=0.3
    Debug=(bDrawDebug=true,bDraw2D=false,bDrawLocalAxes=false,DebugScale=25,DebugSegments=20,DebugThickness=0.25)
    DriverAxis=(X=-1,Y=0,Z=0)
    DriverItem=(Type=Bone,Name="thigh_l")
    FalloffRegionScaleFactors=(PositiveWidth=1,NegativeWidth=0.7,PositiveHeight=1,NegativeHeight=1)
    FalloffSize=0.5
    FlipHeightScaling=False
    FlipWidthScaling=False
    OptionalParentItem=(Type=Bone,Name="pelvis")
    RotationOffset=(X=90,Y=0,Z=0)
   )
   > RigVMVariableNode(Value=@n1.OutputParam) {variable=Left_Leg_SPR}
   > n2 = &RigUnit_SphericalPoseReader(
    ActiveRegionScaleFactors=(PositiveWidth=0.1,NegativeWidth=1,PositiveHeight=1,NegativeHeight=1)
    ActiveRegionSize=0.3
    Debug=(bDrawDebug=true,bDraw2D=false,bDrawLocalAxes=false,DebugScale=25,DebugSegments=20,DebugThickness=0.25)
    DriverAxis=(X=1,Y=0,Z=0)
    DriverItem=(Type=Bone,Name="thigh_r")
    FalloffRegionScaleFactors=(PositiveWidth=0.7,NegativeWidth=1,PositiveHeight=1,NegativeHeight=1)
    FalloffSize=0.5
    FlipHeightScaling=False
    FlipWidthScaling=False
    OptionalParentItem=(Type=Bone,Name="pelvis")
    RotationOffset=(X=-90,Y=0,Z=0)
   )
   > RigVMVariableNode(Value=@n2.OutputParam) {variable=Right_Leg_SPR}
   n3 = &RigUnit_GetRelativeTransformForItem(Child=(Type=Bone,Name="pelvis"), Parent=(Type=Bone,Name="root"), bChildInitial=False, bParentInitial=False)
   KalmanFilter = &RigVMFunction_KalmanFloat(
    BufferSize=17
    Value=&RigVMFunction_MathDoubleRemap(SourceMaximum=75, SourceMinimum=80, TargetMaximum=1, TargetMinimum=0, Value=@n3.RelativeTransform.Translation.Z).Result
   ) {template="KalmanFilter::Execute(in Value,in BufferSize,out Result)"}
   > RigVMVariableNode(Value=@KalmanFilter.Result) {variable=Plevis Relative Root}
   > RigVMVariableNode(Value=&RigVMFunction_MathDoubleRemap(SourceMaximum=22, SourceMinimum=40, TargetMaximum=1, TargetMinimum=0, Value=@n3.RelativeTransform.Translation.Z).Result) {variable=Is sliding}
   > Get Delta Ref_2_2 = RigVMFunctionReferenceNode(Item=(Type=Bone,Name="calf_l"))
   > Get Delta Ref_2_1_1 = RigVMFunctionReferenceNode(Item=(Type=Bone,Name="calf_r"))
   > RigVMVariableNode(Value=&RigVMFunction_MathDoubleRemap(
     SourceMaximum=140
     SourceMinimum=130
     TargetMaximum=1
     TargetMinimum=0
     Value=&RigVMFunction_MathDoubleMul(A=@"Get Delta Ref_2_1_1".Result.Rotation.Yaw, B=@"Get Delta Ref_2_2".Result.Rotation.Yaw).Result
    ).Result) {variable=Calf knee sliding or dbno}
   > RigVMFunction_Sequence = RigVMAggregateNode()
   n4 = &RigUnit_SphericalPoseReader(
    ActiveRegionScaleFactors=(PositiveWidth=1,NegativeWidth=1.2,PositiveHeight=0.9,NegativeHeight=1)
    ActiveRegionSize=0.2
    Debug=(bDrawDebug=false,bDraw2D=False,bDrawLocalAxes=False,DebugScale=25,DebugSegments=20,DebugThickness=0.25)
    DriverAxis=(X=1,Y=0,Z=0)
    DriverItem=(Type=Bone,Name="upperarm_l")
    FalloffRegionScaleFactors=(PositiveWidth=1,NegativeWidth=1,PositiveHeight=4,NegativeHeight=1)
    FalloffSize=0.4
    FlipHeightScaling=False
    FlipWidthScaling=False
    OptionalParentItem=(Type=Bone,Name="spine_05")
    RotationOffset=(X=20,Y=0,Z=0)
    ExecutePin=@RigVMFunction_Sequence.A
   )
   > RigVMVariableNode(Value=@n4.OutputParam) {variable=left Hand up}
   > &RigUnit_ModifyTransforms(
    ItemToModify="((Item=(Type=Bone,Name=\"dyn_collar_l_master\"),Transform=(Rotation=(X=-0.1669,Y=0.2715,Z=0.0479,W=0.9467),Translation=(X=0,Y=0,Z=3),Scale3D=(X=1,Y=1,Z=1))),(Item=(Type=Bone,Name=\"dyn_shoulder_l_master\"),Transform=(Rotation=(X=-0.0435,Y=-0.0871,Z=-0.0038,W=0 ...(truncated)"
    WeightMinimum=0
    Weight=@n4.OutputParam
   )
   > n5 = &RigUnit_SphericalPoseReader(
    ActiveRegionScaleFactors=(PositiveWidth=1,NegativeWidth=1.7,PositiveHeight=1.7,NegativeHeight=1.5)
    ActiveRegionSize=0.2
    Debug=(bDrawDebug=false,bDraw2D=False,bDrawLocalAxes=False,DebugScale=25,DebugSegments=20,DebugThickness=0.25)
    DriverAxis=(X=-1,Y=0,Z=0)
    DriverItem=(Type=Bone,Name="upperarm_r")
    FalloffRegionScaleFactors=(PositiveWidth=1,NegativeWidth=5,PositiveHeight=3,NegativeHeight=2)
    FalloffSize=0.4
    FlipHeightScaling=False
    FlipWidthScaling=False
    OptionalParentItem=(Type=Bone,Name="spine_05")
    RotationOffset=(X=-160,Y=30,Z=20)
   )
   > &RigUnit_ModifyTransforms(
    ItemToModify="((Item=(Type=Bone,Name=\"dyn_collar_r_master\"),Transform=(Rotation=(X=0,Y=0.061,Z=0,W=0.9981),Translation=(X=0,Y=0,Z=-3),Scale3D=(X=1,Y=1,Z=1))),(Item=(Type=Bone,Name=\"dyn_shoulder_r_master\"),Transform=(Rotation=(X=-0.0872,Y=0,Z=0,W=0.9962),Translation=(X=0 ...(truncated)"
    WeightMinimum=0
    Weight=@n5.OutputParam
   )
   n6 = &RigVMFunction_MathDoubleMul(B=@n5.OutputParam, A=@n4.OutputParam)
   > n7 = RigVMVariableNode(Value=@n6.Result) {variable=Is hands up}
   n8 = &RigVMFunction_Sequence(ExecuteContext=@n7.ExecuteContext)
   n9 = &RigVMFunction_MathDoubleMul(A=RigVMVariableNode() {variable=Plevis Relative Root}.Value, B=@n6.Result)
   &RigUnit_ModifyTransforms(
    ItemToModify=((Item=(Type=Bone,Name="dyn_collar_l_master"),Transform=(Rotation=(X=-0.2586,Y=0.0421,Z=0.0113,W=0.965),Translation=(X=0,Y=0,Z=0),Scale3D=(X=1,Y=1,Z=1))))
    WeightMinimum=0
    Weight=@n9.Result
    ExecutePin=@n8.A
   )
   RigVMFunction_MathQuaternionToEuler = &RigVMFunction_MathQuaternionToEuler(
    RotationOrder=ZYX
    Value=&RigUnit_GetRelativeTransformForItem(Child=(Type=Bone,Name="head"), Parent=(Type=Bone,Name="spine_05"), bChildInitial=False, bParentInitial=False).RelativeTransform.Rotation
   )
   n10 = &RigVMFunction_MathDoubleMul(
    B=&RigVMFunction_MathDoubleRemap(
     SourceMaximum=1
     SourceMinimum=0
     TargetMaximum=0
     TargetMinimum=1
     Value=&RigVMFunction_MathDoubleRemap(SourceMaximum=-13, SourceMinimum=-5, TargetMaximum=1, TargetMinimum=0, Value=@RigVMFunction_MathQuaternionToEuler.Result.Y).Result
    ).Result
    A=@n6.Result
   )
   > &RigUnit_ModifyTransforms(
    ItemToModify="((Item=(Type=Bone,Name=\"dyn_collar_l_master\"),Transform=(Rotation=(X=0,Y=-0.1305,Z=0,W=0.9914),Translation=(X=0,Y=0,Z=0),Scale3D=(X=1,Y=1,Z=1))),(Item=(Type=Bone,Name=\"dyn_shoulder_l_02\"),Transform=(Rotation=(X=0,Y=0,Z=0,W=1),Translation=(X=-3,Y=0,Z=0),Sca ...(truncated)"
    WeightMinimum=0
    Weight=&RigVMFunction_MathDoubleMul(
     B=@n10.Result
     A=&RigVMFunction_MathDoubleRemap(SourceMaximum=1, SourceMinimum=0, TargetMaximum=0, TargetMinimum=1, Value=@n9.Result).Result
    ).Result
   )
   &RigUnit_ModifyTransforms(
    ItemToModify="((Item=(Type=Bone,Name=\"dyn_collar_l_master\"),Transform=(Rotation=(X=0.2588,Y=0,Z=0,W=0.9659),Translation=(X=0,Y=0,Z=0),Scale3D=(X=1,Y=1,Z=1))),(Item=(Type=Bone,Name=\"dyn_shoulder_l_02\"),Transform=(Rotation=(X=0,Y=0.1736,Z=0,W=0.9848),Translation=(X=-3,Y=0 ...(truncated)"
    WeightMinimum=0
    Weight=&RigVMFunction_MathDoubleMul(B=@n10.Result, A=@n9.Result).Result
    ExecutePin=@n8.B
   )
   > n11 = &RigUnit_SphericalPoseReader(
    ActiveRegionScaleFactors=(PositiveWidth=1,NegativeWidth=1,PositiveHeight=1,NegativeHeight=1)
    ActiveRegionSize=0.07
    Debug=(bDrawDebug=false,bDraw2D=False,bDrawLocalAxes=False,DebugScale=25,DebugSegments=20,DebugThickness=0.25)
    DriverAxis=(X=1,Y=0,Z=0)
    DriverItem=(Type=Bone,Name="clavicle_l")
    FalloffRegionScaleFactors=(PositiveWidth=1,NegativeWidth=1,PositiveHeight=1,NegativeHeight=1)
    FalloffSize=0.2
    FlipHeightScaling=False
    FlipWidthScaling=False
    OptionalParentItem=(Type=Bone,Name="None")
    RotationOffset=(X=0,Y=-50,Z=-10)
   )
   > n12 = &RigUnit_SphericalPoseReader(
    ActiveRegionScaleFactors=(PositiveWidth=1,NegativeWidth=1,PositiveHeight=1,NegativeHeight=1)
    ActiveRegionSize=0.07
    Debug=(bDrawDebug=false,bDraw2D=False,bDrawLocalAxes=False,DebugScale=25,DebugSegments=20,DebugThickness=0.25)
    DriverAxis=(X=-1,Y=0,Z=0)
    DriverItem=(Type=Bone,Name="clavicle_r")
    FalloffRegionScaleFactors=(PositiveWidth=1,NegativeWidth=1,PositiveHeight=1,NegativeHeight=1)
    FalloffSize=0.2
    FlipHeightScaling=False
    FlipWidthScaling=False
    OptionalParentItem=(Type=Bone,Name="None")
    RotationOffset=(X=0,Y=132,Z=-10)
   )
   n13 = &RigVMFunction_ControlFlowBranch(ExecuteContext=@n12.ExecutePin, Condition=RigVMVariableNode() {variable=Is skydive}.Value)
   n14 = &RigVMFunction_MathDoubleMul(
    B=&RigVMFunction_MathDoubleRemap(SourceMaximum=1, SourceMinimum=0, TargetMaximum=0, TargetMinimum=1, Value=RigVMVariableNode() {variable=Is hands up}.Value).Result
    A=@n12.OutputParam
   )
   n15 = &RigVMFunction_MathDoubleRemap(SourceMaximum=1, SourceMinimum=0, TargetMaximum=0, TargetMinimum=1, Value=RigVMVariableNode() {variable=Plevis Relative Root}.Value)
   &RigUnit_ModifyTransforms(
    ItemToModify=((Item=(Type=Bone,Name="dyn_collar_r_master"),Transform=(Rotation=(X=-0.25,Y=0.25,Z=0.067,W=0.933),Translation=(X=0,Y=0,Z=0),Scale3D=(X=1,Y=1,Z=1))))
    WeightMinimum=0
    Weight=&RigVMFunction_KalmanFloat(BufferSize=6, Value=&RigVMFunction_MathDoubleMul(A=@n14.Result, B=@n15.Result).Result) {template="KalmanFilter::Execute(in Value,in BufferSize,out Result)"}.Result
    ExecutePin=@n13.Completed
   )
   > &RigUnit_ModifyTransforms(
    ItemToModify=((Item=(Type=Bone,Name="dyn_shoulder_r_master"),Transform=(Rotation=(X=0,Y=0.0436,Z=0,W=0.999),Translation=(X=0,Y=0,Z=0),Scale3D=(X=1,Y=1,Z=1))))
    WeightMinimum=0
    Weight=@n14.Result
   )
   n16 = &RigVMFunction_MathDoubleMul(B=@n12.OutputParam, A=@n11.OutputParam)
   > &RigUnit_ModifyTransforms(
    ItemToModify="((Item=(Type=Bone,Name=\"Collar_corrective_ads\"),Transform=(Rotation=(X=-0.2527,Y=-0.056,Z=0.2091,W=0.943),Translation=(X=1,Y=0,Z=0),Scale3D=(X=1,Y=1,Z=1))),(Item=(Type=Bone,Name=\"dyn_collar_r_master\"),Transform=(Rotation=(X=0,Y=0.0872,Z=0,W=0.9962),Transla ...(truncated)"
    WeightMinimum=0
    Weight=@n16.Result
   )
   > n17 = &RigUnit_SphericalPoseReader(
    ActiveRegionScaleFactors=(PositiveWidth=1,NegativeWidth=1,PositiveHeight=1,NegativeHeight=1)
    ActiveRegionSize=0.1
    Debug=(bDrawDebug=false,bDraw2D=False,bDrawLocalAxes=False,DebugScale=25,DebugSegments=20,DebugThickness=0.25)
    DriverAxis=(X=-1,Y=0,Z=0)
    DriverItem=(Type=Bone,Name="upperarm_r")
    FalloffRegionScaleFactors=(PositiveWidth=1,NegativeWidth=1.1,PositiveHeight=1,NegativeHeight=1)
    FalloffSize=0.3
    FlipHeightScaling=False
    FlipWidthScaling=False
    OptionalParentItem=(Type=Bone,Name="spine_05")
    RotationOffset=(X=-90,Y=0,Z=20)
   )
   > &RigUnit_ModifyTransforms(
    ItemToModify=((Item=(Type=Bone,Name="Collar_corrective_ads"),Transform=(Rotation=(X=0,Y=0,Z=0,W=1),Translation=(X=7,Y=0,Z=0),Scale3D=(X=1,Y=1,Z=1))))
    WeightMinimum=0
    Weight=&RigVMFunction_MathDoubleMul(
     A=&RigVMFunction_MathDoubleRemap(SourceMaximum=1, SourceMinimum=0, TargetMaximum=0, TargetMinimum=1, Value=@n16.Result).Result
     B=@n17.OutputParam
    ).Result
   )
   > n18 = &RigUnit_SphericalPoseReader(
    ActiveRegionScaleFactors=(PositiveWidth=0.4,NegativeWidth=0.2,PositiveHeight=1,NegativeHeight=1)
    ActiveRegionSize=0.1
    Debug=(bDrawDebug=true,bDraw2D=False,bDrawLocalAxes=False,DebugScale=25,DebugSegments=20,DebugThickness=0.25)
    DriverAxis=(X=1,Y=0,Z=0)
    DriverItem=(Type=Bone,Name="head")
    FalloffRegionScaleFactors=(PositiveWidth=1,NegativeWidth=0.4,PositiveHeight=1,NegativeHeight=3)
    FalloffSize=0.2
    FlipHeightScaling=False
    FlipWidthScaling=False
    OptionalParentItem=(Type=Bone,Name="spine_05")
    RotationOffset=(X=-60,Y=0,Z=25)
   )
   > &RigUnit_ModifyTransforms(
    ItemToModify="((Item=(Type=Bone,Name=\"dyn_collar_m_master\"),Transform=(Rotation=(X=0.25,Y=0.25,Z=-0.067,W=0.933),Translation=(X=0,Y=3,Z=-4),Scale3D=(X=1,Y=1,Z=1))),(Item=(Type=Bone,Name=\"dyn_collar_r_master\"),Transform=(Rotation=(X=-0.258,Y=-0.0758,Z=-0.0203,W=0.9629),T ...(truncated)"
    WeightMinimum=0
    Weight=@n18.OutputParam
   )
   > n19 = &RigUnit_SphericalPoseReader(
    ActiveRegionScaleFactors=(PositiveWidth=0.7,NegativeWidth=1,PositiveHeight=1,NegativeHeight=1)
    ActiveRegionSize=0.1
    Debug=(bDrawDebug=true,bDraw2D=False,bDrawLocalAxes=False,DebugScale=25,DebugSegments=20,DebugThickness=0.25)
    DriverAxis=(X=1,Y=-1,Z=0)
    DriverItem=(Type=Bone,Name="head")
    FalloffRegionScaleFactors=(PositiveWidth=0.7,NegativeWidth=1,PositiveHeight=1,NegativeHeight=1)
    FalloffSize=0.2
    FlipHeightScaling=False
    FlipWidthScaling=False
    OptionalParentItem=(Type=Bone,Name="spine_05")
    RotationOffset=(X=-80,Y=0,Z=-50)
   )
   > &RigUnit_ModifyTransforms(
    ItemToModify="((Item=(Type=Bone,Name=\"dyn_collar_m_master\"),Transform=(Rotation=(X=0.2241,Y=0.483,Z=-0.1294,W=0.8365),Translation=(X=-1,Y=0,Z=-3),Scale3D=(X=1,Y=1,Z=1))),(Item=(Type=Bone,Name=\"dyn_collar_r_master\"),Transform=(Rotation=(X=-0.5028,Y=-0.0318,Z=-0.1186,W=0. ...(truncated)"
    WeightMinimum=0
    Weight=@n19.OutputParam
   )
   &RigUnit_ModifyTransforms(
    ItemToModify=((Item=(Type=Bone,Name="dyn_collar_l_master"),Transform=(Rotation=(X=-0.2586,Y=-0.0421,Z=-0.0113,W=0.965),Translation=(X=0,Y=0,Z=0),Scale3D=(X=1,Y=1,Z=1))))
    WeightMinimum=0
    Weight=&RigVMFunction_MathDoubleMul(A=@n11.OutputParam, B=RigVMVariableNode() {variable=left Hand up}.Value).Result
    ExecutePin=@n13.True
   )
   > &RigUnit_ModifyTransforms(
    ItemToModify=((Item=(Type=Bone,Name="dyn_collar_r_master"),Transform=(Rotation=(X=0,Y=0.1736,Z=0,W=0.9848),Translation=(X=0,Y=0,Z=0),Scale3D=(X=1,Y=1,Z=1))))
    WeightMinimum=0
    Weight=@n12.OutputParam
   )
   n20 = &RigVMFunction_ControlFlowBranch(ExecuteContext=@RigVMFunction_Sequence.B, Condition=RigVMVariableNode() {variable=is dbno}.Value)
   RigVMFunction_ControlFlowBranch = &RigVMFunction_ControlFlowBranch(ExecuteContext=@n20.False, Condition=RigVMVariableNode() {variable=Is crouch}.Value)
   n21 = &RigVMFunction_MathDoubleRemap(SourceMaximum=1, SourceMinimum=0, TargetMaximum=0, TargetMinimum=1, Value=RigVMVariableNode() {variable=Is targeting}.Value)
   n22 = RigVMVariableNode() {variable=Is Moving}
   &RigUnit_ModifyTransforms(
    ItemToModify=((Item=(Type=Bone,Name="dyn_cloth_fr_master"),Transform=(Rotation=(X=0,Y=0,Z=-0.6428,W=0.766),Translation=(X=-4,Y=0,Z=0),Scale3D=(X=1,Y=1,Z=1))))
    WeightMinimum=0
    Weight=&RigVMFunction_KalmanFloat(BufferSize=16, Value=&RigVMFunction_MathDoubleMul(B=@n21.Result, A=@n22.Value).Result) {template="KalmanFilter::Execute(in Value,in BufferSize,out Result)"}.Result
    ExecutePin=@RigVMFunction_ControlFlowBranch.False
   )
   n23 = RigVMVariableNode() {variable=Left_Leg_SPR}
   > &RigUnit_ModifyTransforms(
    ItemToModify=((Item=(Type=Bone,Name="dyn_cloth_fr_master"),Transform=(Rotation=(X=0.2588,Y=0,Z=0,W=0.9659),Translation=(X=2,Y=5,Z=3),Scale3D=(X=1,Y=1,Z=1))))
    WeightMinimum=0
    Weight=@n23.Value
   )
   > Get Delta Ref_2_3_1 = RigVMFunctionReferenceNode(Item=(Type=Bone,Name="thigh_l"))
   > &RigUnit_ModifyTransforms(
    ItemToModify=((Item=(Type=Bone,Name="dyn_cloth_fr_master"),Transform=(Rotation=(X=0,Y=0,Z=-0.2588,W=0.9659),Translation=(X=5,Y=0,Z=3),Scale3D=(X=1,Y=1,Z=1))))
    WeightMinimum=0
    Weight=&RigVMFunction_MathDoubleMul(
     A=&RigVMFunction_MathDoubleRemap(SourceMaximum=110, SourceMinimum=90, TargetMaximum=1, TargetMinimum=0, Value=@"Get Delta Ref_2_3_1".Result.Rotation.Yaw).Result
     B=@n23.Value
    ).Result
   )
   n24 = RigVMVariableNode() {variable=Right_Leg_SPR}
   > &RigUnit_ModifyTransforms(
    ItemToModify=((Item=(Type=Bone,Name="dyn_cloth_fr_master"),Transform=(Rotation=(X=-0.2588,Y=0,Z=0,W=0.9659),Translation=(X=2,Y=5,Z=-3),Scale3D=(X=1,Y=1,Z=1))))
    WeightMinimum=0
    Weight=@n24.Value
   )
   > Get Delta Ref_2_3 = RigVMFunctionReferenceNode(Item=(Type=Bone,Name="thigh_r"))
   n25 = &RigVMFunction_MathDoubleRemap(SourceMaximum=-110, SourceMinimum=-90, TargetMaximum=1, TargetMinimum=0, Value=@"Get Delta Ref_2_3".Result.Rotation.Yaw)
   n26 = RigVMVariableNode() {variable=Is Moving}
   > &RigUnit_ModifyTransforms(
    ItemToModify=((Item=(Type=Bone,Name="dyn_cloth_fr_master"),Transform=(Rotation=(X=-0.1927,Y=0.0122,Z=-0.5036,W=0.8421),Translation=(X=4,Y=-5,Z=-8),Scale3D=(X=1,Y=1,Z=1))))
    WeightMinimum=0
    Weight=&RigVMFunction_MathDoubleMul(
     A=&RigVMFunction_MathDoubleMul(A=@n25.Result, B=@n24.Value).Result
     B=&RigVMFunction_MathDoubleRemap(SourceMaximum=1, SourceMinimum=0, TargetMaximum=0, TargetMinimum=1, Value=@n26.Value).Result
    ).Result
   )
   n27 = RigVMVariableNode() {variable=Left_Leg_SPR}
   Multiply = &RigVMFunction_MathDoubleMul(
    B=&RigVMFunction_MathDoubleClamp(Value=&RigVMFunction_MathDoubleMul(B=2.5, A=@n27.Value).Result).Result
    A=RigVMVariableNode() {variable=Right_Leg_SPR}.Value
   )
   > &RigUnit_ModifyTransforms(
    ItemToModify=((Item=(Type=Bone,Name="dyn_cloth_fr_master"),Transform=(Rotation=(X=0.2549,Y=0.0449,Z=0.1677,W=0.9513),Translation=(X=-5,Y=-5,Z=2),Scale3D=(X=1,Y=1,Z=1))))
    WeightMinimum=0
    Weight=@Multiply.Result
   )
   > &RigUnit_ModifyTransforms(
    ItemToModify=((Item=(Type=Bone,Name="pelvis"),Transform=(Rotation=(X=0,Y=0,Z=0,W=1),Translation=(X=0,Y=0,Z=0),Scale3D=(X=1,Y=1,Z=1))))
    Weight=1
    WeightMinimum=0
   )
   > Get Delta Ref_3 = RigVMFunctionReferenceNode(Item=(Type=Bone,Name="pelvis"))
   > &RigUnit_ModifyTransforms(
    ItemToModify=((Item=(Type=Bone,Name="dyn_cloth_fr_master"),Transform=(Rotation=(X=0,Y=0,Z=0,W=1),Translation=(X=0,Y=0,Z=-6),Scale3D=(X=1,Y=1,Z=1))))
    WeightMinimum=0
    Weight=&RigVMFunction_MathDoubleMul(
     B=@Multiply.Result
     A=&RigVMFunction_MathDoubleRemap(SourceMaximum=-15, SourceMinimum=-5, TargetMaximum=1, TargetMinimum=0, Value=@"Get Delta Ref_3".Result.Rotation.Pitch).Result
    ).Result
   )
   &RigUnit_ModifyTransforms(
    ItemToModify=((Item=(Type=Bone,Name="dyn_cloth_fr_master"),Transform=(Rotation=(X=0,Y=0,Z=-0.5736,W=0.8192),Translation=(X=0,Y=10,Z=0),Scale3D=(X=1,Y=1,Z=1))))
    WeightMinimum=0
    ExecutePin=@RigVMFunction_ControlFlowBranch.True
    Weight=RigVMVariableNode() {variable=Plevis Relative Root}.Value
   )
   n28 = RigVMVariableNode() {variable=Left_Leg_SPR}
   > &RigUnit_ModifyTransforms(
    ItemToModify=((Item=(Type=Bone,Name="dyn_cloth_fr_master"),Transform=(Rotation=(X=0.2588,Y=0,Z=0,W=0.9659),Translation=(X=0,Y=0,Z=2),Scale3D=(X=1,Y=1,Z=1))))
    WeightMinimum=0
    Weight=@n28.Value
   )
   > Get Delta Ref_2_3_1_1 = RigVMFunctionReferenceNode(Item=(Type=Bone,Name="thigh_l"))
   > &RigUnit_ModifyTransforms(
    ItemToModify=((Item=(Type=Bone,Name="dyn_cloth_fr_master"),Transform=(Rotation=(X=0,Y=0,Z=0,W=1),Translation=(X=-3,Y=4,Z=2),Scale3D=(X=1,Y=1,Z=1))))
    WeightMinimum=0
    Weight=&RigVMFunction_MathDoubleMul(
     A=&RigVMFunction_MathDoubleRemap(SourceMaximum=-120, SourceMinimum=-100, TargetMaximum=1, TargetMinimum=0, Value=@"Get Delta Ref_2_3_1_1".Result.Rotation.Yaw).Result
     B=@n28.Value
    ).Result
   )
   > &RigUnit_ModifyTransforms(
    ItemToModify=((Item=(Type=Bone,Name="dyn_cloth_fr_master"),Transform=(Rotation=(X=-0.2588,Y=0,Z=0,W=0.9659),Translation=(X=0,Y=0,Z=-3),Scale3D=(X=1,Y=1,Z=1))))
    WeightMinimum=0
    Weight=RigVMVariableNode() {variable=Right_Leg_SPR}.Value
   )
   n29 = &RigVMFunction_MathDoubleRemap(SourceMaximum=1, SourceMinimum=0, TargetMaximum=0, TargetMinimum=1, Value=RigVMVariableNode() {variable=Is Moving}.Value)
   Clamp = &RigVMFunction_MathDoubleClamp(Value=&RigVMFunction_MathDoubleMul(B=1.5, A=RigVMVariableNode() {variable=Right_Leg_SPR}.Value).Result)
   n30 = RigVMVariableNode() {variable=Left_Leg_SPR}
   > &RigUnit_ModifyTransforms(
    ItemToModify=((Item=(Type=Bone,Name="dyn_cloth_fr_master"),Transform=(Rotation=(X=0,Y=0,Z=0.342,W=0.9397),Translation=(X=0,Y=-7,Z=0),Scale3D=(X=1,Y=1,Z=1))))
    WeightMinimum=0
    Weight=&RigVMFunction_MathDoubleMul(
     B=&RigVMFunction_KalmanFloat(BufferSize=16, Value=@n29.Result) {template="KalmanFilter::Execute(in Value,in BufferSize,out Result)"}.Result
     A=&RigVMFunction_MathDoubleMul(B=@Clamp.Result, A=@n30.Value).Result
    ).Result
   )
   > &RigUnit_ModifyTransforms(
    ItemToModify="((Item=(Type=Bone,Name=\"dyn_cloth_fr_master\"),Transform=(Rotation=(X=0,Y=0,Z=0.4226,W=0.9063),Translation=(X=5,Y=-10,Z=0),Scale3D=(X=1,Y=1,Z=1))),(Item=(Type=Bone,Name=\"dyn_cloth_back_l_master\"),Transform=(Rotation=(X=0,Y=0,Z=-0.7071,W=0.7071),Translation= ...(truncated)"
    WeightMinimum=0
    Weight=&RigVMFunction_MathDoubleMul(A=RigVMVariableNode() {variable=Is sliding}.Value, B=RigVMVariableNode() {variable=Calf knee sliding or dbno}.Value).Result
   )
   &RigUnit_ModifyTransforms(
    ItemToModify=((Item=(Type=Bone,Name="dyn_cloth_fr_master"),Transform=(Rotation=(X=0,Y=0,Z=-0.866,W=0.5),Translation=(X=18,Y=15,Z=0),Scale3D=(X=1,Y=1,Z=1))))
    WeightMinimum=0
    ExecutePin=@n20.True
    Weight=RigVMVariableNode() {variable=Calf knee sliding or dbno}.Value
   )
   n31 = &RigVMFunction_Sequence(ExecuteContext=@RigVMFunction_Sequence.C)
   Get Delta Ref_2 = RigVMFunctionReferenceNode(Item=(Type=Bone,Name="clavicle_r"), ExecuteContext=@n31.A)
   > &RigUnit_ModifyTransforms(
    ItemToModify=((Item=(Type=Bone,Name="dyn_collar_r_master"),Transform=(Rotation=(X=-0.087,Y=0.0608,Z=0.0053,W=0.9943),Translation=(X=0,Y=0,Z=0),Scale3D=(X=1,Y=1,Z=1))))
    WeightMinimum=0
    Weight=&RigVMFunction_MathDoubleRemap(SourceMaximum=30, SourceMinimum=20, TargetMaximum=1, TargetMinimum=0, Value=@"Get Delta Ref_2".Result.Rotation.Pitch).Result
   )
   > Get Delta Ref_2_1 = RigVMFunctionReferenceNode(Item=(Type=Bone,Name="clavicle_l"))
   > &RigUnit_ModifyTransforms(
    ItemToModify=((Item=(Type=Bone,Name="dyn_collar_l_master"),Transform=(Rotation=(X=-0.0868,Y=0.0868,Z=0.0076,W=0.9924),Translation=(X=0,Y=0,Z=0),Scale3D=(X=1,Y=1,Z=1))))
    WeightMinimum=0
    Weight=&RigVMFunction_MathDoubleRemap(SourceMaximum=30, SourceMinimum=20, TargetMaximum=1, TargetMinimum=0, Value=@"Get Delta Ref_2_1".Result.Rotation.Pitch).Result
   )
   Get Delta Ref = RigVMFunctionReferenceNode(Item=(Type=Bone,Name="foot_l"), ExecuteContext=@n31.B)
   Remap = &RigVMFunction_MathDoubleRemap(SourceMaximum=60, SourceMinimum=35, TargetMaximum=1, TargetMinimum=0, Value=@"Get Delta Ref".Result.Rotation.Yaw)
   > &RigUnit_ModifyTransforms(
    ItemToModify=((Item=(Type=Bone,Name="foot_corrective_l"),Transform=(Rotation=(X=0,Y=0,Z=0,W=1),Translation=(X=-12,Y=-10,Z=0),Scale3D=(X=1,Y=1,Z=1))))
    WeightMinimum=0
    Weight=@Remap.Result
   )
   > Get Delta Ref_1 = RigVMFunctionReferenceNode(Item=(Type=Bone,Name="foot_r"))
   > &RigUnit_ModifyTransforms(
    ItemToModify=((Item=(Type=Bone,Name="foot_corrective_r"),Transform=(Rotation=(X=0,Y=0,Z=0,W=1),Translation=(X=12,Y=10,Z=0),Scale3D=(X=1,Y=1,Z=1))))
    WeightMinimum=0
    Weight=&RigVMFunction_MathDoubleRemap(SourceMaximum=60, SourceMinimum=35, TargetMaximum=1, TargetMinimum=0, Value=@"Get Delta Ref_1".Result.Rotation.Yaw).Result
   )
   &RigUnit_ModifyTransforms(
    ExecutePin=()
    ItemToModify=((Item=(Type=Bone,Name="pelvis"),Transform=(Rotation=(X=0,Y=0,Z=0,W=1),Translation=(X=0,Y=0,Z=0),Scale3D=(X=1,Y=1,Z=1))))
    Weight=1
    WeightMinimum=0
   )
   &RigVMFunction_MathQuaternionToEuler(RotationOrder=ZYX, Value=@n3.RelativeTransform.Rotation)
   RigVMVariableNode() {variable=Is targeting}
  ]}
]
hierarchy [
 root {loc=<0 0 0>, rot=<0 0 0 1>}
 pelvis {parent=root, loc=<0 -0.7071 93.1191>, rot=<0.0127 0.707 -0.0127 -0.707>}
 spine_01 {parent=pelvis, loc=<6.7494 -0.6388 0>, rot=<0 0 0.034 -0.9994>}
 spine_02 {parent=spine_01, loc=<7.2518 -0.035 0>, rot=<0 0 0.0238 -0.9997>}
 spine_03 {parent=spine_02, loc=<6.0104 0 0>, rot=<0 0 0.0149 -0.9999>}
 spine_04 {parent=spine_03, loc=<8.1024 0 0>, rot=<0 0 -0.0132 -0.9999>}
 spine_05 {parent=spine_04, loc=<10.0146 0 0>, rot=<0 0 -0.0015 -1>}
 attach_cape {parent=spine_05, loc=<9.017 -11.714 0>, rot=<0.0616 -0.7044 -0.0616 -0.7044>}
 attach_backpack {parent=spine_05, loc=<-0.123 -13.473 0>, rot=<-0.0093 -0.707 0.0093 -0.707>}
 clavicle_l {parent=spine_05, loc=<4.3575 -0.1466 -2.3223>, rot=<0.0465 -0.7289 -0.009 -0.6829>}
 upperarm_l {parent=clavicle_l, loc=<15.5111 0 0>, rot=<0.0226 -0.3234 -0.0088 -0.9459>}
 lowerarm_l {parent=upperarm_l, loc=<26.6514 0 0>, rot=<0.0391 -0.0099 -0.2453 -0.9686>}
 hand_l {parent=lowerarm_l, loc=<22.3491 0 0>, rot=<-0.7081 -0.0507 0.0975 -0.6975>}
 thumb_01_l {parent=hand_l, loc=<1.3118 -0.3191 -2.459>, rot=<0.6546 -0.4042 0.0664 -0.6354>}
 thumb_02_l {parent=thumb_01_l, loc=<5.9336 -0.0448 0.0401>, rot=<-0.0037 0.0213 0.298 -0.9543>}
 thumb_03_l {parent=thumb_02_l, loc=<2.2223 0.3682 -0.0907>, rot=<0.0118 -0.013 0.0066 -0.9998>}
 index_metacarpal_l {parent=hand_l, loc=<1.6535 0.0183 -2.96>, rot=<0.0158 -0.0124 -0.0173 -0.9996>}
 index_01_l {parent=index_metacarpal_l, loc=<8.3008 0.0357 -0.0524>, rot=<-0.0044 0.0485 0.0924 -0.9945>}
 index_02_l {parent=index_01_l, loc=<3.3587 -0.3004 0.0452>, rot=<-0.002 0.0027 0.3225 -0.9466>}
 index_03_l {parent=index_02_l, loc=<2.5578 -0.123 -0.0252>, rot=<-0.002 0.0067 0.2558 -0.9667>}
 middle_metacarpal_l {parent=hand_l, loc=<2.1856 -0.1725 -0.5984>, rot=<0.006 0.0168 -0.0329 -0.9993>}
 middle_01_l {parent=middle_metacarpal_l, loc=<7.3275 -0.2447 0.0129>, rot=<-0.0208 0.0042 0.0858 -0.9961>}
 middle_02_l {parent=middle_01_l, loc=<4.0123 0.1401 0.1175>, rot=<-0.0218 0.0258 0.3512 -0.9357>}
 middle_03_l {parent=middle_02_l, loc=<2.6077 0.0534 0.1373>, rot=<0.0082 0.0188 0.1708 -0.9851>}
 ring_metacarpal_l {parent=hand_l, loc=<2.2074 -0.3625 1.3638>, rot=<-0.0284 0.0722 -0.0458 -0.9959>}
 ring_01_l {parent=ring_metacarpal_l, loc=<7.0295 -0.6197 -0.1446>, rot=<-0.0489 -0.0252 0.1892 -0.9804>}
 ring_02_l {parent=ring_01_l, loc=<3.6174 0.4647 -0.1027>, rot=<0 0 0.2806 -0.9598>}
 ring_03_l {parent=ring_02_l, loc=<2.4152 0.114 0.0199>, rot=<0 0 0.1189 -0.9929>}
 pinky_metacarpal_l {parent=hand_l, loc=<1.7616 -0.9753 3.3814>, rot=<-0.0397 0.0814 -0.069 -0.9935>}
 pinky_01_l {parent=pinky_metacarpal_l, loc=<7.1709 -0.4831 -0.1094>, rot=<-0.0654 -0.0035 0.2242 -0.9723>}
 pinky_02_l {parent=pinky_01_l, loc=<2.8274 0.0893 -0.1043>, rot=<0.0037 0.0011 0.2854 -0.9584>}
 pinky_03_l {parent=pinky_02_l, loc=<2.2046 0.0307 -0.0078>, rot=<-0.0002 0.0098 0.1445 -0.9895>}
 deform_wrist_dn_base_l {parent=hand_l, loc=<0 -2.2342 0>, rot=<0 0 0 -1>}
 deform_wrist_dn_l {parent=deform_wrist_dn_base_l, loc=<0 0 0>, rot=<0 0 0 1>}
 deform_wrist_up_base_l {parent=hand_l, loc=<0 2.2682 0>, rot=<0 0 0 -1>}
 deform_wrist_up_l {parent=deform_wrist_up_base_l, loc=<0 0 0>, rot=<0 0 0 1>}
 lowerarm_twist_02_l {parent=lowerarm_l, loc=<9.6199 0 0>, rot=<0 0 0 -1>}
 lowerarm_twist_01_l {parent=lowerarm_l, loc=<17.8549 0 0>, rot=<0 0 0 -1>}
 deform_elbow_in_base_l {parent=lowerarm_l, loc=<0.2112 5.259 0.8274>, rot=<0 0 0 -1>}
 deform_elbow_in_l {parent=deform_elbow_in_base_l, loc=<0 0 0>, rot=<0 0 0 1>}
 deform_elbow_out_base_l {parent=lowerarm_l, loc=<0 -4 0>, rot=<0 0 0 -1>}
 deform_elbow_out_l {parent=deform_elbow_out_base_l, loc=<0 0 0>, rot=<0 0 0 1>}
 upperarm_twist_01_l {parent=upperarm_l, loc=<0.0041 0 0>, rot=<0 0 0 1>}
 deform_upperarm_bk_base_l {parent=upperarm_twist_01_l, loc=<1.0813 -7.2193 -1.9453>, rot=<0 0 0 1>}
 deform_upperarm_bk_l {parent=deform_upperarm_bk_base_l, loc=<0 0 0>, rot=<0 0 0 1>}
 deform_upperarm_fr_base_l {parent=upperarm_twist_01_l, loc=<-0.3691 7.0506 -2.765>, rot=<0 0 0 1>}
 deform_upperarm_fr_l {parent=deform_upperarm_fr_base_l, loc=<0 0 0>, rot=<0 0 0 1>}
 deform_upperarm_up_base_l {parent=upperarm_twist_01_l, loc=<-3.2216 -1.1391 3.3311>, rot=<0 0 0 1>}
 deform_upperarm_up_l {parent=deform_upperarm_up_base_l, loc=<0 0 0>, rot=<0 0 0 1>}
 upperarm_twist_02_l {parent=upperarm_l, loc=<13.7941 0 0>, rot=<0 0 0 1>}
 dyn_shoulder_l_master {parent=clavicle_l, loc=<8.7859 -1.872 6.9454>, rot=<-0.0281 -0.9965 -0.0123 -0.0775>}
 dyn_shoulder_l_01 {parent=dyn_shoulder_l_master, loc=<0 0 0>, rot=<0 0 0 1>}
 dyn_shoulder_l_02 {parent=dyn_shoulder_l_01, loc=<-5.7986 0 0.0005>, rot=<0 0 0 1>}
 neck_01 {parent=spine_05, loc=<11.3138 0.1529 0>, rot=<0 0 -0.2003 -0.9797>}
 neck_02 {parent=neck_01, loc=<4.5627 0 0>, rot=<0 0 0.0504 -0.9987>}
 head {parent=neck_02, loc=<5.3914 0 0>, rot=<0 0 0.115 -0.9934>}
 clavicle_r {parent=spine_05, loc=<4.3575 -0.1466 2.3223>, rot=<0.7289 0.0465 -0.6829 0.009>}
 upperarm_r {parent=clavicle_r, loc=<-15.5111 0 0>, rot=<0.0226 -0.3234 -0.0088 -0.9459>}
 lowerarm_r {parent=upperarm_r, loc=<-26.6514 0 0>, rot=<0.0391 -0.0099 -0.2453 -0.9686>}
 hand_r {parent=lowerarm_r, loc=<-22.3491 0 0>, rot=<-0.7081 -0.0507 0.0975 -0.6975>}
 thumb_01_r {parent=hand_r, loc=<-1.3118 0.3191 2.459>, rot=<0.6546 -0.4042 0.0664 -0.6354>}
 thumb_02_r {parent=thumb_01_r, loc=<-5.9336 0.0448 -0.0401>, rot=<-0.0037 0.0213 0.298 -0.9543>}
 thumb_03_r {parent=thumb_02_r, loc=<-2.2223 -0.3682 0.0907>, rot=<0.0118 -0.013 0.0066 -0.9998>}
 index_metacarpal_r {parent=hand_r, loc=<-1.6535 -0.0183 2.96>, rot=<0.0158 -0.0124 -0.0173 -0.9996>}
 index_01_r {parent=index_metacarpal_r, loc=<-8.3008 -0.0357 0.0524>, rot=<-0.0044 0.0485 0.0924 -0.9945>}
 index_02_r {parent=index_01_r, loc=<-3.3587 0.3004 -0.0452>, rot=<-0.002 0.0027 0.3225 -0.9466>}
 index_03_r {parent=index_02_r, loc=<-2.5578 0.123 0.0252>, rot=<-0.002 0.0067 0.2558 -0.9667>}
 middle_metacarpal_r {parent=hand_r, loc=<-2.1856 0.1725 0.5984>, rot=<0.006 0.0168 -0.0329 -0.9993>}
 middle_01_r {parent=middle_metacarpal_r, loc=<-7.3275 0.2447 -0.0129>, rot=<-0.0208 0.0042 0.0858 -0.9961>}
 middle_02_r {parent=middle_01_r, loc=<-4.0123 -0.1401 -0.1175>, rot=<-0.0218 0.0258 0.3512 -0.9357>}
 middle_03_r {parent=middle_02_r, loc=<-2.6077 -0.0534 -0.1373>, rot=<0.0082 0.0188 0.1708 -0.9851>}
 ring_metacarpal_r {parent=hand_r, loc=<-2.2074 0.3625 -1.3638>, rot=<-0.0284 0.0722 -0.0458 -0.9959>}
 ring_01_r {parent=ring_metacarpal_r, loc=<-7.0295 0.6197 0.1446>, rot=<-0.0489 -0.0252 0.1892 -0.9804>}
 ring_02_r {parent=ring_01_r, loc=<-3.6174 -0.4647 0.1027>, rot=<0 0 0.2806 -0.9598>}
 ring_03_r {parent=ring_02_r, loc=<-2.4152 -0.114 -0.0199>, rot=<0 0 0.1189 -0.9929>}
 pinky_metacarpal_r {parent=hand_r, loc=<-1.7616 0.9753 -3.3814>, rot=<-0.0397 0.0814 -0.069 -0.9935>}
 pinky_01_r {parent=pinky_metacarpal_r, loc=<-7.1709 0.4831 0.1094>, rot=<-0.0654 -0.0035 0.2242 -0.9723>}
 pinky_02_r {parent=pinky_01_r, loc=<-2.8274 -0.0893 0.1043>, rot=<0.0037 0.0011 0.2854 -0.9584>}
 pinky_03_r {parent=pinky_02_r, loc=<-2.2046 -0.0307 0.0078>, rot=<-0.0002 0.0098 0.1445 -0.9895>}
 deform_wrist_dn_base_r {parent=hand_r, loc=<-0.0004 2.2345 -0.0001>, rot=<0 0 0 -1>}
 deform_wrist_dn_r {parent=deform_wrist_dn_base_r, loc=<0 0 0>, rot=<0 0 0 1>}
 deform_wrist_up_base_r {parent=hand_r, loc=<0.0003 -2.2684 0.0001>, rot=<0 0 0 -1>}
 deform_wrist_up_r {parent=deform_wrist_up_base_r, loc=<0 0 0>, rot=<0 0 0 1>}
 lowerarm_twist_02_r {parent=lowerarm_r, loc=<-9.6199 0 0>, rot=<0 0 0 -1>}
 lowerarm_twist_01_r {parent=lowerarm_r, loc=<-17.8549 0 0>, rot=<0 0 0 -1>}
 deform_elbow_in_base_r {parent=lowerarm_r, loc=<-0.2111 -5.2591 -0.8275>, rot=<0 0 0 -1>}
 deform_elbow_in_r {parent=deform_elbow_in_base_r, loc=<-0.0001 0 0.0002>, rot=<0 0 0 1>}
 deform_elbow_out_base_r {parent=lowerarm_r, loc=<0 4 -0.0001>, rot=<0 0 0 -1>}
 deform_elbow_out_r {parent=deform_elbow_out_base_r, loc=<0 0 0>, rot=<0 0 0 1>}
 upperarm_twist_01_r {parent=upperarm_r, loc=<-0.0041 0 0>, rot=<0 0 0 -1>}
 deform_upperarm_bk_base_r {parent=upperarm_twist_01_r, loc=<-1.0814 7.2192 1.9455>, rot=<0 0 0 1>}
 deform_upperarm_bk_r {parent=deform_upperarm_bk_base_r, loc=<0 0 0>, rot=<0 0 0 1>}
 deform_upperarm_fr_base_r {parent=upperarm_twist_01_r, loc=<0.3689 -7.0506 2.7652>, rot=<0 0 0 1>}
 deform_upperarm_fr_r {parent=deform_upperarm_fr_base_r, loc=<0 0 0>, rot=<0 0 0 1>}
 deform_upperarm_up_base_r {parent=upperarm_twist_01_r, loc=<3.2214 1.1391 -3.3308>, rot=<0 0 0 1>}
 deform_upperarm_up_r {parent=deform_upperarm_up_base_r, loc=<0 0 0>, rot=<0 0 0 -1>}
 upperarm_twist_02_r {parent=upperarm_r, loc=<-13.7941 0 0>, rot=<0 0 0 -1>}
 dyn_shoulder_r_master {parent=clavicle_r, loc=<-8.7858 1.872 -6.9454>, rot=<-0.0281 -0.9965 -0.0123 -0.0775>}
 dyn_shoulder_r_01 {parent=dyn_shoulder_r_master, loc=<0 0 0>, rot=<0 0 0 1>}
 dyn_shoulder_r_02 {parent=dyn_shoulder_r_01, loc=<5.7987 0 0>, rot=<0 0 0 1>}
 Collar_corrective_ads {parent=clavicle_r, loc=<-11.7557 -5.0479 -5.2307>, rot=<0.9649 0.0156 0.0389 0.2594>}
 dyn_collar_r_master {parent=spine_05, loc=<11.4732 0.2315 8.003>, rot=<-0.1681 0.3919 0.0277 -0.9041>}
 dyn_collar_r_01 {parent=dyn_collar_r_master, loc=<0 0 0>, rot=<0 0 0 1>}
 dyn_collar_r_02 {parent=dyn_collar_r_01, loc=<5.1768 0 0>, rot=<0 -0.0104 0 -0.9999>}
 dyn_collar_r_03 {parent=dyn_collar_r_02, loc=<4.936 0 0>, rot=<0 0 0 1>}
 dyn_collar_l_master {parent=spine_05, loc=<11.4733 0.2315 -8.003>, rot=<0.3919 0.1681 0.9041 0.0277>}
 dyn_collar_l_01 {parent=dyn_collar_l_master, loc=<0 0 0>, rot=<0 0 0 1>}
 dyn_collar_l_02 {parent=dyn_collar_l_01, loc=<-5.177 0 0.0002>, rot=<0 -0.0104 0 -0.9999>}
 dyn_collar_l_03 {parent=dyn_collar_l_02, loc=<-4.9356 0 -0.0004>, rot=<0 0 0 -1>}
 dyn_chestSkull_m_master {parent=spine_05, loc=<0.9793 13.2115 0.3332>, rot=<-0.5727 -0.4147 -0.4147 -0.5727>}
 dyn_chestSkull_m_01 {parent=dyn_chestSkull_m_master, loc=<0 0 0>, rot=<0 0 0 1>}
 dyn_collar_m_master {parent=spine_05, loc=<13.2934 -6.1398 0>, rot=<0.6806 -0.1918 0.1918 -0.6806>}
 dyn_collar_m_01 {parent=dyn_collar_m_master, loc=<0 0 0>, rot=<0 0 0 1>}
 dyn_collar_m_02 {parent=dyn_collar_m_01, loc=<3.9051 0 0>, rot=<0 0 0 1>}
 cape_upper_master {parent=spine_05, loc=<5.793 -12.5382 0>, rot=<0 0 0.9903 -0.1391>}
 cape_upper_01 {parent=cape_upper_master, loc=<0 0 0>, rot=<0 0 0 1>}
 cape_upper_02 {parent=cape_upper_01, loc=<5.5552 0 0>, rot=<0 0 0.0766 -0.9971>}
 deform_pec_aim_base_l {parent=spine_04, loc=<7.8732 12.6388 -2.4123>, rot=<0.0102 -0.4602 0.3979 -0.7936>}
 deform_pec_aim_l {parent=deform_pec_aim_base_l, loc=<0 0 0>, rot=<0 0 0 1>}
 deform_pec_base_l {parent=deform_pec_aim_l, loc=<10.766 0 0>, rot=<0 0 0 1>}
 deform_pec_l {parent=deform_pec_base_l, loc=<0.055 0 0>, rot=<0 0 0 1>}
 deform_pec_aim_base_r {parent=spine_04, loc=<7.8727 12.6388 2.4123>, rot=<-0.4602 -0.0102 0.7936 0.3979>}
 deform_pec_aim_r {parent=deform_pec_aim_base_r, loc=<0 0 0>, rot=<0 0 0 1>}
 deform_pec_base_r {parent=deform_pec_aim_r, loc=<-10.7662 -0.0006 -0.0007>, rot=<0 0 0 1>}
 deform_pec_r {parent=deform_pec_base_r, loc=<0 0 0>, rot=<0 0 0 1>}
 dyn_cloth_back_r_master {parent=pelvis, loc=<4.5 -12 6.5>, rot=<0 -1 0 0>}
 dyn_cloth_back_r_02 {parent=dyn_cloth_back_r_master, loc=<0 0 0>, rot=<0 0 0.0616 -0.9981>}
 dyn_cloth_back_r_03 {parent=dyn_cloth_back_r_02, loc=<9.1999 0 0>, rot=<0 0 -0.0711 -0.9975>}
 dyn_cloth_back_r_04 {parent=dyn_cloth_back_r_03, loc=<10.5234 0 0>, rot=<0 0 0 1>}
 dyn_cloth_back_l_master {parent=pelvis, loc=<4.5 -12 -6.5>, rot=<0 -1 0 0>}
 dyn_cloth_back_l_02 {parent=dyn_cloth_back_l_master, loc=<0 0 0>, rot=<0 0 0.0616 -0.9981>}
 dyn_cloth_back_l_03 {parent=dyn_cloth_back_l_02, loc=<9.1999 0.0001 0>, rot=<0 0 -0.0711 -0.9975>}
 dyn_cloth_back_l_04 {parent=dyn_cloth_back_l_03, loc=<10.5234 0 0>, rot=<0 0 0 1>}
 thigh_l {parent=pelvis, loc=<-2.5693 -0.0109 -7.7501>, rot=<0.0012 0.0572 0.0155 -0.9982>}
 calf_l {parent=thigh_l, loc=<-39.0484 0 0>, rot=<0.0006 -0.0161 -0.0391 -0.9991>}
 foot_l {parent=calf_l, loc=<-42.561 0.1077 0.4292>, rot=<0.0418 -0.0455 0.0343 -0.9975>}
 ball_l {parent=foot_l, loc=<-6.1599 13.8264 0.6806>, rot=<0.0001 -0.0001 -0.7063 -0.7079>}
 foot_corrective_l {parent=foot_l, loc=<-3 -5 0>, rot=<0 0 0 1>}
 calf_twist_02_l {parent=calf_l, loc=<-18.8064 -0.0967 0.0397>, rot=<0.0237 0.0095 -0.0035 -0.9997>}
 calf_twist_01_l {parent=calf_l, loc=<-35.813 -0.0967 0.0397>, rot=<0.0237 0.0095 -0.0035 -0.9997>}
 deform_knee_in_base_l {parent=calf_l, loc=<1.298 -5.2957 -0.0496>, rot=<0 0 0 -1>}
 deform_knee_in_l {parent=deform_knee_in_base_l, loc=<0 0 0>, rot=<0 0 0 1>}
 deform_knee_out_base_l {parent=calf_l, loc=<0 4 0>, rot=<0 0 0 -1>}
 deform_knee_out_l {parent=deform_knee_out_base_l, loc=<0 0 0>, rot=<0 0 0 1>}
 thigh_twist_01_l {parent=thigh_l, loc=<-21.6303 0 0>, rot=<0.0013 0 0 -1>}
 deform_groin_base_l {parent=thigh_l, loc=<-2.1908 8.1142 -1.7442>, rot=<0 0 0 -1>}
 deform_groin_l {parent=deform_groin_base_l, loc=<0 0 0>, rot=<0 0 0 1>}
 thigh_r {parent=pelvis, loc=<-2.5693 -0.0109 7.7501>, rot=<0.0572 -0.0012 0.9982 0.0155>}
 calf_r {parent=thigh_r, loc=<39.0484 0 0>, rot=<0.0006 -0.0161 -0.0391 -0.9991>}
 foot_r {parent=calf_r, loc=<42.561 -0.1077 -0.4292>, rot=<0.0418 -0.0455 0.0343 -0.9975>}
 ball_r {parent=foot_r, loc=<6.1599 -13.8264 -0.6806>, rot=<0.0001 -0.0001 -0.7063 -0.7079>}
 foot_corrective_r {parent=foot_r, loc=<3 5 0>, rot=<0 0 0 1>}
 calf_twist_02_r {parent=calf_r, loc=<18.8064 0.0967 -0.0397>, rot=<0.0237 0.0095 -0.0035 -0.9997>}
 calf_twist_01_r {parent=calf_r, loc=<35.813 0.0967 -0.0397>, rot=<0.0237 0.0095 -0.0035 -0.9997>}
 deform_knee_in_base_r {parent=calf_r, loc=<-1.298 5.2957 0.0496>, rot=<0 0 0 -1>}
 deform_knee_in_r {parent=deform_knee_in_base_r, loc=<0 0 0>, rot=<0 0 0 1>}
 deform_knee_out_base_r {parent=calf_r, loc=<0 -4 0>, rot=<0 0 0 -1>}
 deform_knee_out_r {parent=deform_knee_out_base_r, loc=<0 0 0>, rot=<0 0 0 1>}
 thigh_twist_01_r {parent=thigh_r, loc=<21.6303 0 0>, rot=<0.0013 0 0 -1>}
 deform_groin_base_r {parent=thigh_r, loc=<2.1908 -8.1142 1.7442>, rot=<0 0 0 -1>}
 deform_groin_r {parent=deform_groin_base_r, loc=<0 0 0>, rot=<0 0 0 1>}
 deform_glute_base_l {parent=pelvis, loc=<-8.8064 -8.8222 -8.4287>, rot=<0.0275 0.0501 -0.4472 -0.8926>}
 deform_glute_l {parent=deform_glute_base_l, loc=<0 0 0>, rot=<0 0 0 1>}
 deform_glute_base_r {parent=pelvis, loc=<-8.8062 -8.8222 8.4287>, rot=<0.0501 -0.0275 0.8926 -0.4472>}
 deform_glute_r {parent=deform_glute_base_r, loc=<0 0 0>, rot=<0 0 0 1>}
 dyn_cloth_fr_master {parent=pelvis, loc=<4.8 9 0>, rot=<0 0 -0.0179 -0.9998>}
 dyn_cloth_front_l_master {parent=dyn_cloth_fr_master, loc=<0 0 -4.6>, rot=<-0.0244 -0.9997 0 0>}
 dyn_cloth_front_l_01 {parent=dyn_cloth_front_l_master, loc=<0 0 0>, rot=<0 0 -0.0502 -0.9987>}
 dyn_cloth_front_l_02 {parent=dyn_cloth_front_l_01, loc=<16.307 0 0>, rot=<0 0 0.0222 -0.9998>}
 dyn_cloth_front_l_03 {parent=dyn_cloth_front_l_02, loc=<17.0049 0 0>, rot=<0 0 0.0104 -0.9999>}
 dyn_cloth_front_l_04 {parent=dyn_cloth_front_l_03, loc=<18.5256 0 0>, rot=<0 0 0 1>}
 dyn_cloth_front_r_master {parent=dyn_cloth_fr_master, loc=<0 0 4.6>, rot=<0 -1 0 0>}
 dyn_cloth_front_r_01 {parent=dyn_cloth_front_r_master, loc=<0 0 0>, rot=<0 0 -0.0502 -0.9987>}
 dyn_cloth_front_r_02 {parent=dyn_cloth_front_r_01, loc=<16.307 0 0>, rot=<0 0 0.0222 -0.9998>}
 dyn_cloth_front_r_03 {parent=dyn_cloth_front_r_02, loc=<17.0049 0 0>, rot=<0 0 0.0104 -0.9999>}
 dyn_cloth_front_r_04 {parent=dyn_cloth_front_r_03, loc=<18.5256 0 0>, rot=<0 0 0 1>}
]
variables [
 Calf knee sliding or dbno {type=double, default=0}
 Is Moving {type=double, default=0}
 Is crouch {type=bool, default=False}
 Is hands up {type=double, default=0}
 Is skydive {type=bool, default=False}
 Is sliding {type=double, default=0}
 Is targeting {type=double, default=0}
 Left_Leg_SPR {type=double, default=0}
 NewVar {type=bool, default=False}
 Plevis Relative Root {type=double, default=0}
 Right_Leg_SPR {type=double, default=0}
 is dbno {type=bool, default=False}
 left Hand up {type=double, default=0}
]
