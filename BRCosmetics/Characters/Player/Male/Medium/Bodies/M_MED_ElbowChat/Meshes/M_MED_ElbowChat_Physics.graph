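# Physics asset for the M_MED_ElbowChat male player body, giving its four dyn_cloth bone chains bodies for secondary cloth motion.
# Pelvis and leg capsules cover the lower body, and a wide, flat box sits at the root.
# Front and back chains use angle-limited joints under master bones; only the back masters are constrained to the pelvis.
# Angular-free constraints also cross-link the lower three left and right front segments.
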PhysicsAsset /BRCosmetics/Characters/Player/Male/Medium/Bodies/M_MED_ElbowChat/Meshes/M_MED_ElbowChat_Physics
bodies {
 calf_l [{
   shapeType=Capsule
   length=44.5314
   radius=6.9559
   shapeName=calf_r_capsule
   center=<-17.5235 -0.146 0.3026>
   rotation=<rot -89.321 -0.0008 178.3652>
  }]
 calf_r [{
   shapeType=Capsule
   length=41.0061
   radius=6.9559
   shapeName=calf_r_capsule
   center=<17.9576 2.4041 -0.2975>
   rotation=<rot -89.321 -0.0005 0.0003>
  }]
 dyn_cloth_back_l_02 [{shapeType=Capsule, length=9, radius=6, shapeName=dyn_cloth_back_l_02_capsule, center=<0 -4.6019 0>}]
 dyn_cloth_back_l_03 [{shapeType=Capsule, length=9, radius=6, shapeName=dyn_cloth_back_l_03_capsule, center=<1.7218 -3.4159 0>}]
 dyn_cloth_back_l_04 [{shapeType=Capsule, length=9, radius=6, shapeName=dyn_cloth_back_l_04_capsule, center=<3.7219 -3 0>}]
 dyn_cloth_back_l_master [{shapeType=Capsule, length=0.505, radius=0.505, shapeName=dyn_cloth_back_l_master_capsule, rotation=<rot 0 -90 0>}]
 dyn_cloth_back_r_02 [{shapeType=Capsule, length=9, radius=6.5085, shapeName=dyn_cloth_back_r_02_capsule, center=<0 -4.163 0>}]
 dyn_cloth_back_r_03 [{shapeType=Capsule, length=9, radius=6, shapeName=dyn_cloth_back_r_03_capsule, center=<1.7218 -3.0703 0>}]
 dyn_cloth_back_r_04 [{shapeType=Capsule, length=9, radius=6, shapeName=dyn_cloth_back_r_04_capsule, center=<3.7219 -3 0>}]
 dyn_cloth_back_r_master [{shapeType=Capsule, length=0.505, radius=0.505, shapeName=dyn_cloth_back_r_master_capsule, rotation=<rot 0 -90 0>}]
 dyn_cloth_front_l_01 [{
   shapeType=Capsule
   length=13.0345
   radius=4.231
   shapeName=dyn_cloth_front_l_01_capsule
   center=<6.4126 3.2231 -0.255>
   rotation=<rot -90 180 180>
  }]
 dyn_cloth_front_l_02 [{
   shapeType=Capsule
   length=14.7086
   radius=4.9635
   shapeName=dyn_cloth_front_l_02_capsule
   center=<7.0527 3.5702 -1.3582>
   rotation=<rot -89.8632 -179.9999 -180>
  }]
 dyn_cloth_front_l_03 [{
   shapeType=Capsule
   length=13.0896
   radius=4.3523
   shapeName=dyn_cloth_front_l_03_capsule
   center=<8.0558 2.785 -0.8798>
   rotation=<rot -89.5187 0.0002 0>
  }]
 dyn_cloth_front_l_04 [{shapeType=Capsule, length=5.0318, radius=4.3523, shapeName=dyn_cloth_front_l_04_capsule, center=<6.9693 2.785 0>, rotation=<rot -90 0 0>}]
 dyn_cloth_front_l_master [{shapeType=Capsule, length=0.505, radius=0.505, shapeName=dyn_cloth_front_l_master_capsule, rotation=<rot 0 -90 0>}]
 dyn_cloth_front_r_01 [{
   shapeType=Capsule
   length=12.5426
   radius=4.231
   shapeName=dyn_cloth_front_l_01_capsule
   center=<6.0286 2.8737 0.255>
   rotation=<rot -90 0 180>
  }]
 dyn_cloth_front_r_02 [{
   shapeType=Capsule
   length=14.7086
   radius=4.9635
   shapeName=dyn_cloth_front_l_02_capsule
   center=<6.979 3.5702 1.3247>
   rotation=<rot -89.8632 -179.9999 0>
  }]
 dyn_cloth_front_r_03 [{
   shapeType=Capsule
   length=13.0896
   radius=4.3523
   shapeName=dyn_cloth_front_l_03_capsule
   center=<7.4043 2.785 1.1801>
   rotation=<rot -89.5187 0.0002 -180>
  }]
 dyn_cloth_front_r_04 [{shapeType=Capsule, length=5.0318, radius=4.3523, shapeName=dyn_cloth_front_l_04_capsule, center=<8.1437 2.785 0>, rotation=<rot -90 180 0>}]
 dyn_cloth_front_r_master [{shapeType=Capsule, length=0.505, radius=0.505, shapeName=dyn_cloth_front_r_master_capsule, rotation=<rot 0 -90 0>}]
 foot_l [{
   shapeType=Capsule
   length=17.3186
   radius=8.2493
   shapeName=foot_r_capsule
   center=<-6.0026 5.0311 -0.0345>
   rotation=<rot -0.0212 -164.4644 90.0395>
  }]
 foot_r [{
   shapeType=Capsule
   length=17.3186
   radius=8.2493
   shapeName=foot_r_capsule
   center=<5.3549 -4.8421 0.0345>
   rotation=<rot -0.024 -160.4365 -89.962>
  }]
 pelvis [
  {shapeType=Capsule, length=19.3254, radius=10.1894, shapeName=pelvis_capsule, center=<-1.9254 -1.8272 0>, rotation=<rot 0 -90 0>}
  {shapeType=Capsule, length=22.6933, radius=7.3067, shapeName=pelvis_capsule, center=<0 1.808 0>}
 ]
 root [{shapeType=Box, extentX=134.912, extentY=140.463, extentZ=17.0619, shapeName=root_box, center=<0 0 -6.3867>}]
 thigh_l [{shapeType=Capsule, length=20.7201, radius=8.7291, shapeName=thigh_r_capsule, center=<-18.1557 0.2306 0.2375>, rotation=<rot 90 180 0.0005>}]
 thigh_r [{
   shapeType=Capsule
   length=20.7201
   radius=8.7291
   shapeName=thigh_r_capsule
   center=<18.1557 -0.2306 -0.2375>
   rotation=<rot 90 180 -179.9995>
  }]
}
constraints [
 {bone1Name=dyn_cloth_back_l_03, bone2Name=dyn_cloth_back_l_02, swing1LimitDegrees=25, swing2LimitDegrees=15, twistLimitDegrees=0}
 {bone1Name=dyn_cloth_back_l_04, bone2Name=dyn_cloth_back_l_03, swing1LimitDegrees=45, swing2LimitDegrees=15, twistLimitDegrees=0}
 {bone1Name=dyn_cloth_back_l_02, bone2Name=dyn_cloth_back_l_master, swing1LimitDegrees=80, swing2LimitDegrees=15, twistLimitDegrees=0}
 {bone1Name=dyn_cloth_back_r_02, bone2Name=dyn_cloth_back_r_master, swing1LimitDegrees=80, swing2LimitDegrees=15, twistLimitDegrees=10}
 {bone1Name=dyn_cloth_back_r_03, bone2Name=dyn_cloth_back_r_02, swing1LimitDegrees=25, swing2LimitDegrees=15, twistLimitDegrees=10}
 {bone1Name=dyn_cloth_back_r_04, bone2Name=dyn_cloth_back_r_03, swing1LimitDegrees=45, swing2LimitDegrees=25, twistLimitDegrees=0}
 {bone1Name=dyn_cloth_front_l_01, bone2Name=dyn_cloth_front_l_master, swing1LimitDegrees=25, swing2LimitDegrees=15, twistLimitDegrees=0}
 {bone1Name=dyn_cloth_front_l_02, bone2Name=dyn_cloth_front_l_01, swing1LimitDegrees=30, swing2LimitDegrees=8, twistLimitDegrees=5}
 {bone1Name=dyn_cloth_front_l_03, bone2Name=dyn_cloth_front_l_02, swing1LimitDegrees=35, swing2LimitDegrees=15, twistLimitDegrees=5}
 {bone1Name=dyn_cloth_front_l_04, bone2Name=dyn_cloth_front_l_03, swing1LimitDegrees=25, swing2LimitDegrees=15, twistLimitDegrees=5}
 {
  bone1Name=dyn_cloth_back_l_master
  bone2Name=pelvis
  swing1LimitDegrees=45
  swing1Motion=Free
  swing2LimitDegrees=45
  swing2Motion=Free
  twistLimitDegrees=45
  twistMotion=Free
 }
 {
  bone1Name=dyn_cloth_back_r_master
  bone2Name=pelvis
  swing1LimitDegrees=45
  swing1Motion=Free
  swing2LimitDegrees=45
  swing2Motion=Free
  twistLimitDegrees=45
  twistMotion=Free
 }
 {
  bone1Name=thigh_r
  bone2Name=pelvis
  swing1LimitDegrees=45
  swing1Motion=Free
  swing2LimitDegrees=45
  swing2Motion=Free
  twistLimitDegrees=45
  twistMotion=Free
 }
 {
  bone1Name=foot_r
  bone2Name=calf_r
  swing1LimitDegrees=45
  swing1Motion=Free
  swing2LimitDegrees=45
  swing2Motion=Free
  twistLimitDegrees=45
  twistMotion=Free
 }
 {
  bone1Name=thigh_l
  bone2Name=pelvis
  swing1LimitDegrees=45
  swing1Motion=Free
  swing2LimitDegrees=45
  swing2Motion=Free
  twistLimitDegrees=45
  twistMotion=Free
 }
 {
  bone1Name=calf_l
  bone2Name=thigh_l
  swing1LimitDegrees=45
  swing1Motion=Free
  swing2LimitDegrees=45
  swing2Motion=Free
  twistLimitDegrees=45
  twistMotion=Free
 }
 {
  bone1Name=foot_l
  bone2Name=calf_l
  swing1LimitDegrees=45
  swing1Motion=Free
  swing2LimitDegrees=45
  swing2Motion=Free
  twistLimitDegrees=45
  twistMotion=Free
 }
 {
  bone1Name=pelvis
  bone2Name=root
  swing1LimitDegrees=45
  swing1Motion=Free
  swing2LimitDegrees=45
  swing2Motion=Free
  twistLimitDegrees=45
  twistMotion=Free
 }
 {bone1Name=dyn_cloth_front_r_02, bone2Name=dyn_cloth_front_r_01, swing1LimitDegrees=30, swing2LimitDegrees=8, twistLimitDegrees=5}
 {bone1Name=dyn_cloth_front_r_03, bone2Name=dyn_cloth_front_r_02, swing1LimitDegrees=35, swing2LimitDegrees=15, twistLimitDegrees=5}
 {bone1Name=dyn_cloth_front_r_04, bone2Name=dyn_cloth_front_r_03, swing1LimitDegrees=25, swing2LimitDegrees=15, twistLimitDegrees=5}
 {bone1Name=dyn_cloth_front_r_01, bone2Name=dyn_cloth_front_r_master, swing1LimitDegrees=25, swing2LimitDegrees=15, twistLimitDegrees=0}
 {
  bone1Name=calf_r
  bone2Name=thigh_r
  swing1LimitDegrees=45
  swing1Motion=Free
  swing2LimitDegrees=45
  swing2Motion=Free
  twistLimitDegrees=45
  twistMotion=Free
 }
 {
  bone1Name=dyn_cloth_front_r_04
  bone2Name=dyn_cloth_front_l_04
  swing1LimitDegrees=45
  swing1Motion=Free
  swing2LimitDegrees=45
  swing2Motion=Free
  twistLimitDegrees=45
  twistMotion=Free
 }
 {
  bone1Name=dyn_cloth_front_r_03
  bone2Name=dyn_cloth_front_l_03
  swing1LimitDegrees=45
  swing1Motion=Free
  swing2LimitDegrees=45
  swing2Motion=Free
  twistLimitDegrees=45
  twistMotion=Free
 }
 {
  bone1Name=dyn_cloth_front_r_02
  bone2Name=dyn_cloth_front_l_02
  swing1LimitDegrees=45
  swing1Motion=Free
  swing2LimitDegrees=45
  swing2Motion=Free
  twistLimitDegrees=45
  twistMotion=Free
 }
]
