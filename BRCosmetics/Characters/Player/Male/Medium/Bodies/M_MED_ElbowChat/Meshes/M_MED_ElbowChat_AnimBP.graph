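# Post-process style AnimBP for the M_MED_ElbowChat male body cosmetic, adding secondary motion to a pose copied from the attached parent mesh.
# bClothEnabled selects a cloth Control Rig or a Control Rig, RigidBody and fallback Trail chain.
# AnimDynamics then springs collar and upper cape, and a weighted constraint drives the chest skull.
# The EventGraph derives movement states, selects wind gusts, sets gravity and runs deprecated cloth controller functions.
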AnimBlueprint /BRCosmetics/Characters/Player/Male/Medium/Bodies/M_MED_ElbowChat/Meshes/M_MED_ElbowChat_AnimBP
target_skeleton=&/M_MED_ElbowChat_Skeleton
graphs [
 AnimGraph {nodes [
   n1 = Control Rig!(
    Source=Use cached pose 'PerPose'()
    Alpha
    Is crouch:bool=Get bIsCrouching()
    is dbno:bool=Get Is dbno no carry() -> (Is dbno no carry:bool)
    Is Moving:double=Get Alpha is moving forward() -> (Alpha is moving forward:double)
    Is skydive:bool=Get IS skydive no parachute() -> (IS skydive no parachute:bool)
    Is targeting:double=Get Is targeting() -> (Is targeting:double)
   ) -> (Pose) {
    alphaBoolBlend {blendOption=Linear}
    controlRigAssetReference {blueprintRigClass=&/M_MED_ElbowChat_CtrlRig.M_MED_ElbowChat_CtrlRig_C}
   }
   n2 = RigidBody!(ComponentPose=Local To Component(LocalPose=@n1), Alpha, OverrideWorldGravity:Vector="0,0,0", ExternalForce) -> (Pose) {
    bClampLinearTranslationLimitToRefPose=true
    bOverrideWorldGravity=true
    overridePhysicsAsset=&/M_MED_ElbowChat_Physics
    simulationSpace=BaseBoneSpace
    alphaBoolBlend {blendOption=Linear}
    baseBoneRef {boneName=pelvis}
    componentAppliedLinearAccClamp=<3000 2000 3000>
    componentLinearAccScale=<0.4 0.25 0.4>
    componentLinearVelScale=<0.4 0.25 0.4>
    simSpaceSettings {maxAngularAcceleration=500, maxAngularVelocity=3, maxLinearAcceleration=100, maxLinearVelocity=100}
   }
   n3 = NOT Boolean(A=Get RbanOrCloth() -> (RbanOrCloth:bool))
   n4 = Trail controller - Bone: dyn_cloth_front_l_04!(ComponentPose=@n2, bAlphaBoolEnabled:bool=@n3) -> (Pose) {
    alphaInputType=Bool
    bLimitRotation=true
    bUsePlanarLimit=true
    chainLength=4
    planarLimits [{drivingBone {boneName=pelvis}, planeTransform {location=<-30 15 0>, rotation=<rot 0 -15 89.9999>, scale=<2 1 1>}}]
    rotationLimits [
     {limitMax=<10 10 0>, limitMin=<-10 -10 -180>}
     {limitMax=<10 10 0>, limitMin=<-10 -10 -180>}
     {limitMax=<20 20 0>, limitMin=<-20 -20 -180>}
     {limitMax=<20 20 0>, limitMin=<-20 -20 -180>}
    ]
    rotationOffsets [<0 0 0>, <0 0 0>, <0 0 0>, <0 0 0>]
    trailBone {boneName=dyn_cloth_front_l_04}
   }
   n5 = Trail controller - Bone: dyn_cloth_front_r_04!(ComponentPose=@n4, bAlphaBoolEnabled:bool=@n3) -> (Pose) {
    alphaInputType=Bool
    bLimitRotation=true
    bUsePlanarLimit=true
    chainLength=4
    planarLimits [{drivingBone {boneName=pelvis}, planeTransform {location=<-30 15 0>, rotation=<rot 0 -15 89.9999>, scale=<2 1 1>}}]
    rotationLimits [
     {limitMax=<10 10 0>, limitMin=<-10 -10 -180>}
     {limitMax=<10 10 0>, limitMin=<-10 -10 -180>}
     {limitMax=<20 20 0>, limitMin=<-20 -20 -180>}
     {limitMax=<20 20 0>, limitMin=<-20 -20 -180>}
    ]
    rotationOffsets [<0 0 0>, <0 0 0>, <0 0 0>, <0 0 0>]
    trailBone {boneName=dyn_cloth_front_r_04}
   }
   n6 = Component To Local(ComponentPose=@n5)
   n7 = Control Rig!(Source=Use cached pose 'PerPose'(), Alpha) -> (Pose) {
    alphaBoolBlend {blendOption=Linear}
    controlRigAssetReference {blueprintRigClass=&/M_MED_ElbowChat_Cloth_CtrlRig.M_MED_ElbowChat_Cloth_CtrlRig_C}
   }
   n8 = Get bClothEnabled()
   n9 = Anim Dynamics - Bone: dyn_collar_r_02!(
    ComponentPose=Local To Component(LocalPose=Blend Poses by bool(BlendPose_0=@n7, BlendPose_1=@n6, bActiveValue=@n8) {blendType=Linear})
    Alpha=1
    GravityOverride:Vector="0,0,0"
    ExternalForce="0,0,0"
   ) -> (Pose) {
    angularDampingOverride=0.999
    angularSpringConstant=10000
    bAngularSpring=true
    bGravityOverrideInSimSpace=true
    bOverrideAngularDamping=true
    bUseGravityOverride=true
    simulationSpace=RootRelative
    alphaBoolBlend {blendOption=Linear}
    boundBone {boneName=dyn_collar_r_02}
    componentAppliedLinearAccClamp=<3000 3000 3000>
    componentLinearAccScale=<0.4 0.4 0.4>
    componentLinearVelScale=<0.4 0.4 0.4>
    physicsBodyDefinitions [{
      boundBone {boneName=dyn_collar_r_02}
      boxExtents=<10 10 10>
      constraintSetup {angularTargetAxis=AxisY, angularLimitsMax=<0 5 0>, angularLimitsMin=<0 5 0>, angularTarget=<0 0 0>}
      localJointOffset=<1 0 0>
     }]
   }
   n10 = Anim Dynamics - Bone: dyn_collar_r_03!(ComponentPose=@n9, Alpha=1, GravityOverride:Vector="0,0,0", ExternalForce="0,0,0") -> (Pose) {
    angularDampingOverride=0.999
    angularSpringConstant=10000
    bAngularSpring=true
    bGravityOverrideInSimSpace=true
    bOverrideAngularDamping=true
    bUseGravityOverride=true
    simulationSpace=RootRelative
    alphaBoolBlend {blendOption=Linear}
    boundBone {boneName=dyn_collar_r_03}
    componentAppliedLinearAccClamp=<3000 3000 3000>
    componentLinearAccScale=<0.4 0.4 0.4>
    componentLinearVelScale=<0.4 0.4 0.4>
    physicsBodyDefinitions [{
      boundBone {boneName=dyn_collar_r_03}
      boxExtents=<10 10 10>
      constraintSetup {angularTargetAxis=AxisY, angularLimitsMax=<0 3 0>, angularLimitsMin=<0 -3 0>, angularTarget=<0 0 0>}
      localJointOffset=<1 0 0>
     }]
   }
   n11 = Anim Dynamics - Bone: dyn_collar_l_02!(ComponentPose=@n10, Alpha=1, GravityOverride:Vector="0,0,0", ExternalForce="0,0,0") -> (Pose) {
    angularDampingOverride=0.999
    angularSpringConstant=10000
    bAngularSpring=true
    bOverrideAngularDamping=true
    bUseGravityOverride=true
    simulationSpace=RootRelative
    alphaBoolBlend {blendOption=Linear}
    boundBone {boneName=dyn_collar_l_02}
    componentAppliedLinearAccClamp=<3000 3000 3000>
    componentLinearAccScale=<0.4 0.4 0.4>
    componentLinearVelScale=<0.4 0.4 0.4>
    physicsBodyDefinitions [{
      boundBone {boneName=dyn_collar_l_02}
      boxExtents=<10 10 10>
      constraintSetup {angularTargetAxis=AxisY, angularLimitsMax=<0 3 0>, angularLimitsMin=<0 -3 0>, angularTarget=<0 0 0>}
      localJointOffset=<-2 0 0>
     }]
   }
   n12 = Anim Dynamics - Bone: dyn_collar_l_03!(ComponentPose=@n11, Alpha=1, GravityOverride:Vector="0,0,0", ExternalForce="0,0,0") -> (Pose) {
    angularDampingOverride=0.999
    angularSpringConstant=10000
    bAngularSpring=true
    bOverrideAngularDamping=true
    bUseGravityOverride=true
    simulationSpace=RootRelative
    alphaBoolBlend {blendOption=Linear}
    boundBone {boneName=dyn_collar_l_03}
    componentAppliedLinearAccClamp=<3000 3000 3000>
    componentLinearAccScale=<0.4 0.4 0.4>
    componentLinearVelScale=<0.4 0.4 0.4>
    physicsBodyDefinitions [{
      boundBone {boneName=dyn_collar_l_03}
      boxExtents=<10 10 10>
      constraintSetup {angularTargetAxis=AxisY, angularLimitsMax=<0 3 0>, angularLimitsMin=<0 -3 0>, angularTarget=<0 0 0>}
      localJointOffset=<-2 0 0>
     }]
   }
   n13 = "Transform (Modify) Bone - Bone: cape_upper_master"!(ComponentPose=@n12, bAlphaBoolEnabled:bool=True, Translation="0,0,0", Rotation="0,11,0", Scale="1,1,1") -> (Pose) {
    alphaInputType=Bool
    rotationMode=BMM_Additive
    rotationSpace=BCS_BoneSpace
    alphaBoolBlend {blendOption=Linear}
    boneToModify {boneName=cape_upper_master}
    rotation {yaw=11}
   }
   n14 = Anim Dynamics - Bone: cape_upper_01!(ComponentPose=@n13, Alpha=1, GravityOverride:Vector="0,0,0", ExternalForce="0,0,0") -> (Pose) {
    angularDampingOverride=0.999
    angularSpringConstant=10000
    bAngularSpring=true
    bOverrideAngularDamping=true
    bUseGravityOverride=true
    simulationSpace=RootRelative
    alphaBoolBlend {blendOption=Linear}
    boundBone {boneName=cape_upper_01}
    componentAppliedLinearAccClamp=<3000 3000 3000>
    componentLinearAccScale=<0.4 0.4 0.4>
    componentLinearVelScale=<0.4 0.4 0.4>
    physicsBodyDefinitions [{
      boundBone {boneName=cape_upper_01}
      boxExtents=<10 10 10>
      constraintSetup {angularTargetAxis=AxisY, angularLimitsMax=<0 5 15>, angularLimitsMin=<0 -5 0>, angularTarget=<0 0 0>}
      localJointOffset=<2 0 0>
     }]
   }
   n15 = Anim Dynamics - Bone: cape_upper_02!(ComponentPose=@n14, Alpha=1, GravityOverride:Vector="0,0,0", ExternalForce="0,0,0") -> (Pose) {
    angularDampingOverride=0.999
    angularSpringConstant=10000
    bAngularSpring=true
    bOverrideAngularDamping=true
    bUseGravityOverride=true
    simulationSpace=RootRelative
    alphaBoolBlend {blendOption=Linear}
    boundBone {boneName=cape_upper_02}
    componentAppliedLinearAccClamp=<3000 3000 3000>
    componentLinearAccScale=<0.4 0.4 0.4>
    componentLinearVelScale=<0.4 0.4 0.4>
    physicsBodyDefinitions [{
      boundBone {boneName=cape_upper_02}
      boxExtents=<10 10 10>
      constraintSetup {angularTargetAxis=AxisY, angularLimitsMax=<0 0 25>, angularLimitsMin=<0 0 0>, angularTarget=<0 0 0>}
      localJointOffset=<1 0 0>
     }]
   }
   n16 = "Constraint - dyn_chestSkull_m_master (PPPPPPPP)"!(
    ComponentPose=@n15
    Alpha=1
    ConstraintWeights_0=0.1092
    ConstraintWeights_1:float=0.2468
    ConstraintWeights_2:float=0.019
    ConstraintWeights_3:float=0.0262
    ConstraintWeights_4:float=0.0227
    ConstraintWeights_5:float=0
    ConstraintWeights_6:float=0.0467
    ConstraintWeights_7:float=0.5294
   ) -> (Pose) {
    alphaBoolBlend {blendOption=Linear}
    boneToModify {boneName=dyn_chestSkull_m_master}
    constraintSetup [
     {targetBone {boneName=spine_03}}
     {targetBone {boneName=spine_05}}
     {targetBone {boneName=clavicle_l}}
     {targetBone {boneName=clavicle_r}}
     {targetBone {boneName=deform_upperarm_up_r}}
     {targetBone {boneName=dyn_shoulder_r_02}}
     {targetBone {boneName=dyn_chestSkull_m_01}}
     {targetBone {boneName=spine_04}}
    ]
    constraintWeights [1,1,1,1,1,1,1,1]
   }
   Output Pose(Result=Component To Local(ComponentPose=@n16)) {layerGroup=None}
   Save cached pose 'PerPose'(Pose=Copy Pose From Mesh() {bUseAttachedParent=true}) {cacheName=PerPose}
  ]}
 EventGraph {nodes [
   n1 = Event BlueprintUpdateAnimation()
   > Set Delta Time local(Delta Time local:float=@n1.DeltaTimeX) -> (Output_Get:float)
   > n2 = Sequence!(execute) -> (then_0, then_1, then_2:Exec, then_3:Exec, then_4:Exec)
   > Branch(Condition=Get bClothEnabled())
   > Sequence!(execute) -> (then_0, then_1)
   > n3 = Cloth Controller Function DEPRECATED!(
    execute
    CharacterPartAnimInstance:Custom Character Part Anim Instance Object Reference=Self-Reference()
    ClothingAssetName:String=ElbowChat_Cloth
    Cloth Paramaters:Cloth Params Struct Structure=Get Back Cloth Paramaters() -> (Back Cloth Paramaters:Cloth Params Struct Structure)
    Cloth Core Settings:Cloth Core Setting Struct Structure=Get Back Cloth Core Settings() -> (Back Cloth Core Settings:Cloth Core Setting Struct Structure)
    ClothUpAxis:Vector="0,0,1"
    OnGround_Moving_SpeedLimit:double=0
    OnGround_Moving_AngularVel_Max:double=0.05
    OnGround_Moving_FictitiousVel_Max:double=0.1
    LinVelScale_FwdBwd_Clamp:double=0
    LinVelScale_UpDwn_Clamp:double=0
    DeltaTime:double=Get Delta Time local() -> (Delta Time local:float)
    GameTimeInSeconds:double=GetGameTimeInSeconds()
    WindGustsParams:Wind Gust Params Struct Structure=Get Wind Gusts Back Cloth Params() -> (Wind Gusts Back Cloth Params:Wind Gust Params Struct Structure)
    WindGustLocalSettings:Wind Gust Local Setting Struct Structure=Get Wind Gust Back Cloth local() -> (Wind Gust Back Cloth local:Wind Gust Local Setting Struct Structure)
   ) -> (
    then
    CurrentStateOut:String
    Cloth Settings Out:Cloth Setting Struct Structure
    Cloth Core Settings Out:Cloth Core Setting Struct Structure
    WindGustSettings Out:Wind Gust Setting Struct Structure
    WindGustLocalSettings Out:Wind Gust Local Setting Struct Structure
    GustStrength - WindVelocityOut:Vector
    LinearVelVector-DELETE LATER:Vector
   )
   n4 = Branch(execute=@n2.then_1, Condition=Get Is moving not skydive() -> (Is moving not skydive:bool))
   > n5 = Sequence()
   > Branch(Condition="float > float"(A=Get Alpha is moving forward(), B=0.6))
   n6 = Get bIsCrouching()
   > n7 = Branch(Condition=@n6)
   > Set Wind Gust Params(Wind Gust Params:Wind Gust Params Struct Structure=Get Wind Gust Crouch Forward() -> (Wind Gust Crouch Forward:Wind Gust Params Struct Structure)) -> (Output_Get:Wind Gust Params Struct Structure)
   Set Wind Gust Params(execute=@n7.else, Wind Gust Params=Get Wind Gust Forward() -> (Wind Gust Forward:Wind Gust Params Struct Structure))
   Branch(execute=@n5.then_1, Condition="float > float"(A=Get Alpha moving backward() -> (Alpha moving backward:double), B=0.6))
   > n8 = Branch(Condition=@n6)
   > Set Wind Gust Params(Wind Gust Params=Get Wind Gust Crouch Backward() -> (Wind Gust Crouch Backward:Wind Gust Params Struct Structure))
   Set Wind Gust Params(execute=@n8.else, Wind Gust Params=Get Wind Gust Backward() -> (Wind Gust Backward:Wind Gust Params Struct Structure))
   Set Wind Gust Params(execute=@n4.else, Wind Gust Params=Get Wind Gust Base() -> (Wind Gust Base:Wind Gust Params Struct Structure))
   n9 = Get bIsParachuteOpened()
   Set IS skydive no parachute(execute=@n2.then_2, IS skydive no parachute:bool=AND Boolean(A=Get bIsSkyDiving(), B=NOT Boolean(A=@n9))) -> (Output_Get:bool)
   n10 = Get bIsSkyDiving()
   > Set Is moving not skydive(Is moving not skydive:bool=AND Boolean(A=Get bIsMoving2D(), B=NOT Boolean(A=@n10))) -> (Output_Get:bool)
   > Set Is targeting(Is targeting:double="To Float (Boolean)"(InBool=Get bIsTargeting())) -> (Output_Get:double)
   > n11 = Gravity Override State Function DEPRECATED(
    CharacterPartAnimInstance=Self-Reference()
    GameTimeInSeconds=GetGameTimeInSeconds()
    GravityOverrideParams=Get Gravity Override Params()
    RBAN Params=Get RBAN Params()
   )
   > Set Gravity Override(Gravity Override=@n11)
   n12 = Get bIsDBNOCarried()
   n13 = Get bIsDBNOCarrying() -> (bIsDBNOCarrying:bool)
   > Set Is dbno no carry(Is dbno no carry:bool=AND Boolean!(A=Get bIsDBNO(), B=NOT Boolean(A=@n12), C:bool=NOT Boolean(A=@n13)) -> (ReturnValue)) -> (Output_Get:bool)
   n14 = Get bIsSkyDiving()
   > n15 = Branch(Condition=AND Boolean!(A=Get bIsMoving2D(), B=NOT Boolean(A=@n14)) -> (ReturnValue))
   > n16 = PlayerMovement(AnimInstance=Self-Reference(), XDotProductCoefficient=1, YDotProductCoefficient=1, ZDotProductCoefficient=1)
   > Set Alpha is moving forward(Alpha is moving forward:double=MapRangeClamped(Value=@n16.YMovement, InRangeA=0, InRangeB=1)) -> (Output_Get:double)
   > Set Alpha moving backward(Alpha moving backward:double=MapRangeClamped(Value=@n16.YMovement, InRangeA=0, InRangeB=-1)) -> (Output_Get:double)
   Set Alpha is moving forward(execute=@n15.else, Alpha is moving forward=0)
   n17 = Branch(execute=@n2.then_3, Condition=Get bIsInFrontEnd())
   > Set BwdMovement(BwdMovement:double=0.5) -> (Output_Get:double)
   n18 = PlayerMovement(execute=@n17.else, AnimInstance=Self-Reference(), XDotProductCoefficient=1, YDotProductCoefficient=1, ZDotProductCoefficient=1)
   > n19 = Branch(Condition=Get bIsDBNOCarried())
   > Set BwdMovement(BwdMovement=0.5)
   n20 = Branch(execute=@n19.else, Condition=AND Boolean(A=Get bIsTargeting(), B=Get bIsCrouching()))
   n21 = MapRangeClamped(
    Value=float + float(A="Absolute (Float)"(A=@n18), B=MapRangeClamped(Value=@n18.YMovement, InRangeA=-1, InRangeB=1, OutRangeA=0.3)) {is=title K2Node_CommutativeAssociativeBinaryOperator}
    InRangeA=0
    InRangeB=1.65
   )
   > Set BwdMovement(BwdMovement=float * float(A=@n21, B=0.25) {is=title K2Node_PromotableOperator})
   Set BwdMovement(execute=@n20.else, BwdMovement=@n21)
   Set RbanOrCloth(execute=@n2.then_4, RbanOrCloth:bool=OR Boolean(A=Get bRigidBodyEnabled(), B=Get bClothEnabled())) -> (Output_Get:bool)
   n22 = Event BlueprintInitializeAnimation()
   Set Is a Cape(
    execute=[@n22.then, @n23.then]
    Is a Cape:bool=BackpackItemDefinitionHasAnyMetaTag!(Tags:Array of Gameplay Tag Structures=Make Array("[0]"=(TagName="Cosmetics.MetaTags.ElbowChat")), bRequiresEquipped:bool=false) -> (ReturnValue:bool)
   ) -> (Output_Get:bool)
   n23 = Event OnCharacterCustomizationLoaded()
   Event OnBeginSkydiving()
   > Set SkydivingAlpha(SkydivingAlpha:Rotator="65,0,0") -> (Output_Get:Rotator)
   Event OnEndSkydiving()
   > Set SkydivingAlpha(SkydivingAlpha="10,0,0")
   n24 = Branch(Condition=Get bIsSkyDiving())
   > n25 = Branch(Condition="float > float"(A=Get PawnVelocityZ(), B=50) {is=title K2Node_CallFunction})
   > Set BwdMovement(BwdMovement=float * float(A=Get BwdMovement() -> (BwdMovement:double), B=0) {is=title K2Node_CommutativeAssociativeBinaryOperator})
   Set BwdMovement(execute=[@n24.else, @n25.else], BwdMovement=float * float(A=Get BwdMovement(), B=1))
   DrawVectorFromJointFunction!(
    execute
    SkelMeshComponent:Scene Component Object Reference=GetPartSkeletalMeshComponent()
    JointName:Name=pelvis
    LineEnd:Vector=Break Cloth Setting Struct!(ClothSettingStruct:"Cloth Setting Struct Structure (by ref)"=@n3.Cloth Settings Out) -> (
     LinearVelocityScale_7_#1:Vector=~
     AngularVelocityScale_6_#2:float=~
     FictitiousAngularScale_10_#3:float=~
     GravityOverride_15_#4:Vector=~
     DragCoefficient_23_#5:Vector 2D Structure
     LiftCoefficient_24_#6:Vector 2D Structure
     JointName_27_#7:Name
    ) {is=title K2Node_BreakStruct}.GravityOverride_15_#4
    JointOffset:Vector="0, 0, 0"
    LineColor:Linear Color Structure=(R=1,G=0,B=0,A=1)
    Thickness:double=2
    Foreground:EDrawDebugSceneDepthPriorityGroup Enum=Foreground
   ) -> (then)
  ]}
]
variables [
 Alpha is moving forward {type=number}
 Alpha moving backward {type=number}
 Back Cloth Core Settings {
  type=ClothCoreSettingStruct
  default {
   collisionThickness=0.2
   dampingCoefficient=0.1
   frictionCoefficient=0.2
   localDampingCoefficient=0.1
   numberIterations=2
   numberSubsteps=1
   selfCollisionThickness=0.1
   animDriveDamping=<0 1>
   animDriveStiffness=<0 1>
   areaStiffness=<1 1>
   bendingStiffness=<1 1>
   edgeStiffness=<1 1>
   tetherStiffness=<1.05 1.05>
  }
 }
 Back Cloth Paramaters {
  type=ClothParamsStruct
  default {
   dBNO {
    angularVelocityScale=0.6
    fictitiousAngularScale=0.1
    jointName=pelvis
    dragCoefficient=<0.2 1>
    gravityOverride=<-300 800 0>
    liftCoefficient=<0 0.5>
    linearVelocityScale=<0.3 0.3 0.1>
   }
   dBNO_Carried {
    angularVelocityScale=0.6
    fictitiousAngularScale=0.1
    jointName=pelvis
    dragCoefficient=<0.2 1>
    gravityOverride=<-980 -200 0>
    liftCoefficient=<0 0.5>
    linearVelocityScale=<0.3 0.3 0.1>
   }
   dBNO_Impulse {
    angularVelocityScale=0
    fictitiousAngularScale=0
    jointName=pelvis
    dragCoefficient=<0 0.2>
    gravityOverride=<-3000 -2000 0>
    liftCoefficient=<0 0.1>
    linearVelocityScale=<0.3 0.1 0.3>
   }
   dBNO_RollBwd {
    angularVelocityScale=0
    fictitiousAngularScale=0
    jointName=pelvis
    dragCoefficient=<0 0.2>
    gravityOverride=<-2000 0 0>
    liftCoefficient=<0 0.1>
    linearVelocityScale=<0.3 0.1 0.3>
   }
   dBNO_RollFwd {
    angularVelocityScale=0
    fictitiousAngularScale=0
    jointName=pelvis
    dragCoefficient=<0 0.2>
    gravityOverride=<-980 0 0>
    liftCoefficient=<0 0.1>
    linearVelocityScale=<0.3 0.1 0.3>
   }
   dBNO_Tumbling {
    angularVelocityScale=0
    fictitiousAngularScale=0
    jointName=pelvis
    dragCoefficient=<0 0.2>
    gravityOverride=<-2500 -300 0>
    liftCoefficient=<0 0.1>
    linearVelocityScale=<0.3 0.1 0.3>
   }
   emoteOrMelee {
    angularVelocityScale=0.2
    fictitiousAngularScale=0.1
    jointName=pelvis
    dragCoefficient=<0 0.3>
    gravityOverride=<-980 0 0>
    liftCoefficient=<0 0.2>
    linearVelocityScale=<0.05 0.05 0.05>
   }
   front_End {
    angularVelocityScale=0.1
    fictitiousAngularScale=0.2
    jointName=root
    dragCoefficient=<0.2 0.2>
    gravityOverride=<0 -200 -980>
    liftCoefficient=<0.1 0.1>
    linearVelocityScale=<0.1 0.1 0.1>
   }
   onGround_Crouch_Moving {
    angularVelocityScale=0.1
    fictitiousAngularScale=0.3
    jointName=root
    dragCoefficient=<2 2>
    gravityOverride=<0 0 -980>
    liftCoefficient=<0.6 0.6>
    linearVelocityScale=<0.6 0.6 0.5>
   }
   onGround_Falling {
    angularVelocityScale=0.6
    fictitiousAngularScale=0
    jointName=root
    dragCoefficient=<0 0.2>
    gravityOverride=<0 -500 -980>
    liftCoefficient=<0 0.1>
    linearVelocityScale=<2 1 0.3>
   }
   onGround_Moving {
    angularVelocityScale=0.6
    fictitiousAngularScale=0.1
    jointName=root
    dragCoefficient=<0 1>
    gravityOverride=<0 0 -980>
    liftCoefficient=<0 0.1>
    linearVelocityScale=<0.35 0.35 0.6>
   }
   onGround_Standing {
    angularVelocityScale=0.05
    fictitiousAngularScale=0.4
    jointName=root
    dragCoefficient=<0 1>
    gravityOverride=<0 0 -980>
    liftCoefficient=<0 0.5>
    linearVelocityScale=<0.3 0.3 0.3>
   }
   rail_Grinding {
    angularVelocityScale=0.2
    fictitiousAngularScale=0
    jointName=root
    dragCoefficient=<0.1 0.2>
    gravityOverride=<50 -500 -400>
    liftCoefficient=<0 0.5>
    linearVelocityScale=<0.1 0.1 0.1>
   }
   riding_Moving {
    angularVelocityScale=0.05
    fictitiousAngularScale=1
    jointName=root
    dragCoefficient=<0.5 0.5>
    gravityOverride=<0 0 -980>
    liftCoefficient=<0.5 0.5>
    linearVelocityScale=<0.4 0.6 0.3>
   }
   riding_Moving_Targeting {
    angularVelocityScale=0.05
    fictitiousAngularScale=0.7
    jointName=root
    dragCoefficient=<0.5 0.5>
    gravityOverride=<0 0 -980>
    liftCoefficient=<0.5 0.5>
    linearVelocityScale=<0.2 0.6 0.2>
   }
   riding_Standing {
    angularVelocityScale=0.05
    fictitiousAngularScale=1
    jointName=root
    dragCoefficient=<0.5 0.5>
    gravityOverride=<0 -450 -980>
    liftCoefficient=<0.5 0.5>
    linearVelocityScale=<0.6 0.6 0.3>
   }
   skydive_DivingDown {
    angularVelocityScale=0.1
    fictitiousAngularScale=0
    jointName=pelvis
    dragCoefficient=<0 1>
    gravityOverride=<-450 -100 0>
    liftCoefficient=<0 0.3>
    linearVelocityScale=<0.1 0.1 0.1>
   }
   skydive_FloatingDown {
    angularVelocityScale=0.1
    fictitiousAngularScale=0
    jointName=root
    dragCoefficient=<0 2.5>
    gravityOverride=<0 0 100>
    liftCoefficient=<0 0.4>
    linearVelocityScale=<0.2 0.2 0.1>
   }
   skydive_GoingUp {
    angularVelocityScale=0.1
    fictitiousAngularScale=0
    jointName=root
    dragCoefficient=<0 0>
    gravityOverride=<0 -3000 3000>
    liftCoefficient=<0 0>
    linearVelocityScale=<0.1 0.1 0.05>
   }
   skydive_Parachuting {
    angularVelocityScale=0.1
    fictitiousAngularScale=0
    jointName=root
    dragCoefficient=<0 2.5>
    gravityOverride=<0 0 -980>
    liftCoefficient=<0 0.5>
    linearVelocityScale=<0.2 0.2 0.2>
   }
   skydive_Surfing {
    angularVelocityScale=0.6
    fictitiousAngularScale=0.1
    jointName=root
    dragCoefficient=<1.5 1.5>
    gravityOverride=<0 -600 -500>
    liftCoefficient=<1 1>
    linearVelocityScale=<0.1 0.1 0.3>
   }
   swimming {
    angularVelocityScale=0.6
    fictitiousAngularScale=0
    jointName=root
    dragCoefficient=<0 2>
    gravityOverride=<0 -600 400>
    liftCoefficient=<0 0.4>
    linearVelocityScale=<0.6 0.2 0.2>
   }
   tactical_Sprint {
    angularVelocityScale=0.6
    fictitiousAngularScale=0.1
    jointName=root
    dragCoefficient=<2 2>
    gravityOverride=<0 0 -980>
    liftCoefficient=<0.5 0.5>
    linearVelocityScale=<0.6 0.6 0.5>
   }
   ziplining {
    angularVelocityScale=0.6
    fictitiousAngularScale=0
    jointName=root
    dragCoefficient=<0 1.5>
    gravityOverride=<0 0 -980>
    liftCoefficient=<0 0.3>
    linearVelocityScale=<0.4 0.4 0.5>
   }
  }
 }
 BwdMovement {type=number}
 Cloth Core Settings {
  type=ClothCoreSettingStruct
  default {
   collisionThickness=0.1073
   dampingCoefficient=0
   frictionCoefficient=0.7
   localDampingCoefficient=0.05
   numberIterations=2
   numberSubsteps=1
   selfCollisionThickness=0
   animDriveDamping=<0.1 0>
   animDriveStiffness=<0.5 1>
   areaStiffness=<1 1>
   bendingStiffness=<100 100>
   edgeStiffness=<100 100>
   tetherStiffness=<1 1>
  }
 }
 Cloth Paramaters {
  type=ClothParamsStruct
  default {
   dBNO {
    angularVelocityScale=0.6
    fictitiousAngularScale=0.1
    jointName=pelvis
    dragCoefficient=<0.2 1>
    gravityOverride=<-980 400 0>
    liftCoefficient=<0 0.5>
    linearVelocityScale=<0.3 0.3 0.1>
   }
   dBNO_Carried {
    angularVelocityScale=0.6
    fictitiousAngularScale=0.1
    jointName=pelvis
    dragCoefficient=<0.2 1>
    gravityOverride=<-980 -200 0>
    liftCoefficient=<0 0.5>
    linearVelocityScale=<0.3 0.3 0.1>
   }
   dBNO_Impulse {
    angularVelocityScale=0
    fictitiousAngularScale=0
    jointName=pelvis
    dragCoefficient=<0 0.2>
    gravityOverride=<-3000 -2000 0>
    liftCoefficient=<0 0.1>
    linearVelocityScale=<0.3 0.1 0.3>
   }
   dBNO_RollBwd {
    angularVelocityScale=0
    fictitiousAngularScale=0
    jointName=pelvis
    dragCoefficient=<0 0.2>
    gravityOverride=<-2000 0 0>
    liftCoefficient=<0 0.1>
    linearVelocityScale=<0.3 0.1 0.3>
   }
   dBNO_RollFwd {
    angularVelocityScale=0
    fictitiousAngularScale=0
    jointName=pelvis
    dragCoefficient=<0 0.2>
    gravityOverride=<-980 0 0>
    liftCoefficient=<0 0.1>
    linearVelocityScale=<0.3 0.1 0.3>
   }
   dBNO_Tumbling {
    angularVelocityScale=0
    fictitiousAngularScale=0
    jointName=pelvis
    dragCoefficient=<0 0.2>
    gravityOverride=<-2500 -300 0>
    liftCoefficient=<0 0.1>
    linearVelocityScale=<0.3 0.1 0.3>
   }
   emoteOrMelee {
    angularVelocityScale=0.6
    fictitiousAngularScale=0.1
    jointName=root
    dragCoefficient=<2.5 2.5>
    gravityOverride=<0 0 -980>
    liftCoefficient=<0.5 0.5>
    linearVelocityScale=<0.6 0.6 0.5>
   }
   front_End {
    angularVelocityScale=0.05
    fictitiousAngularScale=0.6
    jointName=root
    dragCoefficient=<0.2 0.2>
    gravityOverride=<0 -200 -980>
    liftCoefficient=<0.1 0.1>
    linearVelocityScale=<0.1 0.1 0.1>
   }
   onGround_Crouch_Moving {
    angularVelocityScale=0.1
    fictitiousAngularScale=0.3
    jointName=root
    dragCoefficient=<2 2>
    gravityOverride=<0 0 -980>
    liftCoefficient=<0.5 0.5>
    linearVelocityScale=<0.6 0.6 0.5>
   }
   onGround_Falling {
    angularVelocityScale=0.6
    fictitiousAngularScale=0
    jointName=root
    dragCoefficient=<0 0.2>
    gravityOverride=<0 -500 -980>
    liftCoefficient=<0 0.1>
    linearVelocityScale=<2 1 0.3>
   }
   onGround_Moving {
    angularVelocityScale=0.2
    fictitiousAngularScale=0.1
    jointName=root
    dragCoefficient=<0.5017 0.0317>
    gravityOverride=<0 0 -500>
    liftCoefficient=<0.3175 0.3787>
    linearVelocityScale=<0.2 0.3 0.3>
   }
   onGround_Standing {
    angularVelocityScale=0.05
    fictitiousAngularScale=1.2
    jointName=root
    dragCoefficient=<2.5 2.5>
    gravityOverride=<0 0 -980>
    liftCoefficient=<0.5 0.5>
    linearVelocityScale=<0.6 0.6 0.5>
   }
   rail_Grinding {
    angularVelocityScale=0.2
    fictitiousAngularScale=0
    jointName=root
    dragCoefficient=<0.1 0.2>
    gravityOverride=<50 -500 -400>
    liftCoefficient=<0 0.5>
    linearVelocityScale=<0.1 0.1 0.1>
   }
   riding_Moving {
    angularVelocityScale=0.05
    fictitiousAngularScale=1
    jointName=root
    dragCoefficient=<0.5 0.5>
    gravityOverride=<0 0 -980>
    liftCoefficient=<0.5 0.5>
    linearVelocityScale=<0.4 0.6 0.3>
   }
   riding_Moving_Targeting {
    angularVelocityScale=0.05
    fictitiousAngularScale=0.7
    jointName=root
    dragCoefficient=<0.5 0.5>
    gravityOverride=<0 0 -980>
    liftCoefficient=<0.5 0.5>
    linearVelocityScale=<0.2 0.6 0.2>
   }
   riding_Standing {
    angularVelocityScale=0.05
    fictitiousAngularScale=1
    jointName=root
    dragCoefficient=<0.5 0.5>
    gravityOverride=<0 -450 -980>
    liftCoefficient=<0.5 0.5>
    linearVelocityScale=<0.6 0.6 0.3>
   }
   skydive_DivingDown {
    angularVelocityScale=0.1
    fictitiousAngularScale=0
    jointName=pelvis
    dragCoefficient=<0.019 0.1126>
    gravityOverride=<-480 -100 0>
    liftCoefficient=<0.2298 0.0989>
    linearVelocityScale=<0.1 0.1 0.1>
   }
   skydive_FloatingDown {
    angularVelocityScale=0.3
    fictitiousAngularScale=0
    jointName=root
    dragCoefficient=<1.3104 2>
    gravityOverride=<0 -800 400>
    liftCoefficient=<0.0798 2>
    linearVelocityScale=<0.1 0.1 0.1>
   }
   skydive_GoingUp {
    angularVelocityScale=0.1
    fictitiousAngularScale=0
    jointName=root
    dragCoefficient=<0.01 0.01>
    gravityOverride=<0 -2000 -980>
    liftCoefficient=<0.05 0.05>
    linearVelocityScale=<0.1 0.1 0.01>
   }
   skydive_Parachuting {
    angularVelocityScale=0.096
    fictitiousAngularScale=0.1
    jointName=root
    dragCoefficient=<0.95 2>
    gravityOverride=<0 0 -980>
    liftCoefficient=<0.1778 2>
    linearVelocityScale=<0.2 0.2 0.05>
   }
   skydive_Surfing {
    angularVelocityScale=0.6
    fictitiousAngularScale=0.1
    jointName=root
    dragCoefficient=<1.5 1.5>
    gravityOverride=<0 -600 -500>
    liftCoefficient=<1 1>
    linearVelocityScale=<0.1 0.1 0.3>
   }
   swimming {
    angularVelocityScale=0.6
    fictitiousAngularScale=0
    jointName=root
    dragCoefficient=<2.5 2.5>
    gravityOverride=<0 0 0>
    liftCoefficient=<0.1 0.1>
    linearVelocityScale=<0.6 0.6 0.5>
   }
   tactical_Sprint {
    angularVelocityScale=0.6
    fictitiousAngularScale=0.1
    jointName=root
    dragCoefficient=<2 2>
    gravityOverride=<0 300 -980>
    liftCoefficient=<0.5 0.5>
    linearVelocityScale=<0.6 0.6 0.5>
   }
   ziplining {
    angularVelocityScale=0.6
    fictitiousAngularScale=0
    jointName=root
    dragCoefficient=<0 1.5>
    gravityOverride=<0 0 -980>
    liftCoefficient=<0 0.3>
    linearVelocityScale=<0.4 0.4 0.5>
   }
  }
 }
 Current State Out {type=string, default=""}
 Delta Time local {type=number}
 Gravity Override {type=Vector}
 Gravity Override Params {
  type=GravityOverrideParamsStruct
  default {
   dBNO {gravityOverride=<-980 1500 0>}
   skydive_DivingDown {windFrequency=<5 12 0>}
   skydive_GoingUp {gravityOverride=<0 0 -980>, windAmplitude=<0 0 0>}
  }
 }
 Gust Local Settings {type=WindGustLocalSettingStruct, default {elapsedTime=0, finalGustStrengthLocal=0, gustDuration=0, gustStrength=0, timeBetweenGusts=0}}
 Is a Cape {type=boolean}
 Is dbno no carry {type=boolean}
 Is moving {type=boolean}
 Is moving not skydive {type=boolean}
 IS skydive no parachute {type=boolean}
 Is targeting {type=number}
 RBAN Params {
  type=RBANParamsStruct
  default {
   dBNO {simSpaceSettings {
     maxAngularAcceleration=1000
     maxAngularVelocity=3
     maxLinearAcceleration=300
     maxLinearVelocity=700
     velocityScaleZ=0
     externalLinearDragV=<0 0.5 0>
    }}
   dBNO_Carried {simSpaceSettings {
     maxAngularAcceleration=1000
     maxAngularVelocity=3
     maxLinearAcceleration=300
     maxLinearVelocity=700
     velocityScaleZ=0
     externalLinearDragV=<0 0.5 0>
    }}
   emoteOrMelee {simSpaceSettings {
     maxAngularAcceleration=1000
     maxAngularVelocity=3
     maxLinearAcceleration=300
     maxLinearVelocity=700
     velocityScaleZ=0
     externalLinearDragV=<0 0.5 0>
    }}
   front_End {simSpaceSettings {
     maxAngularAcceleration=1000
     maxAngularVelocity=3
     maxLinearAcceleration=300
     maxLinearVelocity=700
     velocityScaleZ=0
     externalLinearDragV=<0 0.5 0>
    }}
   onGround_Crouch_Moving {simSpaceSettings {
     maxAngularAcceleration=1000
     maxAngularVelocity=3
     maxLinearAcceleration=300
     maxLinearVelocity=700
     velocityScaleZ=0
     externalLinearDragV=<0 0.5 0>
    }}
   onGround_Falling {simSpaceSettings {
     maxAngularAcceleration=1000
     maxAngularVelocity=3
     maxLinearAcceleration=300
     maxLinearVelocity=700
     velocityScaleZ=0
     externalLinearDragV=<0 0.5 0>
    }}
   onGround_Moving {simSpaceSettings {
     maxAngularAcceleration=1000
     maxAngularVelocity=3
     maxLinearAcceleration=300
     maxLinearVelocity=700
     velocityScaleZ=0
     externalLinearDragV=<0 0.5 0>
    }}
   onGround_Standing {simSpaceSettings {
     maxAngularAcceleration=1000
     maxAngularVelocity=3
     maxLinearAcceleration=300
     maxLinearVelocity=700
     velocityScaleZ=0
     externalLinearDragV=<0 0.5 0>
    }}
   rail_Grinding {simSpaceSettings {
     maxAngularAcceleration=1000
     maxAngularVelocity=3
     maxLinearAcceleration=300
     maxLinearVelocity=700
     velocityScaleZ=0
     worldAlpha=0.1
     externalLinearDragV=<0 0.5 0>
    }}
   riding_Moving {simSpaceSettings {
     maxAngularAcceleration=1000
     maxAngularVelocity=3
     maxLinearAcceleration=300
     maxLinearVelocity=700
     velocityScaleZ=0
     externalLinearDragV=<0 0.5 0>
    }}
   riding_Moving_Targeting {simSpaceSettings {
     maxAngularAcceleration=1000
     maxAngularVelocity=3
     maxLinearAcceleration=300
     maxLinearVelocity=700
     velocityScaleZ=0
     externalLinearDragV=<0 0.5 0>
    }}
   riding_Standing {simSpaceSettings {
     maxAngularAcceleration=1000
     maxAngularVelocity=3
     maxLinearAcceleration=300
     maxLinearVelocity=700
     velocityScaleZ=0
     externalLinearDragV=<0 0.5 0>
    }}
   skydive_DivingDown {simSpaceSettings {
     maxAngularAcceleration=1000
     maxAngularVelocity=3
     maxLinearAcceleration=300
     maxLinearVelocity=700
     velocityScaleZ=0
     externalLinearDragV=<0 0.5 0>
    }}
   skydive_FloatingDown {
    componentLinearAccScale=<0 0.5 0>
    componentLinearVelScale=<0 0.5 0>
    simSpaceSettings {
     maxAngularAcceleration=1000
     maxAngularVelocity=3
     maxLinearAcceleration=300
     maxLinearVelocity=700
     velocityScaleZ=0
     externalLinearDragV=<0 0.5 0>
    }
   }
   skydive_GoingUp {simSpaceSettings {
     maxAngularAcceleration=1000
     maxAngularVelocity=3
     maxLinearAcceleration=300
     maxLinearVelocity=700
     velocityScaleZ=0
     externalLinearDragV=<0 0.5 0>
    }}
   skydive_Parachuting {
    componentLinearAccScale=<0 0.1 0.3>
    componentLinearVelScale=<0 0.1 0.3>
    simSpaceSettings {
     maxAngularAcceleration=1000
     maxAngularVelocity=3
     maxLinearAcceleration=300
     maxLinearVelocity=700
     velocityScaleZ=0
     externalLinearDragV=<0 0.5 0>
    }
   }
   skydive_Surfing {simSpaceSettings {
     maxAngularAcceleration=1000
     maxAngularVelocity=3
     maxLinearAcceleration=300
     maxLinearVelocity=700
     velocityScaleZ=0
     externalLinearDragV=<0 0.5 0>
    }}
   swimming {simSpaceSettings {
     maxAngularAcceleration=1000
     maxAngularVelocity=3
     maxLinearAcceleration=300
     maxLinearVelocity=700
     velocityScaleZ=0
     externalLinearDragV=<0 0.5 0>
    }}
   tactical_Sprint {simSpaceSettings {
     maxAngularAcceleration=1000
     maxAngularVelocity=3
     maxLinearAcceleration=300
     maxLinearVelocity=700
     velocityScaleZ=0
     externalLinearDragV=<0 0.5 0>
    }}
   ziplining {simSpaceSettings {
     maxAngularAcceleration=1000
     maxAngularVelocity=3
     maxLinearAcceleration=300
     maxLinearVelocity=700
     velocityScaleZ=0
     externalLinearDragV=<0 0.5 0>
    }}
  }
 }
 RbanOrCloth {type=boolean}
 SkydivingAlpha {type=Rotator}
 Wind Gust Back Cloth local {type=WindGustLocalSettingStruct, default {elapsedTime=0, finalGustStrengthLocal=0, gustDuration=0, gustStrength=0, timeBetweenGusts=0}}
 Wind Gust Backward {
  type=WindGustParamsStruct
  default {
   dBNO {
    maxDownInterpSpeed=5
    maxGustDuration=2
    maxGustStrength=7
    maxTimeBetweenGusts=7
    maxUpInterpSpeed=5
    minDownInterpSpeed=2
    minGustDuration=0.5
    minGustStrength=3
    minTimeBetweenGusts=0
    minUpInterpSpeed=2
    windTurbulenceScalarA=0.1
    windTurbulenceScalarB=1
    windAmplitude=<0 0 0>
    windFrequency=<0 0 0>
    windOffset=<0 0 0>
   }
   dBNO_Carried {
    maxDownInterpSpeed=5
    maxGustDuration=2
    maxGustStrength=7
    maxTimeBetweenGusts=7
    maxUpInterpSpeed=5
    minDownInterpSpeed=2
    minGustDuration=0.5
    minGustStrength=3
    minTimeBetweenGusts=0
    minUpInterpSpeed=2
    windTurbulenceScalarA=0.1
    windTurbulenceScalarB=1
    windAmplitude=<0 0 0>
    windFrequency=<0 0 0>
    windOffset=<0 0 0>
   }
   dBNO_Impulse {
    maxDownInterpSpeed=5
    maxGustDuration=2
    maxGustStrength=7
    maxTimeBetweenGusts=7
    maxUpInterpSpeed=5
    minDownInterpSpeed=2
    minGustDuration=0.5
    minGustStrength=3
    minTimeBetweenGusts=0
    minUpInterpSpeed=2
    windTurbulenceScalarA=0.1
    windTurbulenceScalarB=1
    windAmplitude=<0 0 0>
    windFrequency=<0 0 0>
    windOffset=<0 0 0>
   }
   dBNO_RollBwd {
    maxDownInterpSpeed=5
    maxGustDuration=2
    maxGustStrength=7
    maxTimeBetweenGusts=7
    maxUpInterpSpeed=5
    minDownInterpSpeed=2
    minGustDuration=0.5
    minGustStrength=3
    minTimeBetweenGusts=0
    minUpInterpSpeed=2
    windTurbulenceScalarA=0.1
    windTurbulenceScalarB=1
    windAmplitude=<0 0 0>
    windFrequency=<0 0 0>
    windOffset=<0 0 0>
   }
   dBNO_RollFwd {
    maxDownInterpSpeed=5
    maxGustDuration=2
    maxGustStrength=7
    maxTimeBetweenGusts=7
    maxUpInterpSpeed=5
    minDownInterpSpeed=2
    minGustDuration=0.5
    minGustStrength=3
    minTimeBetweenGusts=0
    minUpInterpSpeed=2
    windTurbulenceScalarA=0.1
    windTurbulenceScalarB=1
    windAmplitude=<0 0 0>
    windFrequency=<0 0 0>
    windOffset=<0 0 0>
   }
   dBNO_Tumbling {
    maxDownInterpSpeed=5
    maxGustDuration=2
    maxGustStrength=7
    maxTimeBetweenGusts=7
    maxUpInterpSpeed=5
    minDownInterpSpeed=2
    minGustDuration=0.5
    minGustStrength=3
    minTimeBetweenGusts=0
    minUpInterpSpeed=2
    windTurbulenceScalarA=0.1
    windTurbulenceScalarB=1
    windAmplitude=<0 0 0>
    windFrequency=<0 0 0>
    windOffset=<0 0 0>
   }
   emoteOrMelee {
    maxDownInterpSpeed=5
    maxGustDuration=2
    maxGustStrength=7
    maxTimeBetweenGusts=7
    maxUpInterpSpeed=5
    minDownInterpSpeed=2
    minGustDuration=0.5
    minGustStrength=3
    minTimeBetweenGusts=0
    minUpInterpSpeed=2
    windTurbulenceScalarA=0.1
    windTurbulenceScalarB=1
    windAmplitude=<0 0 0>
    windFrequency=<0 0 0>
    windOffset=<0 0 0>
   }
   front_End {
    maxDownInterpSpeed=5
    maxGustDuration=2
    maxGustStrength=7
    maxTimeBetweenGusts=7
    maxUpInterpSpeed=5
    minDownInterpSpeed=2
    minGustDuration=0.5
    minGustStrength=3
    minTimeBetweenGusts=0
    minUpInterpSpeed=2
    windTurbulenceScalarA=0.1
    windTurbulenceScalarB=1
    windAmplitude=<0 0 0>
    windFrequency=<0 0 0>
    windOffset=<0 0 0>
   }
   onGround_Crouch_Moving {
    maxDownInterpSpeed=5
    maxGustDuration=2
    maxGustStrength=7
    maxTimeBetweenGusts=7
    maxUpInterpSpeed=5
    minDownInterpSpeed=2
    minGustDuration=0.5
    minGustStrength=3
    minTimeBetweenGusts=0
    minUpInterpSpeed=2
    windTurbulenceScalarA=0.1
    windTurbulenceScalarB=1
    windAmplitude=<0 0 0>
    windFrequency=<0 0 0>
    windOffset=<0 0 0>
   }
   onGround_Falling {
    maxDownInterpSpeed=5
    maxGustDuration=2
    maxGustStrength=7
    maxTimeBetweenGusts=7
    maxUpInterpSpeed=5
    minDownInterpSpeed=2
    minGustDuration=0.5
    minGustStrength=3
    minTimeBetweenGusts=0
    minUpInterpSpeed=2
    windTurbulenceScalarA=0.1
    windTurbulenceScalarB=1
    windAmplitude=<0 0 0>
    windFrequency=<0 0 0>
    windOffset=<0 0 0>
   }
   onGround_Moving {
    maxDownInterpSpeed=5
    maxGustDuration=2
    maxGustStrength=7
    maxTimeBetweenGusts=7
    maxUpInterpSpeed=1
    minDownInterpSpeed=2
    minGustDuration=0.5
    minGustStrength=-16.1506
    minTimeBetweenGusts=0
    minUpInterpSpeed=2
    windTurbulenceScalarA=0.1
    windTurbulenceScalarB=1
    windAmplitude=<0 0 0>
    windFrequency=<0 0 0>
    windOffset=<4.5838 97.1955 8.7233>
   }
   onGround_Standing {
    maxDownInterpSpeed=5
    maxGustDuration=2
    maxGustStrength=7
    maxTimeBetweenGusts=7
    maxUpInterpSpeed=5
    minDownInterpSpeed=2
    minGustDuration=0.5
    minGustStrength=3
    minTimeBetweenGusts=0
    minUpInterpSpeed=2
    windTurbulenceScalarA=0.1
    windTurbulenceScalarB=1
    windAmplitude=<0 0 0>
    windFrequency=<0 0 0>
    windOffset=<0 0 0>
   }
   rail_Grinding {
    maxDownInterpSpeed=5
    maxGustDuration=8
    maxGustStrength=7
    maxTimeBetweenGusts=8
    maxUpInterpSpeed=5
    minDownInterpSpeed=2
    minGustDuration=0.2
    minGustStrength=3
    minTimeBetweenGusts=1
    minUpInterpSpeed=2
    windTurbulenceScalarA=0.1
    windTurbulenceScalarB=0.8
    windAmplitude=<20 65 42>
    windFrequency=<5 6 8.6>
    windOffset=<50 -500 -400>
   }
   riding_Moving {
    maxDownInterpSpeed=5
    maxGustDuration=2
    maxGustStrength=7
    maxTimeBetweenGusts=7
    maxUpInterpSpeed=5
    minDownInterpSpeed=2
    minGustDuration=0.5
    minGustStrength=3
    minTimeBetweenGusts=0
    minUpInterpSpeed=2
    windTurbulenceScalarA=0.1
    windTurbulenceScalarB=1
    windAmplitude=<0 0 0>
    windFrequency=<0 0 0>
    windOffset=<0 0 0>
   }
   riding_Moving_Targeting {
    maxDownInterpSpeed=5
    maxGustDuration=2
    maxGustStrength=7
    maxTimeBetweenGusts=7
    maxUpInterpSpeed=5
    minDownInterpSpeed=2
    minGustDuration=0.5
    minGustStrength=3
    minTimeBetweenGusts=0
    minUpInterpSpeed=2
    windTurbulenceScalarA=0.1
    windTurbulenceScalarB=1
    windAmplitude=<0 0 0>
    windFrequency=<0 0 0>
    windOffset=<0 0 0>
   }
   riding_Standing {
    maxDownInterpSpeed=5
    maxGustDuration=2
    maxGustStrength=7
    maxTimeBetweenGusts=7
    maxUpInterpSpeed=5
    minDownInterpSpeed=2
    minGustDuration=0.5
    minGustStrength=3
    minTimeBetweenGusts=0
    minUpInterpSpeed=2
    windTurbulenceScalarA=0.1
    windTurbulenceScalarB=1
    windAmplitude=<0 0 0>
    windFrequency=<0 0 0>
    windOffset=<0 0 0>
   }
   skydive_DivingDown {
    maxDownInterpSpeed=5
    maxGustDuration=2
    maxGustStrength=7
    maxTimeBetweenGusts=7
    maxUpInterpSpeed=5
    minDownInterpSpeed=2
    minGustDuration=0.5
    minGustStrength=3
    minTimeBetweenGusts=0
    minUpInterpSpeed=2
    windTurbulenceScalarA=0.1
    windTurbulenceScalarB=1
    windAmplitude=<0 0 0>
    windFrequency=<0 0 0>
    windOffset=<0 0 0>
   }
   skydive_FloatingDown {
    maxDownInterpSpeed=5
    maxGustDuration=2
    maxGustStrength=7
    maxTimeBetweenGusts=7
    maxUpInterpSpeed=5
    minDownInterpSpeed=2
    minGustDuration=0.5
    minGustStrength=3
    minTimeBetweenGusts=0
    minUpInterpSpeed=2
    windTurbulenceScalarA=0.1
    windTurbulenceScalarB=1
    windAmplitude=<0 0 0>
    windFrequency=<0 0 0>
    windOffset=<0 0 0>
   }
   skydive_GoingUp {
    maxDownInterpSpeed=5
    maxGustDuration=2
    maxGustStrength=7
    maxTimeBetweenGusts=7
    maxUpInterpSpeed=5
    minDownInterpSpeed=2
    minGustDuration=0.5
    minGustStrength=3
    minTimeBetweenGusts=0
    minUpInterpSpeed=2
    windTurbulenceScalarA=0.1
    windTurbulenceScalarB=1
    windAmplitude=<0 0 0>
    windFrequency=<0 0 0>
    windOffset=<0 0 0>
   }
   skydive_Parachuting {
    maxDownInterpSpeed=5
    maxGustDuration=2
    maxGustStrength=7
    maxTimeBetweenGusts=7
    maxUpInterpSpeed=5
    minDownInterpSpeed=2
    minGustDuration=0.5
    minGustStrength=3
    minTimeBetweenGusts=0
    minUpInterpSpeed=2
    windTurbulenceScalarA=0.1
    windTurbulenceScalarB=1
    windAmplitude=<0 0 0>
    windFrequency=<0 0 0>
    windOffset=<0 0 0>
   }
   skydive_Surfing {
    maxDownInterpSpeed=5
    maxGustDuration=2
    maxGustStrength=7
    maxTimeBetweenGusts=7
    maxUpInterpSpeed=5
    minDownInterpSpeed=2
    minGustDuration=0.5
    minGustStrength=3
    minTimeBetweenGusts=0
    minUpInterpSpeed=2
    windTurbulenceScalarA=0.1
    windTurbulenceScalarB=1
    windAmplitude=<0 0 0>
    windFrequency=<0 0 0>
    windOffset=<0 0 0>
   }
   swimming {
    maxDownInterpSpeed=5
    maxGustDuration=2
    maxGustStrength=7
    maxTimeBetweenGusts=7
    maxUpInterpSpeed=5
    minDownInterpSpeed=2
    minGustDuration=0.5
    minGustStrength=3
    minTimeBetweenGusts=0
    minUpInterpSpeed=2
    windTurbulenceScalarA=0.1
    windTurbulenceScalarB=1
    windAmplitude=<0 0 0>
    windFrequency=<0 0 0>
    windOffset=<0 0 0>
   }
   tactical_Sprint {
    maxDownInterpSpeed=5
    maxGustDuration=2
    maxGustStrength=7
    maxTimeBetweenGusts=7
    maxUpInterpSpeed=5
    minDownInterpSpeed=2
    minGustDuration=0.5
    minGustStrength=3
    minTimeBetweenGusts=0
    minUpInterpSpeed=2
    windTurbulenceScalarA=0.1
    windTurbulenceScalarB=1
    windAmplitude=<0 0 0>
    windFrequency=<0 0 0>
    windOffset=<0 0 0>
   }
   ziplining {
    maxDownInterpSpeed=5
    maxGustDuration=2
    maxGustStrength=7
    maxTimeBetweenGusts=7
    maxUpInterpSpeed=5
    minDownInterpSpeed=2
    minGustDuration=0.5
    minGustStrength=3
    minTimeBetweenGusts=0
    minUpInterpSpeed=2
    windTurbulenceScalarA=0.1
    windTurbulenceScalarB=1
    windAmplitude=<0 0 0>
    windFrequency=<0 0 0>
    windOffset=<0 0 0>
   }
  }
 }
 Wind Gust Base {
  type=WindGustParamsStruct
  default {
   dBNO {
    maxDownInterpSpeed=5
    maxGustDuration=2
    maxGustStrength=7
    maxTimeBetweenGusts=7
    maxUpInterpSpeed=5
    minDownInterpSpeed=2
    minGustDuration=0.5
    minGustStrength=3
    minTimeBetweenGusts=0
    minUpInterpSpeed=2
    windTurbulenceScalarA=0.1
    windTurbulenceScalarB=1
    windAmplitude=<0 0 0>
    windFrequency=<0 0 0>
    windOffset=<0 0 0>
   }
   dBNO_Carried {
    maxDownInterpSpeed=5
    maxGustDuration=2
    maxGustStrength=7
    maxTimeBetweenGusts=7
    maxUpInterpSpeed=5
    minDownInterpSpeed=2
    minGustDuration=0.5
    minGustStrength=3
    minTimeBetweenGusts=0
    minUpInterpSpeed=2
    windTurbulenceScalarA=0.1
    windTurbulenceScalarB=1
    windAmplitude=<0 0 0>
    windFrequency=<0 0 0>
    windOffset=<0 0 0>
   }
   dBNO_Impulse {
    maxDownInterpSpeed=5
    maxGustDuration=2
    maxGustStrength=7
    maxTimeBetweenGusts=7
    maxUpInterpSpeed=5
    minDownInterpSpeed=2
    minGustDuration=0.5
    minGustStrength=3
    minTimeBetweenGusts=0
    minUpInterpSpeed=2
    windTurbulenceScalarA=0.1
    windTurbulenceScalarB=1
    windAmplitude=<0 0 0>
    windFrequency=<0 0 0>
    windOffset=<0 0 0>
   }
   dBNO_RollBwd {
    maxDownInterpSpeed=5
    maxGustDuration=2
    maxGustStrength=7
    maxTimeBetweenGusts=7
    maxUpInterpSpeed=5
    minDownInterpSpeed=2
    minGustDuration=0.5
    minGustStrength=3
    minTimeBetweenGusts=0
    minUpInterpSpeed=2
    windTurbulenceScalarA=0.1
    windTurbulenceScalarB=1
    windAmplitude=<0 0 0>
    windFrequency=<0 0 0>
    windOffset=<0 0 0>
   }
   dBNO_RollFwd {
    maxDownInterpSpeed=5
    maxGustDuration=2
    maxGustStrength=7
    maxTimeBetweenGusts=7
    maxUpInterpSpeed=5
    minDownInterpSpeed=2
    minGustDuration=0.5
    minGustStrength=3
    minTimeBetweenGusts=0
    minUpInterpSpeed=2
    windTurbulenceScalarA=0.1
    windTurbulenceScalarB=1
    windAmplitude=<0 0 0>
    windFrequency=<0 0 0>
    windOffset=<0 0 0>
   }
   dBNO_Tumbling {
    maxDownInterpSpeed=5
    maxGustDuration=2
    maxGustStrength=7
    maxTimeBetweenGusts=7
    maxUpInterpSpeed=5
    minDownInterpSpeed=2
    minGustDuration=0.5
    minGustStrength=3
    minTimeBetweenGusts=0
    minUpInterpSpeed=2
    windTurbulenceScalarA=0.1
    windTurbulenceScalarB=1
    windAmplitude=<0 0 0>
    windFrequency=<0 0 0>
    windOffset=<0 0 0>
   }
   emoteOrMelee {
    maxDownInterpSpeed=5
    maxGustDuration=2
    maxGustStrength=7
    maxTimeBetweenGusts=7
    maxUpInterpSpeed=5
    minDownInterpSpeed=2
    minGustDuration=0.5
    minGustStrength=3
    minTimeBetweenGusts=0
    minUpInterpSpeed=2
    windTurbulenceScalarA=0.1
    windTurbulenceScalarB=1
    windAmplitude=<0 0 0>
    windFrequency=<0 0 0>
    windOffset=<0 0 0>
   }
   front_End {
    maxDownInterpSpeed=5
    maxGustDuration=2
    maxGustStrength=7
    maxTimeBetweenGusts=7
    maxUpInterpSpeed=5
    minDownInterpSpeed=2
    minGustDuration=0.5
    minGustStrength=3
    minTimeBetweenGusts=0
    minUpInterpSpeed=2
    windTurbulenceScalarA=0.1
    windTurbulenceScalarB=1
    windAmplitude=<0 0 0>
    windFrequency=<0 0 0>
    windOffset=<0 0 0>
   }
   onGround_Crouch_Moving {
    maxDownInterpSpeed=5
    maxGustDuration=2
    maxGustStrength=7
    maxTimeBetweenGusts=7
    maxUpInterpSpeed=5
    minDownInterpSpeed=2
    minGustDuration=0.5
    minGustStrength=3
    minTimeBetweenGusts=0
    minUpInterpSpeed=2
    windTurbulenceScalarA=0.1
    windTurbulenceScalarB=1
    windAmplitude=<0 0 0>
    windFrequency=<0 0 0>
    windOffset=<0 0 0>
   }
   onGround_Falling {
    maxDownInterpSpeed=5
    maxGustDuration=2
    maxGustStrength=7
    maxTimeBetweenGusts=7
    maxUpInterpSpeed=5
    minDownInterpSpeed=2
    minGustDuration=0.5
    minGustStrength=3
    minTimeBetweenGusts=0
    minUpInterpSpeed=2
    windTurbulenceScalarA=0.1
    windTurbulenceScalarB=1
    windAmplitude=<0 0 0>
    windFrequency=<0 0 0>
    windOffset=<0 0 0>
   }
   onGround_Moving {
    maxDownInterpSpeed=0
    maxGustDuration=0
    maxGustStrength=0
    maxTimeBetweenGusts=0
    maxUpInterpSpeed=0
    minDownInterpSpeed=0
    minGustDuration=0
    minGustStrength=0
    minTimeBetweenGusts=0
    minUpInterpSpeed=0
    windTurbulenceScalarA=0
    windTurbulenceScalarB=0
    windAmplitude=<0 0 0>
    windFrequency=<0 0 0>
    windOffset=<0 0 6.3456>
   }
   onGround_Standing {
    maxDownInterpSpeed=5
    maxGustDuration=2
    maxGustStrength=7
    maxTimeBetweenGusts=7
    maxUpInterpSpeed=5
    minDownInterpSpeed=2
    minGustDuration=0.5
    minGustStrength=3
    minTimeBetweenGusts=0
    minUpInterpSpeed=2
    windTurbulenceScalarA=0.1
    windTurbulenceScalarB=1
    windAmplitude=<0 0 0>
    windFrequency=<0 0 0>
    windOffset=<0 0 0>
   }
   rail_Grinding {
    maxDownInterpSpeed=5
    maxGustDuration=8
    maxGustStrength=7
    maxTimeBetweenGusts=8
    maxUpInterpSpeed=5
    minDownInterpSpeed=2
    minGustDuration=0.2
    minGustStrength=3
    minTimeBetweenGusts=1
    minUpInterpSpeed=2
    windTurbulenceScalarA=0.1
    windTurbulenceScalarB=0.8
    windAmplitude=<20 65 42>
    windFrequency=<5 6 8.6>
    windOffset=<50 -500 -400>
   }
   riding_Moving {
    maxDownInterpSpeed=5
    maxGustDuration=2
    maxGustStrength=7
    maxTimeBetweenGusts=7
    maxUpInterpSpeed=5
    minDownInterpSpeed=2
    minGustDuration=0.5
    minGustStrength=3
    minTimeBetweenGusts=0
    minUpInterpSpeed=2
    windTurbulenceScalarA=0.1
    windTurbulenceScalarB=1
    windAmplitude=<0 0 0>
    windFrequency=<0 0 0>
    windOffset=<0 0 0>
   }
   riding_Moving_Targeting {
    maxDownInterpSpeed=5
    maxGustDuration=2
    maxGustStrength=7
    maxTimeBetweenGusts=7
    maxUpInterpSpeed=5
    minDownInterpSpeed=2
    minGustDuration=0.5
    minGustStrength=3
    minTimeBetweenGusts=0
    minUpInterpSpeed=2
    windTurbulenceScalarA=0.1
    windTurbulenceScalarB=1
    windAmplitude=<0 0 0>
    windFrequency=<0 0 0>
    windOffset=<0 0 0>
   }
   riding_Standing {
    maxDownInterpSpeed=5
    maxGustDuration=2
    maxGustStrength=7
    maxTimeBetweenGusts=7
    maxUpInterpSpeed=5
    minDownInterpSpeed=2
    minGustDuration=0.5
    minGustStrength=3
    minTimeBetweenGusts=0
    minUpInterpSpeed=2
    windTurbulenceScalarA=0.1
    windTurbulenceScalarB=1
    windAmplitude=<0 0 0>
    windFrequency=<0 0 0>
    windOffset=<0 0 0>
   }
   skydive_DivingDown {
    maxDownInterpSpeed=5
    maxGustDuration=2
    maxGustStrength=7
    maxTimeBetweenGusts=7
    maxUpInterpSpeed=5
    minDownInterpSpeed=2
    minGustDuration=0.5
    minGustStrength=3
    minTimeBetweenGusts=0
    minUpInterpSpeed=2
    windTurbulenceScalarA=0.1
    windTurbulenceScalarB=1
    windAmplitude=<0 0 0>
    windFrequency=<0 0 0>
    windOffset=<0 0 0>
   }
   skydive_FloatingDown {
    maxDownInterpSpeed=5
    maxGustDuration=2
    maxGustStrength=7
    maxTimeBetweenGusts=7
    maxUpInterpSpeed=5
    minDownInterpSpeed=2
    minGustDuration=0.5
    minGustStrength=3
    minTimeBetweenGusts=0
    minUpInterpSpeed=2
    windTurbulenceScalarA=0.1
    windTurbulenceScalarB=1
    windAmplitude=<0 0 0>
    windFrequency=<0 0 0>
    windOffset=<0 0 0>
   }
   skydive_GoingUp {
    maxDownInterpSpeed=5
    maxGustDuration=2
    maxGustStrength=7
    maxTimeBetweenGusts=7
    maxUpInterpSpeed=5
    minDownInterpSpeed=2
    minGustDuration=0.5
    minGustStrength=3
    minTimeBetweenGusts=0
    minUpInterpSpeed=2
    windTurbulenceScalarA=0.1
    windTurbulenceScalarB=1
    windAmplitude=<0 0 0>
    windFrequency=<0 0 0>
    windOffset=<0 0 0>
   }
   skydive_Parachuting {
    maxDownInterpSpeed=5
    maxGustDuration=2
    maxGustStrength=7
    maxTimeBetweenGusts=7
    maxUpInterpSpeed=5
    minDownInterpSpeed=2
    minGustDuration=0.5
    minGustStrength=3
    minTimeBetweenGusts=0
    minUpInterpSpeed=2
    windTurbulenceScalarA=0.1
    windTurbulenceScalarB=1
    windAmplitude=<0 0 0>
    windFrequency=<0 0 0>
    windOffset=<0 0 0>
   }
   skydive_Surfing {
    maxDownInterpSpeed=5
    maxGustDuration=2
    maxGustStrength=7
    maxTimeBetweenGusts=7
    maxUpInterpSpeed=5
    minDownInterpSpeed=2
    minGustDuration=0.5
    minGustStrength=3
    minTimeBetweenGusts=0
    minUpInterpSpeed=2
    windTurbulenceScalarA=0.1
    windTurbulenceScalarB=1
    windAmplitude=<0 0 0>
    windFrequency=<0 0 0>
    windOffset=<0 0 0>
   }
   swimming {
    maxDownInterpSpeed=5
    maxGustDuration=2
    maxGustStrength=7
    maxTimeBetweenGusts=7
    maxUpInterpSpeed=5
    minDownInterpSpeed=2
    minGustDuration=0.5
    minGustStrength=3
    minTimeBetweenGusts=0
    minUpInterpSpeed=2
    windTurbulenceScalarA=0.1
    windTurbulenceScalarB=1
    windAmplitude=<0 0 0>
    windFrequency=<0 0 0>
    windOffset=<0 0 0>
   }
   tactical_Sprint {
    maxDownInterpSpeed=5
    maxGustDuration=2
    maxGustStrength=7
    maxTimeBetweenGusts=7
    maxUpInterpSpeed=5
    minDownInterpSpeed=2
    minGustDuration=0.5
    minGustStrength=3
    minTimeBetweenGusts=0
    minUpInterpSpeed=2
    windTurbulenceScalarA=0.1
    windTurbulenceScalarB=1
    windAmplitude=<0 0 0>
    windFrequency=<0 0 0>
    windOffset=<0 0 0>
   }
   ziplining {
    maxDownInterpSpeed=5
    maxGustDuration=2
    maxGustStrength=7
    maxTimeBetweenGusts=7
    maxUpInterpSpeed=5
    minDownInterpSpeed=2
    minGustDuration=0.5
    minGustStrength=3
    minTimeBetweenGusts=0
    minUpInterpSpeed=2
    windTurbulenceScalarA=0.1
    windTurbulenceScalarB=1
    windAmplitude=<0 0 0>
    windFrequency=<0 0 0>
    windOffset=<0 0 0>
   }
  }
 }
 Wind Gust Crouch Backward {
  type=WindGustParamsStruct
  default {
   dBNO {
    maxDownInterpSpeed=5
    maxGustDuration=2
    maxGustStrength=7
    maxTimeBetweenGusts=7
    maxUpInterpSpeed=5
    minDownInterpSpeed=2
    minGustDuration=0.5
    minGustStrength=3
    minTimeBetweenGusts=0
    minUpInterpSpeed=2
    windTurbulenceScalarA=0.1
    windTurbulenceScalarB=1
    windAmplitude=<0 0 0>
    windFrequency=<0 0 0>
    windOffset=<0 0 0>
   }
   dBNO_Carried {
    maxDownInterpSpeed=5
    maxGustDuration=2
    maxGustStrength=7
    maxTimeBetweenGusts=7
    maxUpInterpSpeed=5
    minDownInterpSpeed=2
    minGustDuration=0.5
    minGustStrength=3
    minTimeBetweenGusts=0
    minUpInterpSpeed=2
    windTurbulenceScalarA=0.1
    windTurbulenceScalarB=1
    windAmplitude=<0 0 0>
    windFrequency=<0 0 0>
    windOffset=<0 0 0>
   }
   dBNO_Impulse {
    maxDownInterpSpeed=5
    maxGustDuration=2
    maxGustStrength=7
    maxTimeBetweenGusts=7
    maxUpInterpSpeed=5
    minDownInterpSpeed=2
    minGustDuration=0.5
    minGustStrength=3
    minTimeBetweenGusts=0
    minUpInterpSpeed=2
    windTurbulenceScalarA=0.1
    windTurbulenceScalarB=1
    windAmplitude=<0 0 0>
    windFrequency=<0 0 0>
    windOffset=<0 0 0>
   }
   dBNO_RollBwd {
    maxDownInterpSpeed=5
    maxGustDuration=2
    maxGustStrength=7
    maxTimeBetweenGusts=7
    maxUpInterpSpeed=5
    minDownInterpSpeed=2
    minGustDuration=0.5
    minGustStrength=3
    minTimeBetweenGusts=0
    minUpInterpSpeed=2
    windTurbulenceScalarA=0.1
    windTurbulenceScalarB=1
    windAmplitude=<0 0 0>
    windFrequency=<0 0 0>
    windOffset=<0 0 0>
   }
   dBNO_RollFwd {
    maxDownInterpSpeed=5
    maxGustDuration=2
    maxGustStrength=7
    maxTimeBetweenGusts=7
    maxUpInterpSpeed=5
    minDownInterpSpeed=2
    minGustDuration=0.5
    minGustStrength=3
    minTimeBetweenGusts=0
    minUpInterpSpeed=2
    windTurbulenceScalarA=0.1
    windTurbulenceScalarB=1
    windAmplitude=<0 0 0>
    windFrequency=<0 0 0>
    windOffset=<0 0 0>
   }
   dBNO_Tumbling {
    maxDownInterpSpeed=5
    maxGustDuration=2
    maxGustStrength=7
    maxTimeBetweenGusts=7
    maxUpInterpSpeed=5
    minDownInterpSpeed=2
    minGustDuration=0.5
    minGustStrength=3
    minTimeBetweenGusts=0
    minUpInterpSpeed=2
    windTurbulenceScalarA=0.1
    windTurbulenceScalarB=1
    windAmplitude=<0 0 0>
    windFrequency=<0 0 0>
    windOffset=<0 0 0>
   }
   emoteOrMelee {
    maxDownInterpSpeed=5
    maxGustDuration=2
    maxGustStrength=7
    maxTimeBetweenGusts=7
    maxUpInterpSpeed=5
    minDownInterpSpeed=2
    minGustDuration=0.5
    minGustStrength=3
    minTimeBetweenGusts=0
    minUpInterpSpeed=2
    windTurbulenceScalarA=0.1
    windTurbulenceScalarB=1
    windAmplitude=<0 0 0>
    windFrequency=<0 0 0>
    windOffset=<0 0 0>
   }
   front_End {
    maxDownInterpSpeed=5
    maxGustDuration=2
    maxGustStrength=7
    maxTimeBetweenGusts=7
    maxUpInterpSpeed=5
    minDownInterpSpeed=2
    minGustDuration=0.5
    minGustStrength=3
    minTimeBetweenGusts=0
    minUpInterpSpeed=2
    windTurbulenceScalarA=0.1
    windTurbulenceScalarB=1
    windAmplitude=<0 0 0>
    windFrequency=<0 0 0>
    windOffset=<0 0 0>
   }
   onGround_Crouch_Moving {
    maxDownInterpSpeed=5
    maxGustDuration=2
    maxGustStrength=7
    maxTimeBetweenGusts=7
    maxUpInterpSpeed=5
    minDownInterpSpeed=2
    minGustDuration=0.5
    minGustStrength=3
    minTimeBetweenGusts=0
    minUpInterpSpeed=2
    windTurbulenceScalarA=0.1
    windTurbulenceScalarB=1
    windAmplitude=<0 0 0>
    windFrequency=<0 0 0>
    windOffset=<0 0 0>
   }
   onGround_Falling {
    maxDownInterpSpeed=5
    maxGustDuration=2
    maxGustStrength=7
    maxTimeBetweenGusts=7
    maxUpInterpSpeed=5
    minDownInterpSpeed=2
    minGustDuration=0.5
    minGustStrength=3
    minTimeBetweenGusts=0
    minUpInterpSpeed=2
    windTurbulenceScalarA=0.1
    windTurbulenceScalarB=1
    windAmplitude=<0 0 0>
    windFrequency=<0 0 0>
    windOffset=<0 0 0>
   }
   onGround_Moving {
    maxDownInterpSpeed=5
    maxGustDuration=2
    maxGustStrength=7
    maxTimeBetweenGusts=7
    maxUpInterpSpeed=1
    minDownInterpSpeed=2
    minGustDuration=0.5
    minGustStrength=-16.1506
    minTimeBetweenGusts=0
    minUpInterpSpeed=2
    windTurbulenceScalarA=0.1
    windTurbulenceScalarB=1
    windAmplitude=<0 0 0>
    windFrequency=<0 0 0>
    windOffset=<4.5838 62.1545 8.7233>
   }
   onGround_Standing {
    maxDownInterpSpeed=5
    maxGustDuration=2
    maxGustStrength=7
    maxTimeBetweenGusts=7
    maxUpInterpSpeed=5
    minDownInterpSpeed=2
    minGustDuration=0.5
    minGustStrength=3
    minTimeBetweenGusts=0
    minUpInterpSpeed=2
    windTurbulenceScalarA=0.1
    windTurbulenceScalarB=1
    windAmplitude=<0 0 0>
    windFrequency=<0 0 0>
    windOffset=<0 0 0>
   }
   rail_Grinding {
    maxDownInterpSpeed=5
    maxGustDuration=8
    maxGustStrength=7
    maxTimeBetweenGusts=8
    maxUpInterpSpeed=5
    minDownInterpSpeed=2
    minGustDuration=0.2
    minGustStrength=3
    minTimeBetweenGusts=1
    minUpInterpSpeed=2
    windTurbulenceScalarA=0.1
    windTurbulenceScalarB=0.8
    windAmplitude=<20 65 42>
    windFrequency=<5 6 8.6>
    windOffset=<50 -500 -400>
   }
   riding_Moving {
    maxDownInterpSpeed=5
    maxGustDuration=2
    maxGustStrength=7
    maxTimeBetweenGusts=7
    maxUpInterpSpeed=5
    minDownInterpSpeed=2
    minGustDuration=0.5
    minGustStrength=3
    minTimeBetweenGusts=0
    minUpInterpSpeed=2
    windTurbulenceScalarA=0.1
    windTurbulenceScalarB=1
    windAmplitude=<0 0 0>
    windFrequency=<0 0 0>
    windOffset=<0 0 0>
   }
   riding_Moving_Targeting {
    maxDownInterpSpeed=5
    maxGustDuration=2
    maxGustStrength=7
    maxTimeBetweenGusts=7
    maxUpInterpSpeed=5
    minDownInterpSpeed=2
    minGustDuration=0.5
    minGustStrength=3
    minTimeBetweenGusts=0
    minUpInterpSpeed=2
    windTurbulenceScalarA=0.1
    windTurbulenceScalarB=1
    windAmplitude=<0 0 0>
    windFrequency=<0 0 0>
    windOffset=<0 0 0>
   }
   riding_Standing {
    maxDownInterpSpeed=5
    maxGustDuration=2
    maxGustStrength=7
    maxTimeBetweenGusts=7
    maxUpInterpSpeed=5
    minDownInterpSpeed=2
    minGustDuration=0.5
    minGustStrength=3
    minTimeBetweenGusts=0
    minUpInterpSpeed=2
    windTurbulenceScalarA=0.1
    windTurbulenceScalarB=1
    windAmplitude=<0 0 0>
    windFrequency=<0 0 0>
    windOffset=<0 0 0>
   }
   skydive_DivingDown {
    maxDownInterpSpeed=5
    maxGustDuration=2
    maxGustStrength=7
    maxTimeBetweenGusts=7
    maxUpInterpSpeed=5
    minDownInterpSpeed=2
    minGustDuration=0.5
    minGustStrength=3
    minTimeBetweenGusts=0
    minUpInterpSpeed=2
    windTurbulenceScalarA=0.1
    windTurbulenceScalarB=1
    windAmplitude=<0 0 0>
    windFrequency=<0 0 0>
    windOffset=<0 0 0>
   }
   skydive_FloatingDown {
    maxDownInterpSpeed=5
    maxGustDuration=2
    maxGustStrength=7
    maxTimeBetweenGusts=7
    maxUpInterpSpeed=5
    minDownInterpSpeed=2
    minGustDuration=0.5
    minGustStrength=3
    minTimeBetweenGusts=0
    minUpInterpSpeed=2
    windTurbulenceScalarA=0.1
    windTurbulenceScalarB=1
    windAmplitude=<0 0 0>
    windFrequency=<0 0 0>
    windOffset=<0 0 0>
   }
   skydive_GoingUp {
    maxDownInterpSpeed=5
    maxGustDuration=2
    maxGustStrength=7
    maxTimeBetweenGusts=7
    maxUpInterpSpeed=5
    minDownInterpSpeed=2
    minGustDuration=0.5
    minGustStrength=3
    minTimeBetweenGusts=0
    minUpInterpSpeed=2
    windTurbulenceScalarA=0.1
    windTurbulenceScalarB=1
    windAmplitude=<0 0 0>
    windFrequency=<0 0 0>
    windOffset=<0 0 0>
   }
   skydive_Parachuting {
    maxDownInterpSpeed=5
    maxGustDuration=2
    maxGustStrength=7
    maxTimeBetweenGusts=7
    maxUpInterpSpeed=5
    minDownInterpSpeed=2
    minGustDuration=0.5
    minGustStrength=3
    minTimeBetweenGusts=0
    minUpInterpSpeed=2
    windTurbulenceScalarA=0.1
    windTurbulenceScalarB=1
    windAmplitude=<0 0 0>
    windFrequency=<0 0 0>
    windOffset=<0 0 0>
   }
   skydive_Surfing {
    maxDownInterpSpeed=5
    maxGustDuration=2
    maxGustStrength=7
    maxTimeBetweenGusts=7
    maxUpInterpSpeed=5
    minDownInterpSpeed=2
    minGustDuration=0.5
    minGustStrength=3
    minTimeBetweenGusts=0
    minUpInterpSpeed=2
    windTurbulenceScalarA=0.1
    windTurbulenceScalarB=1
    windAmplitude=<0 0 0>
    windFrequency=<0 0 0>
    windOffset=<0 0 0>
   }
   swimming {
    maxDownInterpSpeed=5
    maxGustDuration=2
    maxGustStrength=7
    maxTimeBetweenGusts=7
    maxUpInterpSpeed=5
    minDownInterpSpeed=2
    minGustDuration=0.5
    minGustStrength=3
    minTimeBetweenGusts=0
    minUpInterpSpeed=2
    windTurbulenceScalarA=0.1
    windTurbulenceScalarB=1
    windAmplitude=<0 0 0>
    windFrequency=<0 0 0>
    windOffset=<0 0 0>
   }
   tactical_Sprint {
    maxDownInterpSpeed=5
    maxGustDuration=2
    maxGustStrength=7
    maxTimeBetweenGusts=7
    maxUpInterpSpeed=5
    minDownInterpSpeed=2
    minGustDuration=0.5
    minGustStrength=3
    minTimeBetweenGusts=0
    minUpInterpSpeed=2
    windTurbulenceScalarA=0.1
    windTurbulenceScalarB=1
    windAmplitude=<0 0 0>
    windFrequency=<0 0 0>
    windOffset=<0 0 0>
   }
   ziplining {
    maxDownInterpSpeed=5
    maxGustDuration=2
    maxGustStrength=7
    maxTimeBetweenGusts=7
    maxUpInterpSpeed=5
    minDownInterpSpeed=2
    minGustDuration=0.5
    minGustStrength=3
    minTimeBetweenGusts=0
    minUpInterpSpeed=2
    windTurbulenceScalarA=0.1
    windTurbulenceScalarB=1
    windAmplitude=<0 0 0>
    windFrequency=<0 0 0>
    windOffset=<0 0 0>
   }
  }
 }
 Wind Gust Crouch Forward {
  type=WindGustParamsStruct
  default {
   dBNO {
    maxDownInterpSpeed=5
    maxGustDuration=2
    maxGustStrength=7
    maxTimeBetweenGusts=7
    maxUpInterpSpeed=5
    minDownInterpSpeed=2
    minGustDuration=0.5
    minGustStrength=3
    minTimeBetweenGusts=0
    minUpInterpSpeed=2
    windTurbulenceScalarA=0.1
    windTurbulenceScalarB=1
    windAmplitude=<0 0 0>
    windFrequency=<0 0 0>
    windOffset=<0 0 0>
   }
   dBNO_Carried {
    maxDownInterpSpeed=5
    maxGustDuration=2
    maxGustStrength=7
    maxTimeBetweenGusts=7
    maxUpInterpSpeed=5
    minDownInterpSpeed=2
    minGustDuration=0.5
    minGustStrength=3
    minTimeBetweenGusts=0
    minUpInterpSpeed=2
    windTurbulenceScalarA=0.1
    windTurbulenceScalarB=1
    windAmplitude=<0 0 0>
    windFrequency=<0 0 0>
    windOffset=<0 0 0>
   }
   dBNO_Impulse {
    maxDownInterpSpeed=5
    maxGustDuration=2
    maxGustStrength=7
    maxTimeBetweenGusts=7
    maxUpInterpSpeed=5
    minDownInterpSpeed=2
    minGustDuration=0.5
    minGustStrength=3
    minTimeBetweenGusts=0
    minUpInterpSpeed=2
    windTurbulenceScalarA=0.1
    windTurbulenceScalarB=1
    windAmplitude=<0 0 0>
    windFrequency=<0 0 0>
    windOffset=<0 0 0>
   }
   dBNO_RollBwd {
    maxDownInterpSpeed=5
    maxGustDuration=2
    maxGustStrength=7
    maxTimeBetweenGusts=7
    maxUpInterpSpeed=5
    minDownInterpSpeed=2
    minGustDuration=0.5
    minGustStrength=3
    minTimeBetweenGusts=0
    minUpInterpSpeed=2
    windTurbulenceScalarA=0.1
    windTurbulenceScalarB=1
    windAmplitude=<0 0 0>
    windFrequency=<0 0 0>
    windOffset=<0 0 0>
   }
   dBNO_RollFwd {
    maxDownInterpSpeed=5
    maxGustDuration=2
    maxGustStrength=7
    maxTimeBetweenGusts=7
    maxUpInterpSpeed=5
    minDownInterpSpeed=2
    minGustDuration=0.5
    minGustStrength=3
    minTimeBetweenGusts=0
    minUpInterpSpeed=2
    windTurbulenceScalarA=0.1
    windTurbulenceScalarB=1
    windAmplitude=<0 0 0>
    windFrequency=<0 0 0>
    windOffset=<0 0 0>
   }
   dBNO_Tumbling {
    maxDownInterpSpeed=5
    maxGustDuration=2
    maxGustStrength=7
    maxTimeBetweenGusts=7
    maxUpInterpSpeed=5
    minDownInterpSpeed=2
    minGustDuration=0.5
    minGustStrength=3
    minTimeBetweenGusts=0
    minUpInterpSpeed=2
    windTurbulenceScalarA=0.1
    windTurbulenceScalarB=1
    windAmplitude=<0 0 0>
    windFrequency=<0 0 0>
    windOffset=<0 0 0>
   }
   emoteOrMelee {
    maxDownInterpSpeed=5
    maxGustDuration=2
    maxGustStrength=7
    maxTimeBetweenGusts=7
    maxUpInterpSpeed=5
    minDownInterpSpeed=2
    minGustDuration=0.5
    minGustStrength=3
    minTimeBetweenGusts=0
    minUpInterpSpeed=2
    windTurbulenceScalarA=0.1
    windTurbulenceScalarB=1
    windAmplitude=<0 0 0>
    windFrequency=<0 0 0>
    windOffset=<0 0 0>
   }
   front_End {
    maxDownInterpSpeed=5
    maxGustDuration=2
    maxGustStrength=7
    maxTimeBetweenGusts=7
    maxUpInterpSpeed=5
    minDownInterpSpeed=2
    minGustDuration=0.5
    minGustStrength=3
    minTimeBetweenGusts=0
    minUpInterpSpeed=2
    windTurbulenceScalarA=0.1
    windTurbulenceScalarB=1
    windAmplitude=<0 0 0>
    windFrequency=<0 0 0>
    windOffset=<0 0 0>
   }
   onGround_Crouch_Moving {
    maxDownInterpSpeed=5
    maxGustDuration=2
    maxGustStrength=7
    maxTimeBetweenGusts=7
    maxUpInterpSpeed=5
    minDownInterpSpeed=2
    minGustDuration=0.5
    minGustStrength=3
    minTimeBetweenGusts=0
    minUpInterpSpeed=2
    windTurbulenceScalarA=0.1
    windTurbulenceScalarB=1
    windAmplitude=<0 0 0>
    windFrequency=<0 0 0>
    windOffset=<0 0 -496.3245>
   }
   onGround_Falling {
    maxDownInterpSpeed=5
    maxGustDuration=2
    maxGustStrength=7
    maxTimeBetweenGusts=7
    maxUpInterpSpeed=5
    minDownInterpSpeed=2
    minGustDuration=0.5
    minGustStrength=3
    minTimeBetweenGusts=0
    minUpInterpSpeed=2
    windTurbulenceScalarA=0.1
    windTurbulenceScalarB=1
    windAmplitude=<0 0 0>
    windFrequency=<0 0 0>
    windOffset=<0 0 0>
   }
   onGround_Moving {
    maxDownInterpSpeed=5
    maxGustDuration=2
    maxGustStrength=1.4618
    maxTimeBetweenGusts=5.7574
    maxUpInterpSpeed=1
    minDownInterpSpeed=2
    minGustDuration=0.5
    minGustStrength=-20.0533
    minTimeBetweenGusts=0
    minUpInterpSpeed=2
    windTurbulenceScalarA=0.1
    windTurbulenceScalarB=1
    windAmplitude=<0 0 0>
    windFrequency=<0 0 0>
    windOffset=<4.5838 371.8155 32.6673>
   }
   onGround_Standing {
    maxDownInterpSpeed=5
    maxGustDuration=2
    maxGustStrength=7
    maxTimeBetweenGusts=7
    maxUpInterpSpeed=5
    minDownInterpSpeed=2
    minGustDuration=0.5
    minGustStrength=3
    minTimeBetweenGusts=0
    minUpInterpSpeed=2
    windTurbulenceScalarA=0.1
    windTurbulenceScalarB=1
    windAmplitude=<0 0 0>
    windFrequency=<0 0 0>
    windOffset=<0 0 0>
   }
   rail_Grinding {
    maxDownInterpSpeed=5
    maxGustDuration=8
    maxGustStrength=7
    maxTimeBetweenGusts=8
    maxUpInterpSpeed=5
    minDownInterpSpeed=2
    minGustDuration=0.2
    minGustStrength=3
    minTimeBetweenGusts=1
    minUpInterpSpeed=2
    windTurbulenceScalarA=0.1
    windTurbulenceScalarB=0.8
    windAmplitude=<20 65 42>
    windFrequency=<5 6 8.6>
    windOffset=<50 -500 -400>
   }
   riding_Moving {
    maxDownInterpSpeed=5
    maxGustDuration=2
    maxGustStrength=7
    maxTimeBetweenGusts=7
    maxUpInterpSpeed=5
    minDownInterpSpeed=2
    minGustDuration=0.5
    minGustStrength=3
    minTimeBetweenGusts=0
    minUpInterpSpeed=2
    windTurbulenceScalarA=0.1
    windTurbulenceScalarB=1
    windAmplitude=<0 0 0>
    windFrequency=<0 0 0>
    windOffset=<0 0 0>
   }
   riding_Moving_Targeting {
    maxDownInterpSpeed=5
    maxGustDuration=2
    maxGustStrength=7
    maxTimeBetweenGusts=7
    maxUpInterpSpeed=5
    minDownInterpSpeed=2
    minGustDuration=0.5
    minGustStrength=3
    minTimeBetweenGusts=0
    minUpInterpSpeed=2
    windTurbulenceScalarA=0.1
    windTurbulenceScalarB=1
    windAmplitude=<0 0 0>
    windFrequency=<0 0 0>
    windOffset=<0 0 0>
   }
   riding_Standing {
    maxDownInterpSpeed=5
    maxGustDuration=2
    maxGustStrength=7
    maxTimeBetweenGusts=7
    maxUpInterpSpeed=5
    minDownInterpSpeed=2
    minGustDuration=0.5
    minGustStrength=3
    minTimeBetweenGusts=0
    minUpInterpSpeed=2
    windTurbulenceScalarA=0.1
    windTurbulenceScalarB=1
    windAmplitude=<0 0 0>
    windFrequency=<0 0 0>
    windOffset=<0 0 0>
   }
   skydive_DivingDown {
    maxDownInterpSpeed=5
    maxGustDuration=2
    maxGustStrength=7
    maxTimeBetweenGusts=7
    maxUpInterpSpeed=5
    minDownInterpSpeed=2
    minGustDuration=0.5
    minGustStrength=3
    minTimeBetweenGusts=0
    minUpInterpSpeed=2
    windTurbulenceScalarA=0.1
    windTurbulenceScalarB=1
    windAmplitude=<0 0 0>
    windFrequency=<0 0 0>
    windOffset=<0 0 0>
   }
   skydive_FloatingDown {
    maxDownInterpSpeed=5
    maxGustDuration=2
    maxGustStrength=7
    maxTimeBetweenGusts=7
    maxUpInterpSpeed=5
    minDownInterpSpeed=2
    minGustDuration=0.5
    minGustStrength=3
    minTimeBetweenGusts=0
    minUpInterpSpeed=2
    windTurbulenceScalarA=0.1
    windTurbulenceScalarB=1
    windAmplitude=<0 0 0>
    windFrequency=<0 0 0>
    windOffset=<0 0 0>
   }
   skydive_GoingUp {
    maxDownInterpSpeed=5
    maxGustDuration=2
    maxGustStrength=7
    maxTimeBetweenGusts=7
    maxUpInterpSpeed=5
    minDownInterpSpeed=2
    minGustDuration=0.5
    minGustStrength=3
    minTimeBetweenGusts=0
    minUpInterpSpeed=2
    windTurbulenceScalarA=0.1
    windTurbulenceScalarB=1
    windAmplitude=<0 0 0>
    windFrequency=<0 0 0>
    windOffset=<0 0 0>
   }
   skydive_Parachuting {
    maxDownInterpSpeed=5
    maxGustDuration=2
    maxGustStrength=7
    maxTimeBetweenGusts=7
    maxUpInterpSpeed=5
    minDownInterpSpeed=2
    minGustDuration=0.5
    minGustStrength=3
    minTimeBetweenGusts=0
    minUpInterpSpeed=2
    windTurbulenceScalarA=0.1
    windTurbulenceScalarB=1
    windAmplitude=<0 0 0>
    windFrequency=<0 0 0>
    windOffset=<0 0 0>
   }
   skydive_Surfing {
    maxDownInterpSpeed=5
    maxGustDuration=2
    maxGustStrength=7
    maxTimeBetweenGusts=7
    maxUpInterpSpeed=5
    minDownInterpSpeed=2
    minGustDuration=0.5
    minGustStrength=3
    minTimeBetweenGusts=0
    minUpInterpSpeed=2
    windTurbulenceScalarA=0.1
    windTurbulenceScalarB=1
    windAmplitude=<0 0 0>
    windFrequency=<0 0 0>
    windOffset=<0 0 0>
   }
   swimming {
    maxDownInterpSpeed=5
    maxGustDuration=2
    maxGustStrength=7
    maxTimeBetweenGusts=7
    maxUpInterpSpeed=5
    minDownInterpSpeed=2
    minGustDuration=0.5
    minGustStrength=3
    minTimeBetweenGusts=0
    minUpInterpSpeed=2
    windTurbulenceScalarA=0.1
    windTurbulenceScalarB=1
    windAmplitude=<0 0 0>
    windFrequency=<0 0 0>
    windOffset=<0 0 0>
   }
   tactical_Sprint {
    maxDownInterpSpeed=5
    maxGustDuration=2
    maxGustStrength=7
    maxTimeBetweenGusts=7
    maxUpInterpSpeed=5
    minDownInterpSpeed=2
    minGustDuration=0.5
    minGustStrength=3
    minTimeBetweenGusts=0
    minUpInterpSpeed=2
    windTurbulenceScalarA=0.1
    windTurbulenceScalarB=1
    windAmplitude=<0 0 0>
    windFrequency=<0 0 0>
    windOffset=<0 0 0>
   }
   ziplining {
    maxDownInterpSpeed=5
    maxGustDuration=2
    maxGustStrength=7
    maxTimeBetweenGusts=7
    maxUpInterpSpeed=5
    minDownInterpSpeed=2
    minGustDuration=0.5
    minGustStrength=3
    minTimeBetweenGusts=0
    minUpInterpSpeed=2
    windTurbulenceScalarA=0.1
    windTurbulenceScalarB=1
    windAmplitude=<0 0 0>
    windFrequency=<0 0 0>
    windOffset=<0 0 0>
   }
  }
 }
 Wind Gust Forward {
  type=WindGustParamsStruct
  default {
   dBNO {
    maxDownInterpSpeed=5
    maxGustDuration=2
    maxGustStrength=7
    maxTimeBetweenGusts=7
    maxUpInterpSpeed=5
    minDownInterpSpeed=2
    minGustDuration=0.5
    minGustStrength=3
    minTimeBetweenGusts=0
    minUpInterpSpeed=2
    windTurbulenceScalarA=0.1
    windTurbulenceScalarB=1
    windAmplitude=<0 0 0>
    windFrequency=<0 0 0>
    windOffset=<0 0 0>
   }
   dBNO_Carried {
    maxDownInterpSpeed=5
    maxGustDuration=2
    maxGustStrength=7
    maxTimeBetweenGusts=7
    maxUpInterpSpeed=5
    minDownInterpSpeed=2
    minGustDuration=0.5
    minGustStrength=3
    minTimeBetweenGusts=0
    minUpInterpSpeed=2
    windTurbulenceScalarA=0.1
    windTurbulenceScalarB=1
    windAmplitude=<0 0 0>
    windFrequency=<0 0 0>
    windOffset=<0 0 0>
   }
   dBNO_Impulse {
    maxDownInterpSpeed=5
    maxGustDuration=2
    maxGustStrength=7
    maxTimeBetweenGusts=7
    maxUpInterpSpeed=5
    minDownInterpSpeed=2
    minGustDuration=0.5
    minGustStrength=3
    minTimeBetweenGusts=0
    minUpInterpSpeed=2
    windTurbulenceScalarA=0.1
    windTurbulenceScalarB=1
    windAmplitude=<0 0 0>
    windFrequency=<0 0 0>
    windOffset=<0 0 0>
   }
   dBNO_RollBwd {
    maxDownInterpSpeed=5
    maxGustDuration=2
    maxGustStrength=7
    maxTimeBetweenGusts=7
    maxUpInterpSpeed=5
    minDownInterpSpeed=2
    minGustDuration=0.5
    minGustStrength=3
    minTimeBetweenGusts=0
    minUpInterpSpeed=2
    windTurbulenceScalarA=0.1
    windTurbulenceScalarB=1
    windAmplitude=<0 0 0>
    windFrequency=<0 0 0>
    windOffset=<0 0 0>
   }
   dBNO_RollFwd {
    maxDownInterpSpeed=5
    maxGustDuration=2
    maxGustStrength=7
    maxTimeBetweenGusts=7
    maxUpInterpSpeed=5
    minDownInterpSpeed=2
    minGustDuration=0.5
    minGustStrength=3
    minTimeBetweenGusts=0
    minUpInterpSpeed=2
    windTurbulenceScalarA=0.1
    windTurbulenceScalarB=1
    windAmplitude=<0 0 0>
    windFrequency=<0 0 0>
    windOffset=<0 0 0>
   }
   dBNO_Tumbling {
    maxDownInterpSpeed=5
    maxGustDuration=2
    maxGustStrength=7
    maxTimeBetweenGusts=7
    maxUpInterpSpeed=5
    minDownInterpSpeed=2
    minGustDuration=0.5
    minGustStrength=3
    minTimeBetweenGusts=0
    minUpInterpSpeed=2
    windTurbulenceScalarA=0.1
    windTurbulenceScalarB=1
    windAmplitude=<0 0 0>
    windFrequency=<0 0 0>
    windOffset=<0 0 0>
   }
   emoteOrMelee {
    maxDownInterpSpeed=5
    maxGustDuration=2
    maxGustStrength=7
    maxTimeBetweenGusts=7
    maxUpInterpSpeed=5
    minDownInterpSpeed=2
    minGustDuration=0.5
    minGustStrength=3
    minTimeBetweenGusts=0
    minUpInterpSpeed=2
    windTurbulenceScalarA=0.1
    windTurbulenceScalarB=1
    windAmplitude=<0 0 0>
    windFrequency=<0 0 0>
    windOffset=<0 0 0>
   }
   front_End {
    maxDownInterpSpeed=5
    maxGustDuration=2
    maxGustStrength=7
    maxTimeBetweenGusts=7
    maxUpInterpSpeed=5
    minDownInterpSpeed=2
    minGustDuration=0.5
    minGustStrength=3
    minTimeBetweenGusts=0
    minUpInterpSpeed=2
    windTurbulenceScalarA=0.1
    windTurbulenceScalarB=1
    windAmplitude=<0 0 0>
    windFrequency=<0 0 0>
    windOffset=<0 0 0>
   }
   onGround_Crouch_Moving {
    maxDownInterpSpeed=5
    maxGustDuration=2
    maxGustStrength=7
    maxTimeBetweenGusts=7
    maxUpInterpSpeed=5
    minDownInterpSpeed=2
    minGustDuration=0.5
    minGustStrength=3
    minTimeBetweenGusts=0
    minUpInterpSpeed=2
    windTurbulenceScalarA=0.1
    windTurbulenceScalarB=1
    windAmplitude=<0 0 0>
    windFrequency=<0 0 0>
    windOffset=<0 0 -496.3245>
   }
   onGround_Falling {
    maxDownInterpSpeed=5
    maxGustDuration=2
    maxGustStrength=7
    maxTimeBetweenGusts=7
    maxUpInterpSpeed=5
    minDownInterpSpeed=2
    minGustDuration=0.5
    minGustStrength=3
    minTimeBetweenGusts=0
    minUpInterpSpeed=2
    windTurbulenceScalarA=0.1
    windTurbulenceScalarB=1
    windAmplitude=<0 0 0>
    windFrequency=<0 0 0>
    windOffset=<0 0 0>
   }
   onGround_Moving {
    maxDownInterpSpeed=5
    maxGustDuration=2
    maxGustStrength=7
    maxTimeBetweenGusts=7
    maxUpInterpSpeed=1
    minDownInterpSpeed=2
    minGustDuration=0.5
    minGustStrength=-16.1506
    minTimeBetweenGusts=0
    minUpInterpSpeed=2
    windTurbulenceScalarA=0.1
    windTurbulenceScalarB=1
    windAmplitude=<0 0 0>
    windFrequency=<0 0 0>
    windOffset=<4.5838 534.2123 32.6673>
   }
   onGround_Standing {
    maxDownInterpSpeed=5
    maxGustDuration=2
    maxGustStrength=7
    maxTimeBetweenGusts=7
    maxUpInterpSpeed=5
    minDownInterpSpeed=2
    minGustDuration=0.5
    minGustStrength=3
    minTimeBetweenGusts=0
    minUpInterpSpeed=2
    windTurbulenceScalarA=0.1
    windTurbulenceScalarB=1
    windAmplitude=<0 0 0>
    windFrequency=<0 0 0>
    windOffset=<0 0 0>
   }
   rail_Grinding {
    maxDownInterpSpeed=5
    maxGustDuration=8
    maxGustStrength=7
    maxTimeBetweenGusts=8
    maxUpInterpSpeed=5
    minDownInterpSpeed=2
    minGustDuration=0.2
    minGustStrength=3
    minTimeBetweenGusts=1
    minUpInterpSpeed=2
    windTurbulenceScalarA=0.1
    windTurbulenceScalarB=0.8
    windAmplitude=<20 65 42>
    windFrequency=<5 6 8.6>
    windOffset=<50 -500 -400>
   }
   riding_Moving {
    maxDownInterpSpeed=5
    maxGustDuration=2
    maxGustStrength=7
    maxTimeBetweenGusts=7
    maxUpInterpSpeed=5
    minDownInterpSpeed=2
    minGustDuration=0.5
    minGustStrength=3
    minTimeBetweenGusts=0
    minUpInterpSpeed=2
    windTurbulenceScalarA=0.1
    windTurbulenceScalarB=1
    windAmplitude=<0 0 0>
    windFrequency=<0 0 0>
    windOffset=<0 0 0>
   }
   riding_Moving_Targeting {
    maxDownInterpSpeed=5
    maxGustDuration=2
    maxGustStrength=7
    maxTimeBetweenGusts=7
    maxUpInterpSpeed=5
    minDownInterpSpeed=2
    minGustDuration=0.5
    minGustStrength=3
    minTimeBetweenGusts=0
    minUpInterpSpeed=2
    windTurbulenceScalarA=0.1
    windTurbulenceScalarB=1
    windAmplitude=<0 0 0>
    windFrequency=<0 0 0>
    windOffset=<0 0 0>
   }
   riding_Standing {
    maxDownInterpSpeed=5
    maxGustDuration=2
    maxGustStrength=7
    maxTimeBetweenGusts=7
    maxUpInterpSpeed=5
    minDownInterpSpeed=2
    minGustDuration=0.5
    minGustStrength=3
    minTimeBetweenGusts=0
    minUpInterpSpeed=2
    windTurbulenceScalarA=0.1
    windTurbulenceScalarB=1
    windAmplitude=<0 0 0>
    windFrequency=<0 0 0>
    windOffset=<0 0 0>
   }
   skydive_DivingDown {
    maxDownInterpSpeed=5
    maxGustDuration=2
    maxGustStrength=7
    maxTimeBetweenGusts=7
    maxUpInterpSpeed=5
    minDownInterpSpeed=2
    minGustDuration=0.5
    minGustStrength=3
    minTimeBetweenGusts=0
    minUpInterpSpeed=2
    windTurbulenceScalarA=0.1
    windTurbulenceScalarB=1
    windAmplitude=<0 0 0>
    windFrequency=<0 0 0>
    windOffset=<0 0 0>
   }
   skydive_FloatingDown {
    maxDownInterpSpeed=5
    maxGustDuration=2
    maxGustStrength=7
    maxTimeBetweenGusts=7
    maxUpInterpSpeed=5
    minDownInterpSpeed=2
    minGustDuration=0.5
    minGustStrength=3
    minTimeBetweenGusts=0
    minUpInterpSpeed=2
    windTurbulenceScalarA=0.1
    windTurbulenceScalarB=1
    windAmplitude=<0 0 0>
    windFrequency=<0 0 0>
    windOffset=<0 0 0>
   }
   skydive_GoingUp {
    maxDownInterpSpeed=5
    maxGustDuration=2
    maxGustStrength=7
    maxTimeBetweenGusts=7
    maxUpInterpSpeed=5
    minDownInterpSpeed=2
    minGustDuration=0.5
    minGustStrength=3
    minTimeBetweenGusts=0
    minUpInterpSpeed=2
    windTurbulenceScalarA=0.1
    windTurbulenceScalarB=1
    windAmplitude=<0 0 0>
    windFrequency=<0 0 0>
    windOffset=<0 0 0>
   }
   skydive_Parachuting {
    maxDownInterpSpeed=5
    maxGustDuration=2
    maxGustStrength=7
    maxTimeBetweenGusts=7
    maxUpInterpSpeed=5
    minDownInterpSpeed=2
    minGustDuration=0.5
    minGustStrength=3
    minTimeBetweenGusts=0
    minUpInterpSpeed=2
    windTurbulenceScalarA=0.1
    windTurbulenceScalarB=1
    windAmplitude=<0 0 0>
    windFrequency=<0 0 0>
    windOffset=<0 0 0>
   }
   skydive_Surfing {
    maxDownInterpSpeed=5
    maxGustDuration=2
    maxGustStrength=7
    maxTimeBetweenGusts=7
    maxUpInterpSpeed=5
    minDownInterpSpeed=2
    minGustDuration=0.5
    minGustStrength=3
    minTimeBetweenGusts=0
    minUpInterpSpeed=2
    windTurbulenceScalarA=0.1
    windTurbulenceScalarB=1
    windAmplitude=<0 0 0>
    windFrequency=<0 0 0>
    windOffset=<0 0 0>
   }
   swimming {
    maxDownInterpSpeed=5
    maxGustDuration=2
    maxGustStrength=7
    maxTimeBetweenGusts=7
    maxUpInterpSpeed=5
    minDownInterpSpeed=2
    minGustDuration=0.5
    minGustStrength=3
    minTimeBetweenGusts=0
    minUpInterpSpeed=2
    windTurbulenceScalarA=0.1
    windTurbulenceScalarB=1
    windAmplitude=<0 0 0>
    windFrequency=<0 0 0>
    windOffset=<0 0 0>
   }
   tactical_Sprint {
    maxDownInterpSpeed=5
    maxGustDuration=2
    maxGustStrength=7
    maxTimeBetweenGusts=7
    maxUpInterpSpeed=1
    minDownInterpSpeed=2
    minGustDuration=0.5
    minGustStrength=-16.1506
    minTimeBetweenGusts=0
    minUpInterpSpeed=2
    windTurbulenceScalarA=0.1
    windTurbulenceScalarB=1
    windAmplitude=<0 0 0>
    windFrequency=<0 0 0>
    windOffset=<0.5548 735.9639 14.3098>
   }
   ziplining {
    maxDownInterpSpeed=5
    maxGustDuration=2
    maxGustStrength=7
    maxTimeBetweenGusts=7
    maxUpInterpSpeed=5
    minDownInterpSpeed=2
    minGustDuration=0.5
    minGustStrength=3
    minTimeBetweenGusts=0
    minUpInterpSpeed=2
    windTurbulenceScalarA=0.1
    windTurbulenceScalarB=1
    windAmplitude=<0 0 0>
    windFrequency=<0 0 0>
    windOffset=<0 0 0>
   }
  }
 }
 Wind Gust Params {
  type=WindGustParamsStruct
  default {
   dBNO {
    maxDownInterpSpeed=5
    maxGustDuration=2
    maxGustStrength=7
    maxTimeBetweenGusts=7
    maxUpInterpSpeed=5
    minDownInterpSpeed=2
    minGustDuration=0.5
    minGustStrength=3
    minTimeBetweenGusts=0
    minUpInterpSpeed=2
    windTurbulenceScalarA=0.1
    windTurbulenceScalarB=1
    windAmplitude=<0 0 0>
    windFrequency=<0 0 0>
    windOffset=<0 0 0>
   }
   dBNO_Carried {
    maxDownInterpSpeed=5
    maxGustDuration=2
    maxGustStrength=7
    maxTimeBetweenGusts=7
    maxUpInterpSpeed=5
    minDownInterpSpeed=2
    minGustDuration=0.5
    minGustStrength=3
    minTimeBetweenGusts=0
    minUpInterpSpeed=2
    windTurbulenceScalarA=0.1
    windTurbulenceScalarB=1
    windAmplitude=<0 0 0>
    windFrequency=<0 0 0>
    windOffset=<0 0 0>
   }
   dBNO_Impulse {
    maxDownInterpSpeed=5
    maxGustDuration=2
    maxGustStrength=7
    maxTimeBetweenGusts=7
    maxUpInterpSpeed=5
    minDownInterpSpeed=2
    minGustDuration=0.5
    minGustStrength=3
    minTimeBetweenGusts=0
    minUpInterpSpeed=2
    windTurbulenceScalarA=0.1
    windTurbulenceScalarB=1
    windAmplitude=<0 0 0>
    windFrequency=<0 0 0>
    windOffset=<0 0 0>
   }
   dBNO_RollBwd {
    maxDownInterpSpeed=5
    maxGustDuration=2
    maxGustStrength=7
    maxTimeBetweenGusts=7
    maxUpInterpSpeed=5
    minDownInterpSpeed=2
    minGustDuration=0.5
    minGustStrength=3
    minTimeBetweenGusts=0
    minUpInterpSpeed=2
    windTurbulenceScalarA=0.1
    windTurbulenceScalarB=1
    windAmplitude=<0 0 0>
    windFrequency=<0 0 0>
    windOffset=<0 0 0>
   }
   dBNO_RollFwd {
    maxDownInterpSpeed=5
    maxGustDuration=2
    maxGustStrength=7
    maxTimeBetweenGusts=7
    maxUpInterpSpeed=5
    minDownInterpSpeed=2
    minGustDuration=0.5
    minGustStrength=3
    minTimeBetweenGusts=0
    minUpInterpSpeed=2
    windTurbulenceScalarA=0.1
    windTurbulenceScalarB=1
    windAmplitude=<0 0 0>
    windFrequency=<0 0 0>
    windOffset=<0 0 0>
   }
   dBNO_Tumbling {
    maxDownInterpSpeed=5
    maxGustDuration=2
    maxGustStrength=7
    maxTimeBetweenGusts=7
    maxUpInterpSpeed=5
    minDownInterpSpeed=2
    minGustDuration=0.5
    minGustStrength=3
    minTimeBetweenGusts=0
    minUpInterpSpeed=2
    windTurbulenceScalarA=0.1
    windTurbulenceScalarB=1
    windAmplitude=<0 0 0>
    windFrequency=<0 0 0>
    windOffset=<0 0 0>
   }
   emoteOrMelee {
    maxDownInterpSpeed=5
    maxGustDuration=2
    maxGustStrength=7
    maxTimeBetweenGusts=7
    maxUpInterpSpeed=5
    minDownInterpSpeed=2
    minGustDuration=0.5
    minGustStrength=3
    minTimeBetweenGusts=0
    minUpInterpSpeed=2
    windTurbulenceScalarA=0.1
    windTurbulenceScalarB=1
    windAmplitude=<0 0 0>
    windFrequency=<0 0 0>
    windOffset=<0 0 0>
   }
   front_End {
    maxDownInterpSpeed=5
    maxGustDuration=2
    maxGustStrength=7
    maxTimeBetweenGusts=7
    maxUpInterpSpeed=5
    minDownInterpSpeed=2
    minGustDuration=0.5
    minGustStrength=3
    minTimeBetweenGusts=0
    minUpInterpSpeed=2
    windTurbulenceScalarA=0.1
    windTurbulenceScalarB=1
    windAmplitude=<0 0 0>
    windFrequency=<0 0 0>
    windOffset=<0 0 0>
   }
   onGround_Crouch_Moving {
    maxDownInterpSpeed=5
    maxGustDuration=2
    maxGustStrength=7
    maxTimeBetweenGusts=7
    maxUpInterpSpeed=5
    minDownInterpSpeed=2
    minGustDuration=0.5
    minGustStrength=3
    minTimeBetweenGusts=0
    minUpInterpSpeed=2
    windTurbulenceScalarA=0.1
    windTurbulenceScalarB=1
    windAmplitude=<0 0 0>
    windFrequency=<0 0 0>
    windOffset=<0 0 0>
   }
   onGround_Falling {
    maxDownInterpSpeed=5
    maxGustDuration=2
    maxGustStrength=7
    maxTimeBetweenGusts=7
    maxUpInterpSpeed=5
    minDownInterpSpeed=2
    minGustDuration=0.5
    minGustStrength=3
    minTimeBetweenGusts=0
    minUpInterpSpeed=2
    windTurbulenceScalarA=0.1
    windTurbulenceScalarB=1
    windAmplitude=<0 0 0>
    windFrequency=<0 0 0>
    windOffset=<0 0 0>
   }
   onGround_Moving {
    maxDownInterpSpeed=5
    maxGustDuration=2
    maxGustStrength=7
    maxTimeBetweenGusts=7
    maxUpInterpSpeed=1
    minDownInterpSpeed=2
    minGustDuration=0.5
    minGustStrength=-16.1506
    minTimeBetweenGusts=0
    minUpInterpSpeed=2
    windTurbulenceScalarA=0.1
    windTurbulenceScalarB=1
    windAmplitude=<0 0 0>
    windFrequency=<0 0 0>
    windOffset=<4.5838 298.5716 8.7233>
   }
   onGround_Standing {
    maxDownInterpSpeed=5
    maxGustDuration=2
    maxGustStrength=7
    maxTimeBetweenGusts=7
    maxUpInterpSpeed=5
    minDownInterpSpeed=2
    minGustDuration=0.5
    minGustStrength=3
    minTimeBetweenGusts=0
    minUpInterpSpeed=2
    windTurbulenceScalarA=0.1
    windTurbulenceScalarB=1
    windAmplitude=<0 0 0>
    windFrequency=<0 0 0>
    windOffset=<0 0 0>
   }
   rail_Grinding {
    maxDownInterpSpeed=5
    maxGustDuration=8
    maxGustStrength=7
    maxTimeBetweenGusts=8
    maxUpInterpSpeed=5
    minDownInterpSpeed=2
    minGustDuration=0.2
    minGustStrength=3
    minTimeBetweenGusts=1
    minUpInterpSpeed=2
    windTurbulenceScalarA=0.1
    windTurbulenceScalarB=0.8
    windAmplitude=<20 65 42>
    windFrequency=<5 6 8.6>
    windOffset=<50 -500 -400>
   }
   riding_Moving {
    maxDownInterpSpeed=5
    maxGustDuration=2
    maxGustStrength=7
    maxTimeBetweenGusts=7
    maxUpInterpSpeed=5
    minDownInterpSpeed=2
    minGustDuration=0.5
    minGustStrength=3
    minTimeBetweenGusts=0
    minUpInterpSpeed=2
    windTurbulenceScalarA=0.1
    windTurbulenceScalarB=1
    windAmplitude=<0 0 0>
    windFrequency=<0 0 0>
    windOffset=<0 0 0>
   }
   riding_Moving_Targeting {
    maxDownInterpSpeed=5
    maxGustDuration=2
    maxGustStrength=7
    maxTimeBetweenGusts=7
    maxUpInterpSpeed=5
    minDownInterpSpeed=2
    minGustDuration=0.5
    minGustStrength=3
    minTimeBetweenGusts=0
    minUpInterpSpeed=2
    windTurbulenceScalarA=0.1
    windTurbulenceScalarB=1
    windAmplitude=<0 0 0>
    windFrequency=<0 0 0>
    windOffset=<0 0 0>
   }
   riding_Standing {
    maxDownInterpSpeed=5
    maxGustDuration=2
    maxGustStrength=7
    maxTimeBetweenGusts=7
    maxUpInterpSpeed=5
    minDownInterpSpeed=2
    minGustDuration=0.5
    minGustStrength=3
    minTimeBetweenGusts=0
    minUpInterpSpeed=2
    windTurbulenceScalarA=0.1
    windTurbulenceScalarB=1
    windAmplitude=<0 0 0>
    windFrequency=<0 0 0>
    windOffset=<0 0 0>
   }
   skydive_DivingDown {
    maxDownInterpSpeed=5
    maxGustDuration=2
    maxGustStrength=7
    maxTimeBetweenGusts=7
    maxUpInterpSpeed=5
    minDownInterpSpeed=2
    minGustDuration=0.5
    minGustStrength=3
    minTimeBetweenGusts=0
    minUpInterpSpeed=2
    windTurbulenceScalarA=0.1
    windTurbulenceScalarB=1
    windAmplitude=<0 0 0>
    windFrequency=<0 0 0>
    windOffset=<0 0 0>
   }
   skydive_FloatingDown {
    maxDownInterpSpeed=5
    maxGustDuration=2
    maxGustStrength=7
    maxTimeBetweenGusts=7
    maxUpInterpSpeed=5
    minDownInterpSpeed=2
    minGustDuration=0.5
    minGustStrength=3
    minTimeBetweenGusts=0
    minUpInterpSpeed=2
    windTurbulenceScalarA=0.1
    windTurbulenceScalarB=1
    windAmplitude=<0 0 0>
    windFrequency=<0 0 0>
    windOffset=<0 0 0>
   }
   skydive_GoingUp {
    maxDownInterpSpeed=5
    maxGustDuration=2
    maxGustStrength=7
    maxTimeBetweenGusts=7
    maxUpInterpSpeed=5
    minDownInterpSpeed=2
    minGustDuration=0.5
    minGustStrength=3
    minTimeBetweenGusts=0
    minUpInterpSpeed=2
    windTurbulenceScalarA=0.1
    windTurbulenceScalarB=1
    windAmplitude=<0 0 0>
    windFrequency=<0 0 0>
    windOffset=<0 0 0>
   }
   skydive_Parachuting {
    maxDownInterpSpeed=5
    maxGustDuration=2
    maxGustStrength=7
    maxTimeBetweenGusts=7
    maxUpInterpSpeed=5
    minDownInterpSpeed=2
    minGustDuration=0.5
    minGustStrength=3
    minTimeBetweenGusts=0
    minUpInterpSpeed=2
    windTurbulenceScalarA=0.1
    windTurbulenceScalarB=1
    windAmplitude=<0 0 0>
    windFrequency=<0 0 0>
    windOffset=<0 0 0>
   }
   skydive_Surfing {
    maxDownInterpSpeed=5
    maxGustDuration=2
    maxGustStrength=7
    maxTimeBetweenGusts=7
    maxUpInterpSpeed=5
    minDownInterpSpeed=2
    minGustDuration=0.5
    minGustStrength=3
    minTimeBetweenGusts=0
    minUpInterpSpeed=2
    windTurbulenceScalarA=0.1
    windTurbulenceScalarB=1
    windAmplitude=<0 0 0>
    windFrequency=<0 0 0>
    windOffset=<0 0 0>
   }
   swimming {
    maxDownInterpSpeed=5
    maxGustDuration=2
    maxGustStrength=7
    maxTimeBetweenGusts=7
    maxUpInterpSpeed=5
    minDownInterpSpeed=2
    minGustDuration=0.5
    minGustStrength=3
    minTimeBetweenGusts=0
    minUpInterpSpeed=2
    windTurbulenceScalarA=0.1
    windTurbulenceScalarB=1
    windAmplitude=<0 0 0>
    windFrequency=<0 0 0>
    windOffset=<0 0 0>
   }
   tactical_Sprint {
    maxDownInterpSpeed=5
    maxGustDuration=2
    maxGustStrength=7
    maxTimeBetweenGusts=7
    maxUpInterpSpeed=5
    minDownInterpSpeed=2
    minGustDuration=0.5
    minGustStrength=3
    minTimeBetweenGusts=0
    minUpInterpSpeed=2
    windTurbulenceScalarA=0.1
    windTurbulenceScalarB=1
    windAmplitude=<0 0 0>
    windFrequency=<0 0 0>
    windOffset=<0 0 0>
   }
   ziplining {
    maxDownInterpSpeed=5
    maxGustDuration=2
    maxGustStrength=7
    maxTimeBetweenGusts=7
    maxUpInterpSpeed=5
    minDownInterpSpeed=2
    minGustDuration=0.5
    minGustStrength=3
    minTimeBetweenGusts=0
    minUpInterpSpeed=2
    windTurbulenceScalarA=0.1
    windTurbulenceScalarB=1
    windAmplitude=<0 0 0>
    windFrequency=<0 0 0>
    windOffset=<0 0 0>
   }
  }
 }
 Wind Gusts Back Cloth Params {
  type=WindGustParamsStruct
  default {
   dBNO {
    maxDownInterpSpeed=5
    maxGustDuration=2
    maxGustStrength=7
    maxTimeBetweenGusts=7
    maxUpInterpSpeed=5
    minDownInterpSpeed=2
    minGustDuration=0.5
    minGustStrength=3
    minTimeBetweenGusts=0
    minUpInterpSpeed=2
    windTurbulenceScalarA=0.1
    windTurbulenceScalarB=1
    windAmplitude=<0 0 0>
    windFrequency=<0 0 0>
    windOffset=<0 0 0>
   }
   dBNO_Carried {
    maxDownInterpSpeed=5
    maxGustDuration=2
    maxGustStrength=7
    maxTimeBetweenGusts=7
    maxUpInterpSpeed=5
    minDownInterpSpeed=2
    minGustDuration=0.5
    minGustStrength=3
    minTimeBetweenGusts=0
    minUpInterpSpeed=2
    windTurbulenceScalarA=0.1
    windTurbulenceScalarB=1
    windAmplitude=<0 0 0>
    windFrequency=<0 0 0>
    windOffset=<0 0 0>
   }
   dBNO_Impulse {
    maxDownInterpSpeed=5
    maxGustDuration=2
    maxGustStrength=7
    maxTimeBetweenGusts=7
    maxUpInterpSpeed=5
    minDownInterpSpeed=2
    minGustDuration=0.5
    minGustStrength=3
    minTimeBetweenGusts=0
    minUpInterpSpeed=2
    windTurbulenceScalarA=0.1
    windTurbulenceScalarB=1
    windAmplitude=<0 0 0>
    windFrequency=<0 0 0>
    windOffset=<0 0 0>
   }
   dBNO_RollBwd {
    maxDownInterpSpeed=5
    maxGustDuration=2
    maxGustStrength=7
    maxTimeBetweenGusts=7
    maxUpInterpSpeed=5
    minDownInterpSpeed=2
    minGustDuration=0.5
    minGustStrength=3
    minTimeBetweenGusts=0
    minUpInterpSpeed=2
    windTurbulenceScalarA=0.1
    windTurbulenceScalarB=1
    windAmplitude=<0 0 0>
    windFrequency=<0 0 0>
    windOffset=<0 0 0>
   }
   dBNO_RollFwd {
    maxDownInterpSpeed=5
    maxGustDuration=2
    maxGustStrength=7
    maxTimeBetweenGusts=7
    maxUpInterpSpeed=5
    minDownInterpSpeed=2
    minGustDuration=0.5
    minGustStrength=3
    minTimeBetweenGusts=0
    minUpInterpSpeed=2
    windTurbulenceScalarA=0.1
    windTurbulenceScalarB=1
    windAmplitude=<0 0 0>
    windFrequency=<0 0 0>
    windOffset=<0 0 0>
   }
   dBNO_Tumbling {
    maxDownInterpSpeed=5
    maxGustDuration=2
    maxGustStrength=7
    maxTimeBetweenGusts=7
    maxUpInterpSpeed=5
    minDownInterpSpeed=2
    minGustDuration=0.5
    minGustStrength=3
    minTimeBetweenGusts=0
    minUpInterpSpeed=2
    windTurbulenceScalarA=0.1
    windTurbulenceScalarB=1
    windAmplitude=<0 0 0>
    windFrequency=<0 0 0>
    windOffset=<0 0 0>
   }
   emoteOrMelee {
    maxDownInterpSpeed=5
    maxGustDuration=2
    maxGustStrength=7
    maxTimeBetweenGusts=7
    maxUpInterpSpeed=5
    minDownInterpSpeed=2
    minGustDuration=0.5
    minGustStrength=3
    minTimeBetweenGusts=0
    minUpInterpSpeed=2
    windTurbulenceScalarA=0.1
    windTurbulenceScalarB=1
    windAmplitude=<30 50 50>
    windFrequency=<2 3 5>
    windOffset=<0 50 -100>
   }
   front_End {
    maxDownInterpSpeed=5
    maxGustDuration=2
    maxGustStrength=4
    maxTimeBetweenGusts=7
    maxUpInterpSpeed=5
    minDownInterpSpeed=2
    minGustDuration=0.5
    minGustStrength=2
    minTimeBetweenGusts=0
    minUpInterpSpeed=2
    windTurbulenceScalarA=0.1
    windTurbulenceScalarB=1
    windAmplitude=<10 30 50>
    windFrequency=<10 3 4>
    windOffset=<0 0 -250>
   }
   onGround_Crouch_Moving {
    maxDownInterpSpeed=5
    maxGustDuration=2
    maxGustStrength=7
    maxTimeBetweenGusts=7
    maxUpInterpSpeed=5
    minDownInterpSpeed=2
    minGustDuration=0.5
    minGustStrength=3
    minTimeBetweenGusts=0
    minUpInterpSpeed=2
    windTurbulenceScalarA=0.1
    windTurbulenceScalarB=1
    windAmplitude=<100 25 75>
    windFrequency=<1.5 0.5 7.5>
    windOffset=<0 -100 -175>
   }
   onGround_Falling {
    maxDownInterpSpeed=5
    maxGustDuration=2
    maxGustStrength=7
    maxTimeBetweenGusts=7
    maxUpInterpSpeed=5
    minDownInterpSpeed=2
    minGustDuration=0.5
    minGustStrength=3
    minTimeBetweenGusts=0
    minUpInterpSpeed=2
    windTurbulenceScalarA=0.1
    windTurbulenceScalarB=1
    windAmplitude=<0 0 0>
    windFrequency=<0 0 0>
    windOffset=<0 0 0>
   }
   onGround_Moving {
    maxDownInterpSpeed=5
    maxGustDuration=8
    maxGustStrength=7
    maxTimeBetweenGusts=8
    maxUpInterpSpeed=5
    minDownInterpSpeed=2
    minGustDuration=0.2
    minGustStrength=3
    minTimeBetweenGusts=1
    minUpInterpSpeed=2
    windTurbulenceScalarA=0.1
    windTurbulenceScalarB=0.8
    windAmplitude=<30 65 42>
    windFrequency=<5 6 8.6>
    windOffset=<30 -150 -200>
   }
   onGround_Standing {
    maxDownInterpSpeed=5
    maxGustDuration=8
    maxGustStrength=7
    maxTimeBetweenGusts=8
    maxUpInterpSpeed=5
    minDownInterpSpeed=2
    minGustDuration=0.2
    minGustStrength=2
    minTimeBetweenGusts=1
    minUpInterpSpeed=2
    windTurbulenceScalarA=0.1
    windTurbulenceScalarB=0.8
    windAmplitude=<75 24 28>
    windFrequency=<1.5 2.3 3.7>
    windOffset=<0 0 -150>
   }
   rail_Grinding {
    maxDownInterpSpeed=5
    maxGustDuration=8
    maxGustStrength=7
    maxTimeBetweenGusts=8
    maxUpInterpSpeed=5
    minDownInterpSpeed=2
    minGustDuration=0.2
    minGustStrength=3
    minTimeBetweenGusts=1
    minUpInterpSpeed=2
    windTurbulenceScalarA=0.1
    windTurbulenceScalarB=0.8
    windAmplitude=<20 65 42>
    windFrequency=<5 6 8.6>
    windOffset=<50 -500 -400>
   }
   riding_Moving {
    maxDownInterpSpeed=5
    maxGustDuration=2
    maxGustStrength=7
    maxTimeBetweenGusts=7
    maxUpInterpSpeed=5
    minDownInterpSpeed=2
    minGustDuration=0.5
    minGustStrength=3
    minTimeBetweenGusts=0
    minUpInterpSpeed=2
    windTurbulenceScalarA=0.1
    windTurbulenceScalarB=1
    windAmplitude=<0 0 0>
    windFrequency=<0 0 0>
    windOffset=<0 0 0>
   }
   riding_Moving_Targeting {
    maxDownInterpSpeed=5
    maxGustDuration=2
    maxGustStrength=7
    maxTimeBetweenGusts=7
    maxUpInterpSpeed=5
    minDownInterpSpeed=2
    minGustDuration=0.5
    minGustStrength=3
    minTimeBetweenGusts=0
    minUpInterpSpeed=2
    windTurbulenceScalarA=0.1
    windTurbulenceScalarB=1
    windAmplitude=<0 0 0>
    windFrequency=<0 0 0>
    windOffset=<0 0 0>
   }
   riding_Standing {
    maxDownInterpSpeed=5
    maxGustDuration=2
    maxGustStrength=7
    maxTimeBetweenGusts=7
    maxUpInterpSpeed=5
    minDownInterpSpeed=2
    minGustDuration=0.5
    minGustStrength=3
    minTimeBetweenGusts=0
    minUpInterpSpeed=2
    windTurbulenceScalarA=0.1
    windTurbulenceScalarB=1
    windAmplitude=<0 0 0>
    windFrequency=<0 0 0>
    windOffset=<0 0 0>
   }
   skydive_DivingDown {
    maxDownInterpSpeed=5
    maxGustDuration=1
    maxGustStrength=5
    maxTimeBetweenGusts=5
    maxUpInterpSpeed=5
    minDownInterpSpeed=2
    minGustDuration=0.2
    minGustStrength=2
    minTimeBetweenGusts=1
    minUpInterpSpeed=2
    windTurbulenceScalarA=0.1
    windTurbulenceScalarB=2
    windAmplitude=<0 15 25>
    windFrequency=<0 2 4>
    windOffset=<-450 50 100>
   }
   skydive_FloatingDown {
    maxDownInterpSpeed=5
    maxGustDuration=8
    maxGustStrength=1
    maxTimeBetweenGusts=6
    maxUpInterpSpeed=5
    minDownInterpSpeed=2
    minGustDuration=0.2
    minGustStrength=0
    minTimeBetweenGusts=1
    minUpInterpSpeed=2
    windTurbulenceScalarA=0.5
    windTurbulenceScalarB=1.1
    windAmplitude=<50 20 35>
    windFrequency=<8 2 3>
    windOffset=<0 -300 0>
   }
   skydive_GoingUp {
    maxDownInterpSpeed=5
    maxGustDuration=8
    maxGustStrength=7
    maxTimeBetweenGusts=8
    maxUpInterpSpeed=5
    minDownInterpSpeed=2
    minGustDuration=0.2
    minGustStrength=3
    minTimeBetweenGusts=1
    minUpInterpSpeed=2
    windTurbulenceScalarA=0.1
    windTurbulenceScalarB=0.8
    windAmplitude=<0 0 0>
    windFrequency=<0 0 0>
    windOffset=<0 -500 300>
   }
   skydive_Parachuting {
    maxDownInterpSpeed=5
    maxGustDuration=8
    maxGustStrength=1
    maxTimeBetweenGusts=8
    maxUpInterpSpeed=5
    minDownInterpSpeed=2
    minGustDuration=5
    minGustStrength=0
    minTimeBetweenGusts=5
    minUpInterpSpeed=2
    windTurbulenceScalarA=0.2
    windTurbulenceScalarB=1
    windAmplitude=<25 50 75>
    windFrequency=<8 6 2>
    windOffset=<-150 -400 -150>
   }
   skydive_Surfing {
    maxDownInterpSpeed=5
    maxGustDuration=2
    maxGustStrength=7
    maxTimeBetweenGusts=7
    maxUpInterpSpeed=5
    minDownInterpSpeed=2
    minGustDuration=0.5
    minGustStrength=3
    minTimeBetweenGusts=0
    minUpInterpSpeed=2
    windTurbulenceScalarA=0.1
    windTurbulenceScalarB=1
    windAmplitude=<0 0 0>
    windFrequency=<0 0 0>
    windOffset=<0 0 0>
   }
   swimming {
    maxDownInterpSpeed=5
    maxGustDuration=2
    maxGustStrength=7
    maxTimeBetweenGusts=7
    maxUpInterpSpeed=5
    minDownInterpSpeed=2
    minGustDuration=0.5
    minGustStrength=3
    minTimeBetweenGusts=0
    minUpInterpSpeed=2
    windTurbulenceScalarA=0.1
    windTurbulenceScalarB=1
    windAmplitude=<10 25 20>
    windFrequency=<1.5 0.5 7.5>
    windOffset=<0 -150 50>
   }
   tactical_Sprint {
    maxDownInterpSpeed=5
    maxGustDuration=2
    maxGustStrength=7
    maxTimeBetweenGusts=7
    maxUpInterpSpeed=5
    minDownInterpSpeed=2
    minGustDuration=0.5
    minGustStrength=3
    minTimeBetweenGusts=0
    minUpInterpSpeed=2
    windTurbulenceScalarA=0.1
    windTurbulenceScalarB=1
    windAmplitude=<0 0 0>
    windFrequency=<0 0 0>
    windOffset=<0 0 0>
   }
   ziplining {
    maxDownInterpSpeed=5
    maxGustDuration=2
    maxGustStrength=7
    maxTimeBetweenGusts=7
    maxUpInterpSpeed=5
    minDownInterpSpeed=2
    minGustDuration=0.5
    minGustStrength=3
    minTimeBetweenGusts=0
    minUpInterpSpeed=2
    windTurbulenceScalarA=0.1
    windTurbulenceScalarB=1
    windAmplitude=<0 0 0>
    windFrequency=<0 0 0>
    windOffset=<0 0 0>
   }
  }
 }
]
