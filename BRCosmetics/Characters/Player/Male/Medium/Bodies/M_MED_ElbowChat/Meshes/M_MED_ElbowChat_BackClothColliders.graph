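# Physics asset for the M_MED_ElbowChat male medium body that, per its name, supplies colliders for back cloth.
# Two overlapping, slightly offset pelvis capsules cover the hips, and a wide, flat box on root acts as a floor near the character's origin.
# Thigh, calf and foot bodies show no shapes.
# Constraints chain each leg through pelvis to root, with free swing and twist on the pelvis, both thighs and left calf.
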PhysicsAsset /BRCosmetics/Characters/Player/Male/Medium/Bodies/M_MED_ElbowChat/Meshes/M_MED_ElbowChat_BackClothColliders
bodies {
 calf_l []
 calf_r []
 foot_l []
 foot_r []
 pelvis [
  {shapeType=Capsule, length=6.7167, radius=11.1416, shapeName=pelvis_capsule, center=<1.4715 -2.2482 0>}
  {shapeType=Capsule, length=6.2282, radius=11.6301, shapeName=pelvis_capsule, center=<-1.0024 -2.2482 0>}
 ]
 root [{shapeType=Box, extentX=200, extentY=200, extentZ=20, shapeName=root_box, center=<0 0 -8>}]
 thigh_l []
 thigh_r []
}
constraints [
 {bone1Name=foot_r, bone2Name=calf_r, swing1LimitDegrees=45, swing2LimitDegrees=45, twistLimitDegrees=45}
 {bone1Name=calf_r, bone2Name=thigh_r, swing1LimitDegrees=45, swing2LimitDegrees=45, twistLimitDegrees=45}
 {
  bone1Name=pelvis
  bone2Name=root
  swing1LimitDegrees=45
  swing1Motion=Free
  swing2LimitDegrees=45
  swing2Motion=Free
  twistLimitDegrees=45
  twistMotion=Free
 }
 {
  bone1Name=calf_l
  bone2Name=thigh_l
  swing1LimitDegrees=45
  swing1Motion=Free
  swing2LimitDegrees=45
  swing2Motion=Free
  twistLimitDegrees=45
  twistMotion=Free
 }
 {
  bone1Name=thigh_l
  bone2Name=pelvis
  swing1LimitDegrees=45
  swing1Motion=Free
  swing2LimitDegrees=45
  swing2Motion=Free
  twistLimitDegrees=45
  twistMotion=Free
 }
 {
  bone1Name=thigh_r
  bone2Name=pelvis
  swing1LimitDegrees=45
  swing1Motion=Free
  swing2LimitDegrees=45
  swing2Motion=Free
  twistLimitDegrees=45
  twistMotion=Free
 }
 {bone1Name=foot_l, bone2Name=calf_l, swing1LimitDegrees=45, swing2LimitDegrees=45, twistLimitDegrees=45}
]
