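# Control Rig for the ElbowChat character's face accessory part, posing the braid master bones and left collar ribbon bones from body motion.
# Head rotation, via remapped Euler X and four spherical pose readers against spine_05 or root, weights corrective braid offsets.
# An isDBNO branch then applies a fixed braid offset when true, or parent and aim constraints to the ribbon when false.
# Several constraint, neck-offset and pose-reader nodes are left unconnected.
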ControlRigBlueprint /BRCosmetics/Characters/Player/Male/Medium/Bodies/M_MED_ElbowChat/Meshes/Parts/M_MED_ElbowChat_FaceAcc_CtrlRig
graphs [
 RigVMFunctionLibrary {links [], nodes []}
 RigVMFunction_Sequence_1_ContainedGraph {nodes [
   n1 = RigVMFunctionEntryNode()
   RigVMFunction_Sequence = &RigVMFunction_Sequence(ExecuteContext=@n1.ExecuteContext)
   n2 = &RigVMFunction_Sequence(ExecuteContext=@RigVMFunction_Sequence.B)
   RigVMFunctionReturnNode(A=@RigVMFunction_Sequence.A, B=@n2.A, C=@n2.B)
  ]}
 RigVMModel {nodes [
   &RigUnit_ParentConstraint(
    AdvancedSettings=(InterpolationType=Average,RotationOrderForFilter=XZY)
    Child=(Type=Bone,Name="dyn_braid_r_02")
    ExecutePin=()
    Filter=(TranslationFilter=(bX=True,bY=True,bZ=True),RotationFilter=(bX=false,bY=false,bZ=false),ScaleFilter=(bX=True,bY=True,bZ=True))
    Parents=((Item=(Type=Bone,Name="clavicle_l"),Weight=0.2),(Item=(Type=Bone,Name="spine_05"),Weight=0.7),(Item=(Type=Bone,Name="upperarm_l"),Weight=0.1),(Item=(Type=Bone,Name="neck_01"),Weight=0))
    Weight=1
   )
   > &RigUnit_ParentConstraint(
    AdvancedSettings=(InterpolationType=Average,RotationOrderForFilter=XZY)
    Child=(Type=Bone,Name="dyn_braid_l_02")
    Filter=(TranslationFilter=(bX=True,bY=True,bZ=True),RotationFilter=(bX=false,bY=false,bZ=false),ScaleFilter=(bX=True,bY=True,bZ=True))
    Parents=((Item=(Type=Bone,Name="clavicle_l"),Weight=0.2),(Item=(Type=Bone,Name="spine_05"),Weight=0.7),(Item=(Type=Bone,Name="upperarm_l"),Weight=0.1),(Item=(Type=Bone,Name="neck_01"),Weight=0))
    Weight=1
   )
   RigUnit_BeginExecution = &RigUnit_BeginExecution()
   RigVMFunction_Sequence = RigVMAggregateNode(ExecuteContext=@RigUnit_BeginExecution.ExecutePin)
   RigVMFunction_MathQuaternionToEuler = &RigVMFunction_MathQuaternionToEuler(
    RotationOrder=ZYX
    Value=&RigUnit_GetRelativeTransformForItem(Child=(Type=Bone,Name="head"), Parent=(Type=Bone,Name="spine_05"), bChildInitial=False, bParentInitial=False).RelativeTransform.Rotation
   )
   &RigUnit_ModifyTransforms(
    ItemToModify="((Item=(Type=Bone,Name=\"dyn_braid_l_master\"),Transform=(Rotation=(X=0.0189,Y=-0.2156,Z=0.0851,W=0.9726),Translation=(X=-1.1488,Y=-2.6546,Z=-0.7425),Scale3D=(X=1,Y=1,Z=1))),(Item=(Type=Bone,Name=\"dyn_braid_r_master\"),Transform=(Rotation=(X=-0.0426,Y=0.1304, ...(truncated)"
    WeightMinimum=0
    Weight=&RigVMFunction_MathDoubleRemap(SourceMaximum=-80, SourceMinimum=0, TargetMaximum=1, TargetMinimum=0, Value=@RigVMFunction_MathQuaternionToEuler.Result.X).Result
    ExecutePin=@RigVMFunction_Sequence.A
   )
   > &RigUnit_ModifyTransforms(
    ItemToModify="((Item=(Type=Bone,Name=\"dyn_braid_l_master\"),Transform=(Rotation=(X=0,Y=0.1736,Z=0,W=0.9848),Translation=(X=0,Y=0,Z=0),Scale3D=(X=1,Y=1,Z=1))),(Item=(Type=Bone,Name=\"dyn_braid_r_master\"),Transform=(Rotation=(X=-0.0168,Y=-0.2394,Z=0.0855,W=0.967),Translatio ...(truncated)"
    WeightMinimum=0
    Weight=&RigVMFunction_MathDoubleRemap(SourceMaximum=80, SourceMinimum=0, TargetMaximum=1, TargetMinimum=0, Value=@RigVMFunction_MathQuaternionToEuler.Result.X).Result
   )
   > n1 = &RigUnit_SphericalPoseReader(
    ActiveRegionScaleFactors=(PositiveWidth=1.4,NegativeWidth=1,PositiveHeight=1.2,NegativeHeight=1.2)
    ActiveRegionSize=0.1
    Debug=(bDrawDebug=false,bDraw2D=False,bDrawLocalAxes=False,DebugScale=25,DebugSegments=20,DebugThickness=0.25)
    DriverAxis=(X=0,Y=1,Z=0)
    DriverItem=(Type=Bone,Name="head")
    FalloffRegionScaleFactors=(PositiveWidth=1.7,NegativeWidth=1,PositiveHeight=1.5,NegativeHeight=1)
    FalloffSize=0.2
    FlipHeightScaling=False
    FlipWidthScaling=False
    OptionalParentItem=(Type=Bone,Name="spine_05")
    RotationOffset=(X=135,Y=-50,Z=0)
   )
   > &RigUnit_ModifyTransforms(
    ItemToModify=((Item=(Type=Bone,Name="dyn_braid_l_master"),Transform=(Rotation=(X=0,Y=0,Z=-0.342,W=-0.9397),Translation=(X=0.8076,Y=-0.9625,Z=0.33),Scale3D=(X=1,Y=1,Z=1))))
    WeightMinimum=0
    Weight=@n1.OutputParam
   )
   > n2 = &RigUnit_SphericalPoseReader(
    ActiveRegionScaleFactors=(PositiveWidth=1,NegativeWidth=1,PositiveHeight=1,NegativeHeight=1)
    ActiveRegionSize=0.1
    Debug=(bDrawDebug=false,bDraw2D=False,bDrawLocalAxes=False,DebugScale=25,DebugSegments=20,DebugThickness=0.25)
    DriverAxis=(X=0,Y=1,Z=0)
    DriverItem=(Type=Bone,Name="head")
    FalloffRegionScaleFactors=(PositiveWidth=1,NegativeWidth=1,PositiveHeight=1,NegativeHeight=1)
    FalloffSize=0.2
    FlipHeightScaling=False
    FlipWidthScaling=False
    OptionalParentItem=(Type=Bone,Name="spine_05")
    RotationOffset=(X=35.8785,Y=-88.0455,Z=0)
   )
   > &RigUnit_ModifyTransforms(
    ItemToModify="((Item=(Type=Bone,Name=\"dyn_braid_l_master\"),Transform=(Rotation=(X=0,Y=0,Z=-0.2588,W=0.9659),Translation=(X=0.8076,Y=-0.9625,Z=0.33),Scale3D=(X=1,Y=1,Z=1))),(Item=(Type=Bone,Name=\"dyn_braid_r_master\"),Transform=(Rotation=(X=0,Y=0,Z=-0.2588,W=0.9659),Trans ...(truncated)"
    WeightMinimum=0
    Weight=@n2.OutputParam
   )
   > n3 = &RigUnit_SphericalPoseReader(
    ActiveRegionScaleFactors=(PositiveWidth=1,NegativeWidth=4,PositiveHeight=4,NegativeHeight=4)
    ActiveRegionSize=0.1
    Debug=(bDrawDebug=false,bDraw2D=False,bDrawLocalAxes=False,DebugScale=25,DebugSegments=20,DebugThickness=0.25)
    DriverAxis=(X=1,Y=0,Z=0)
    DriverItem=(Type=Bone,Name="head")
    FalloffRegionScaleFactors=(PositiveWidth=0.7,NegativeWidth=1.2,PositiveHeight=0,NegativeHeight=0)
    FalloffSize=0.5
    FlipHeightScaling=False
    FlipWidthScaling=False
    OptionalParentItem=(Type=Bone,Name="root")
    RotationOffset=(X=90,Y=0,Z=0)
   )
   > &RigUnit_ModifyTransforms(
    ItemToModify="((Item=(Type=Bone,Name=\"dyn_braid_l_master\"),Transform=(Rotation=(X=0,Y=0,Z=0,W=1),Translation=(X=1.5,Y=-5,Z=0),Scale3D=(X=1,Y=1,Z=1))),(Item=(Type=Bone,Name=\"dyn_braid_r_master\"),Transform=(Rotation=(X=0,Y=0,Z=0,W=-1),Translation=(X=-1.5,Y=5,Z=0),Scale3D= ...(truncated)"
    WeightMinimum=0
    Weight=@n3.OutputParam
   )
   > n4 = &RigUnit_SphericalPoseReader(
    ActiveRegionScaleFactors=(PositiveWidth=1,NegativeWidth=4,PositiveHeight=4,NegativeHeight=4)
    ActiveRegionSize=0.1
    Debug=(bDrawDebug=false,bDraw2D=False,bDrawLocalAxes=False,DebugScale=25,DebugSegments=20,DebugThickness=0.25)
    DriverAxis=(X=0,Y=1,Z=0)
    DriverItem=(Type=Bone,Name="head")
    FalloffRegionScaleFactors=(PositiveWidth=0.7,NegativeWidth=1.2,PositiveHeight=0,NegativeHeight=0)
    FalloffSize=0.5
    FlipHeightScaling=False
    FlipWidthScaling=False
    OptionalParentItem=(Type=Bone,Name="root")
    RotationOffset=(X=270,Y=0,Z=90)
   )
   > &RigUnit_ModifyTransforms(
    ItemToModify="((Item=(Type=Bone,Name=\"dyn_braid_l_master\"),Transform=(Rotation=(X=0,Y=0,Z=0,W=1),Translation=(X=1.5,Y=5,Z=0),Scale3D=(X=1,Y=1,Z=1))),(Item=(Type=Bone,Name=\"dyn_braid_r_master\"),Transform=(Rotation=(X=0,Y=0,Z=0,W=-1),Translation=(X=-1.5,Y=-5,Z=0),Scale3D= ...(truncated)"
    WeightMinimum=0
    Weight=@n4.OutputParam
   )
   RigVMFunction_ControlFlowBranch = &RigVMFunction_ControlFlowBranch(ExecuteContext=@RigVMFunction_Sequence.B, Condition=RigVMVariableNode() {variable=isDBNO}.Value)
   &RigUnit_ParentConstraint(
    AdvancedSettings=(InterpolationType=Average,RotationOrderForFilter=XZY)
    Child=(Type=Bone,Name="dyn_collar_l_rban_01_master")
    Parents="((Item=(Type=Bone,Name=\"spine_03\"),Weight=0.0181),(Item=(Type=Bone,Name=\"spine_05\"),Weight=0.8387),(Item=(Type=Bone,Name=\"clavicle_l\"),Weight=0.081),(Item=(Type=Bone,Name=\"clavicle_r\"),Weight=0.0108),(Item=(Type=Bone,Name=\"deform_upperarm_up_r\"),Weight=0.0 ...(truncated)"
    Weight=1
    ExecutePin=@RigVMFunction_ControlFlowBranch.False
   )
   > &RigUnit_ParentConstraint(
    AdvancedSettings=(InterpolationType=Average,RotationOrderForFilter=XZY)
    Child=(Type=Bone,Name="dyn_collar_l_rban_02_master")
    Parents="((Item=(Type=Bone,Name=\"spine_05\"),Weight=0.7604),(Item=(Type=Bone,Name=\"clavicle_l\"),Weight=0.2277),(Item=(Type=Bone,Name=\"dyn_collar_l_01\"),Weight=0.0069),(Item=(Type=Bone,Name=\"dyn_collar_l_02\"),Weight=0.0005),(Item=(Type=Bone,Name=\"dyn_collar_m_01\"),We ...(truncated)"
    Weight=1
   )
   > &RigUnit_ParentConstraint(
    AdvancedSettings=(InterpolationType=Average,RotationOrderForFilter=XZY)
    Child=(Type=Bone,Name="dyn_collar_l_rban_03")
    Parents="((Item=(Type=Bone,Name=\"spine_05\"),Weight=0.4894),(Item=(Type=Bone,Name=\"clavicle_l\"),Weight=0.3009),(Item=(Type=Bone,Name=\"dyn_collar_r_03\"),Weight=0.0001),(Item=(Type=Bone,Name=\"dyn_collar_l_01\"),Weight=0.1625),(Item=(Type=Bone,Name=\"dyn_collar_l_02\"),We ...(truncated)"
    Weight=1
   )
   > &RigUnit_AimConstraintLocalSpaceOffset(
    Child=(Type=Bone,Name="dyn_collar_l_rban_02")
    Filter=(bX=True,bY=True,bZ=True)
    Parents=((Item=(Type=Bone,Name="dyn_collar_l_rban_03_master"),Weight=1))
    Weight=1
    WorldUp=(Target=(X=0,Y=0,Z=1),Kind=Direction,Space=(Type=None,Name="None"))
   )
   > &RigUnit_AimConstraintLocalSpaceOffset(
    Child=(Type=Bone,Name="dyn_collar_l_rban_01")
    Filter=(bX=True,bY=True,bZ=True)
    Parents=((Item=(Type=Bone,Name="dyn_collar_l_rban_02_master"),Weight=1))
    Weight=1
    WorldUp=(Target=(X=0,Y=0,Z=1),Kind=Direction,Space=(Type=None,Name="None"))
   )
   &RigUnit_ModifyTransforms(
    ItemToModify="((Item=(Type=Bone,Name=\"dyn_braid_l_master\"),Transform=(Rotation=(X=0,Y=0,Z=0.3827,W=0.9239),Translation=(X=0,Y=0,Z=0),Scale3D=(X=1,Y=1,Z=1))),(Item=(Type=Bone,Name=\"dyn_braid_r_master\"),Transform=(Rotation=(X=0,Y=0,Z=0.3827,W=0.9239),Translation=(X=0,Y=0, ...(truncated)"
    Weight=1
    WeightMinimum=0
    ExecutePin=@RigVMFunction_ControlFlowBranch.True
   )
   &RigUnit_ModifyTransforms(
    ExecutePin=()
    ItemToModify="((Item=(Type=Bone,Name=\"neck_01\"),Transform=(Rotation=(X=0.2588,Y=0,Z=0,W=0.9659),Translation=(X=0,Y=0,Z=0),Scale3D=(X=1,Y=1,Z=1))),(Item=(Type=Bone,Name=\"neck_02\"),Transform=(Rotation=(X=0.2588,Y=0,Z=0,W=0.9659),Translation=(X=0,Y=0,Z=0),Scale3D=(X=1,Y=1, ...(truncated)"
    Weight=1
    WeightMinimum=0
   )
   Get Delta Ref = RigVMFunctionReferenceNode(ExecuteContext=(), Item=(Type=Bone,Name="spine_04"))
   Remap = &RigVMFunction_MathDoubleRemap(SourceMaximum=10, SourceMinimum=5, TargetMaximum=1, TargetMinimum=0, Value=@"Get Delta Ref".Result.Rotation.Yaw)
   &RigUnit_SphericalPoseReader(
    ActiveRegionScaleFactors=(PositiveWidth=1.5,NegativeWidth=1.5,PositiveHeight=3,NegativeHeight=1)
    ActiveRegionSize=0.1
    Debug=(bDrawDebug=true,bDraw2D=False,bDrawLocalAxes=False,DebugScale=25,DebugSegments=20,DebugThickness=0.25)
    DriverAxis=(X=0,Y=1,Z=0)
    DriverItem=(Type=Bone,Name="head")
    ExecutePin=()
    FalloffRegionScaleFactors=(PositiveWidth=1.4,NegativeWidth=1.4,PositiveHeight=5,NegativeHeight=0.7)
    FalloffSize=0.3
    FlipHeightScaling=False
    FlipWidthScaling=False
    OptionalParentItem=(Type=Bone,Name="spine_05")
    RotationOffset=(X=135,Y=0,Z=0)
   )
   &RigUnit_SphericalPoseReader(
    ActiveRegionScaleFactors=(PositiveWidth=1.5,NegativeWidth=1.5,PositiveHeight=1,NegativeHeight=3)
    ActiveRegionSize=0.1
    Debug=(bDrawDebug=true,bDraw2D=False,bDrawLocalAxes=False,DebugScale=25,DebugSegments=20,DebugThickness=0.25)
    DriverAxis=(X=0,Y=1,Z=0)
    DriverItem=(Type=Bone,Name="head")
    ExecutePin=()
    FalloffRegionScaleFactors=(PositiveWidth=1.4,NegativeWidth=1.4,PositiveHeight=0.7,NegativeHeight=5)
    FalloffSize=0.3
    FlipHeightScaling=False
    FlipWidthScaling=False
    OptionalParentItem=(Type=Bone,Name="spine_05")
    RotationOffset=(X=45,Y=0,Z=0)
   )
  ]}
]
hierarchy [
 root {loc=<0 0 0>, rot=<0 0 0 1>}
 pelvis {parent=root, loc=<0 -0.7071 93.1191>, rot=<0.0127 0.707 -0.0127 -0.707>}
 spine_01 {parent=pelvis, loc=<6.7494 -0.6388 0>, rot=<0 0 0.034 -0.9994>}
 spine_02 {parent=spine_01, loc=<7.2518 -0.035 0>, rot=<0 0 0.0238 -0.9997>}
 spine_03 {parent=spine_02, loc=<6.0104 0 0>, rot=<0 0 0.0149 -0.9999>}
 spine_04 {parent=spine_03, loc=<8.1024 0 0>, rot=<0 0 -0.0132 -0.9999>}
 spine_05 {parent=spine_04, loc=<10.0146 0 0>, rot=<0 0 -0.0015 -1>}
 clavicle_l {parent=spine_05, loc=<4.3575 -0.1466 -2.3223>, rot=<0.0465 -0.7289 -0.009 -0.6829>}
 upperarm_l {parent=clavicle_l, loc=<15.5111 0 0>, rot=<0.0226 -0.3234 -0.0088 -0.9459>}
 lowerarm_l {parent=upperarm_l, loc=<26.6514 0 0>, rot=<0.0391 -0.0099 -0.2453 -0.9686>}
 dyn_shoulder_l_master {parent=clavicle_l, loc=<8.7859 -1.872 6.9454>, rot=<-0.0281 -0.9965 -0.0123 -0.0775>}
 dyn_shoulder_l_01 {parent=dyn_shoulder_l_master, loc=<0 0 0>, rot=<0 0 0 1>}
 dyn_shoulder_l_02 {parent=dyn_shoulder_l_01, loc=<-5.7986 0 0.0005>, rot=<0 0 0 1>}
 neck_01 {parent=spine_05, loc=<11.3138 0.1529 0>, rot=<0 0 -0.2003 -0.9797>}
 neck_02 {parent=neck_01, loc=<4.5627 0 0>, rot=<0 0 0.0504 -0.9987>}
 head {parent=neck_02, loc=<5.3914 0 0>, rot=<0 0 0.115 -0.9934>}
 dyn_braid_l_master {parent=head, loc=<-0.5903 3.3684 -5.9025>, rot=<0.9952 -0.0944 0.0111 -0.0241>}
 dyn_braid_l_01 {parent=dyn_braid_l_master, loc=<0 0 0>, rot=<0 0 0 1>}
 dyn_braid_l_02 {parent=dyn_braid_l_01, loc=<-5.5297 -0.0001 0>, rot=<0.2104 -0.0073 -0.1477 -0.9664>}
 dyn_braid_l_03 {parent=dyn_braid_l_02, loc=<-6.3508 0.0001 0>, rot=<0.0126 -0.0184 -0.067 -0.9975>}
 dyn_braid_l_04 {parent=dyn_braid_l_03, loc=<-6.6171 -0.0001 0>, rot=<-0.0497 0.0021 0.0793 -0.9956>}
 dyn_braid_l_05 {parent=dyn_braid_l_04, loc=<-6.4553 -0.0002 0>, rot=<0.0166 -0.019 0.098 -0.9949>}
 dyn_hair_r_master {parent=head, loc=<11.1335 -3.1882 8.272>, rot=<0.1424 0.1857 0.3254 -0.9162>}
 dyn_hair_r_01 {parent=dyn_hair_r_master, loc=<0 0 0>, rot=<0 0 0 1>}
 dyn_hair_l_master {parent=head, loc=<11.1335 -3.1882 -8.272>, rot=<0.1857 -0.1424 0.9162 0.3254>}
 dyn_hair_l_01 {parent=dyn_hair_l_master, loc=<0 0 0>, rot=<0 0 0 -1>}
 dyn_braid_r_master {parent=head, loc=<-0.5901 3.3684 5.9025>, rot=<-0.0944 -0.9952 0.0241 0.0111>}
 dyn_braid_r_01 {parent=dyn_braid_r_master, loc=<0 0 0>, rot=<0 0 0 1>}
 dyn_braid_r_02 {parent=dyn_braid_r_01, loc=<5.5298 0.0001 0>, rot=<0.2104 -0.0073 -0.1477 -0.9664>}
 dyn_braid_r_03 {parent=dyn_braid_r_02, loc=<6.3511 -0.0003 0>, rot=<0.0126 -0.0184 -0.067 -0.9975>}
 dyn_braid_r_04 {parent=dyn_braid_r_03, loc=<6.6172 0 0>, rot=<-0.0497 0.0021 0.0793 -0.9956>}
 dyn_braid_r_05 {parent=dyn_braid_r_04, loc=<6.4552 0.0002 0>, rot=<0.0166 -0.019 0.098 -0.9949>}
 dyn_hair_front_master {parent=head, loc=<10.5047 8.861 0.0615>, rot=<0 0 0.1005 -0.9949>}
 dyn_hair_front_01 {parent=dyn_hair_front_master, loc=<0 0 0>, rot=<0 0 0.1251 -0.9922>}
 dyn_hair_front_02 {parent=dyn_hair_front_01, loc=<7.7608 0 0>, rot=<0 0 0.512 -0.859>}
 clavicle_r {parent=spine_05, loc=<4.3575 -0.1466 2.3223>, rot=<0.7289 0.0465 -0.6829 0.009>}
 upperarm_r {parent=clavicle_r, loc=<-15.5111 0 0>, rot=<0.0226 -0.3234 -0.0088 -0.9459>}
 lowerarm_r {parent=upperarm_r, loc=<-26.6514 0 0>, rot=<0.0391 -0.0099 -0.2453 -0.9686>}
 upperarm_twist_01_r {parent=upperarm_r, loc=<-0.0041 0 0>, rot=<0 0 0 -1>}
 deform_upperarm_up_base_r {parent=upperarm_twist_01_r, loc=<3.2214 1.1391 -3.3308>, rot=<0 0 0 1>}
 deform_upperarm_up_r {parent=deform_upperarm_up_base_r, loc=<0 0 0>, rot=<0 0 0 -1>}
 dyn_shoulder_r_master {parent=clavicle_r, loc=<-8.7858 1.872 -6.9454>, rot=<-0.0281 -0.9965 -0.0123 -0.0775>}
 dyn_shoulder_r_01 {parent=dyn_shoulder_r_master, loc=<0 0 0>, rot=<0 0 0 1>}
 dyn_shoulder_r_02 {parent=dyn_shoulder_r_01, loc=<5.7987 0 0>, rot=<0 0 0 1>}
 dyn_collar_r_master {parent=spine_05, loc=<11.4732 0.2315 8.003>, rot=<-0.1681 0.3919 0.0277 -0.9041>}
 dyn_collar_r_01 {parent=dyn_collar_r_master, loc=<0 0 0>, rot=<0 0 0 1>}
 dyn_collar_r_02 {parent=dyn_collar_r_01, loc=<5.1768 0 0>, rot=<0 -0.0104 0 -0.9999>}
 dyn_collar_r_03 {parent=dyn_collar_r_02, loc=<4.936 0 0>, rot=<0 0 0 1>}
 dyn_collar_l_master {parent=spine_05, loc=<11.4733 0.2315 -8.003>, rot=<0.3919 0.1681 0.9041 0.0277>}
 dyn_collar_l_01 {parent=dyn_collar_l_master, loc=<0 0 0>, rot=<0 0 0 1>}
 dyn_collar_l_02 {parent=dyn_collar_l_01, loc=<-5.177 0 0.0002>, rot=<0 -0.0104 0 -0.9999>}
 dyn_collar_l_03 {parent=dyn_collar_l_02, loc=<-4.9356 0 -0.0004>, rot=<0 0 0 -1>}
 dyn_chestSkull_m_master {parent=spine_05, loc=<0.9793 13.2115 0.3332>, rot=<-0.5727 -0.4147 -0.4147 -0.5727>}
 dyn_chestSkull_m_01 {parent=dyn_chestSkull_m_master, loc=<0 0 0>, rot=<0 0 0 1>}
 dyn_collar_m_master {parent=spine_05, loc=<13.2934 -6.1398 0>, rot=<0.6806 -0.1918 0.1918 -0.6806>}
 dyn_collar_m_01 {parent=dyn_collar_m_master, loc=<0 0 0>, rot=<0 0 0 1>}
 dyn_collar_m_02 {parent=dyn_collar_m_01, loc=<3.9051 0 0>, rot=<0 0 0 1>}
 cape_upper_master {parent=spine_05, loc=<8.102 -11.6678 0>, rot=<0 0 0.9838 -0.1793>}
 cape_upper_01 {parent=cape_upper_master, loc=<0 0 0>, rot=<0 0 0 1>}
 cape_upper_02 {parent=cape_upper_01, loc=<5.5552 0 0>, rot=<0 0 0.0766 -0.9971>}
 dyn_collar_l_rban_03_master {parent=spine_05, loc=<12.6658 5.4737 -9.453>, rot=<0.0666 -0.4202 0.2671 -0.8647>}
 dyn_collar_l_rban_03 {parent=dyn_collar_l_rban_03_master, loc=<0 0 0>, rot=<0 0 0 1>}
 dyn_collar_l_rban_01_master {parent=spine_05, loc=<7.4636 10.6448 -2.4469>, rot=<0.0502 -0.4225 0.2335 -0.8743>}
 dyn_collar_l_rban_01 {parent=dyn_collar_l_rban_01_master, loc=<0 0 0>, rot=<0 0 0 1>}
 dyn_collar_l_rban_02_master {parent=spine_05, loc=<10.3008 8.0442 -6.1231>, rot=<0.0666 -0.4202 0.2671 -0.8647>}
 dyn_collar_l_rban_02 {parent=dyn_collar_l_rban_02_master, loc=<0 0 0>, rot=<0 0 0 1>}
 deform_pec_aim_base_l {parent=spine_04, loc=<7.8732 12.6388 -2.4123>, rot=<0.0102 -0.4602 0.3979 -0.7936>}
 deform_pec_aim_l {parent=deform_pec_aim_base_l, loc=<0 0 0>, rot=<0 0 0 1>}
 deform_pec_base_l {parent=deform_pec_aim_l, loc=<10.766 0 0>, rot=<0 0 0 1>}
 deform_pec_l {parent=deform_pec_base_l, loc=<0.055 0 0>, rot=<0 0 0 1>}
 deform_pec_aim_base_r {parent=spine_04, loc=<7.8727 12.6388 2.4123>, rot=<-0.4602 -0.0102 0.7936 0.3979>}
 deform_pec_aim_r {parent=deform_pec_aim_base_r, loc=<0 0 0>, rot=<0 0 0 1>}
 deform_pec_base_r {parent=deform_pec_aim_r, loc=<-10.7662 -0.0006 -0.0007>, rot=<0 0 0 1>}
 deform_pec_r {parent=deform_pec_base_r, loc=<0 0 0>, rot=<0 0 0 1>}
]
variables [isDBNO {type=bool, default=False}]
