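# Physics asset for the ElbowChat medium male character's FaceAcc mesh, defining collision and constraints for its dangling braids.
# Capsules cover head, neck, upper spine, clavicles and upper arms, plus dynamic braid, collar and chest skull bones.
# Each braid chains five segments to a master bone with swing and tight twist limits, stiffest at the tip.
# Duplicated free-rotation constraints link both tips to spine_05; collar and skull bodies have no constraints.
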PhysicsAsset /BRCosmetics/Characters/Player/Male/Medium/Bodies/M_MED_ElbowChat/Meshes/Parts/M_MED_ElbowChat_FaceAcc_Physics
bodies {
 clavicle_l [{shapeType=Capsule, length=6.1561, radius=9.9501, shapeName=clavicle_l_capsule, center=<9.2041 0 -2.9536>, rotation=<rot -90 0 -0.0002>}]
 clavicle_r [{shapeType=Capsule, length=6.1561, radius=10, shapeName=clavicle_l_capsule, center=<-9.2041 0 2.9536>, rotation=<rot -90 0 -0.0002>}]
 dyn_braid_l_01 [{
   shapeType=Capsule
   length=3.4085
   radius=1.1636
   shapeName=dyn_braid_l_01_capsule
   center=<-2.7162 0.2881 -0.1395>
   rotation=<rot 90 -179.7625 -179.7628>
  }]
 dyn_braid_l_02 [{
   shapeType=Capsule
   length=4.5847
   radius=1.2096
   shapeName=dyn_braid_l_02_capsule
   center=<-3.0644 -0.0928 0>
   rotation=<rot 90 179.9998 179.9995>
  }]
 dyn_braid_l_03 [{shapeType=Capsule, length=4.5578, radius=1.4737, shapeName=dyn_braid_l_03_capsule, center=<-3.4405 0 0>, rotation=<rot 90 90 89.9998>}]
 dyn_braid_l_04 [{shapeType=Capsule, length=3.8829, radius=1.4737, shapeName=dyn_braid_l_04_capsule, center=<-3.3432 0 0>, rotation=<rot 90 -0.0002 -0.0004>}]
 dyn_braid_l_05 [{
   shapeType=Capsule
   length=1.7701
   radius=1.5851
   shapeName=dyn_braid_l_05_capsule
   center=<-2.4141 0.5161 -0.172>
   rotation=<rot 90 180 -170.0005>
  }]
 dyn_braid_l_master [{
   shapeType=Capsule
   length=0.4481
   radius=1.4737
   shapeName=dyn_braid_l_master_capsule
   center=<0.4955 0 0>
   rotation=<rot 90 -179.9899 -179.9902>
  }]
 dyn_braid_r_01 [{shapeType=Capsule, length=3.35, radius=1.4737, shapeName=dyn_braid_r_01_capsule, center=<2.4142 0 0>, rotation=<rot 90 -179.9999 179.9998>}]
 dyn_braid_r_02 [{shapeType=Capsule, length=3.154, radius=1.6698, shapeName=dyn_braid_r_02_capsule, center=<3.0297 0.2613 0>, rotation=<rot 90 0 -0.0003>}]
 dyn_braid_r_03 [{shapeType=Capsule, length=3.029, radius=1.7948, shapeName=dyn_braid_r_03_capsule, center=<2.9069 0.1979 0>, rotation=<rot 90 180 -176.5469>}]
 dyn_braid_r_04 [{
   shapeType=Capsule
   length=3.2843
   radius=2.0595
   shapeName=dyn_braid_r_04_capsule
   center=<3.0191 0 0.1586>
   rotation=<rot 90 179.9704 179.9702>
  }]
 dyn_braid_r_05 [{
   shapeType=Capsule
   length=6.0857
   radius=1.9224
   shapeName=dyn_braid_r_05_capsule
   center=<4.1288 -1.3834 0.1517>
   rotation=<rot 90 0 14.4873>
  }]
 dyn_braid_r_master [{
   shapeType=Capsule
   length=0.2801
   radius=1.4737
   shapeName=dyn_braid_r_master_capsule
   center=<-0.8565 0 0>
   rotation=<rot 90 -160.5288 -160.5291>
  }]
 dyn_chestSkull_m_01 [
  {
   shapeType=Capsule
   length=0
   radius=4.7144
   shapeName=dyn_chestSkull_m_01_capsule
   center=<-2.8621 -0.3684 -0.007>
   rotation=<rot -4.9997 0 0>
  }
  {
   shapeType=Capsule
   length=2.9484
   radius=1.3414
   shapeName=dyn_chestSkull_m_01_capsule
   center=<-0.5831 0 -4.0848>
   rotation=<rot -14.9998 0 0>
  }
 ]
 dyn_collar_l_03 [{
   shapeType=Capsule
   length=13.3993
   radius=4.7818
   shapeName=dyn_collar_l_03_capsule
   center=<2.6157 1.6095 -3.0772>
   rotation=<rot 90 0 10.0007>
  }]
 dyn_collar_l_rban_01 [{
   shapeType=Capsule
   length=4.8854
   radius=0.7747
   shapeName=dyn_collar_l_rban_01_capsule
   center=<0 -0.7498 -0.0919>
   rotation=<rot 90 -90 -90>
  }]
 dyn_collar_l_rban_02 [{shapeType=Capsule, length=4.8854, radius=0.7747, shapeName=dyn_collar_l_rban_02_capsule, center=<0 -0.7423 -0.0919>, rotation=<rot 90 0 0>}]
 dyn_collar_l_rban_03 [{
   shapeType=Capsule
   length=8.7383
   radius=0.7747
   shapeName=dyn_collar_l_rban_03_capsule
   center=<1.9088 -0.6488 0.0337>
   rotation=<rot 87.1875 -179.9999 -179.9999>
  }]
 dyn_collar_m_02 [{shapeType=Capsule, length=21.8011, radius=5.1411, shapeName=dyn_collar_m_02_capsule, center=<0.2447 0 0>, rotation=<rot 90 172.875 -97.125>}]
 dyn_collar_r_03 [{
   shapeType=Capsule
   length=13.6888
   radius=4.4922
   shapeName=dyn_collar_r_03_capsule
   center=<-2.5233 -1.4654 2.4041>
   rotation=<rot 88.2655 75.0792 94.9287>
  }]
 head [
  {shapeType=Capsule, length=7.6024, radius=4.233, shapeName=head_capsule, center=<1.3201 5.1651 0>, rotation=<rot 90 180 -149.9995>}
  {shapeType=Capsule, length=34.5878, radius=15, shapeName=head_capsule, center=<17.8584 9.8513 0>}
  {shapeType=Capsule, length=15, radius=15, shapeName=head_capsule, center=<4.2157 -11.5828 0>, rotation=<rot -90 0 19.9995>}
 ]
 neck_01 [{shapeType=Capsule, length=10.4184, radius=4.8688, shapeName=neck_01_capsule, center=<4.4717 0 0>, rotation=<rot 90 179.9802 179.9803>}]
 spine_04 [{shapeType=Capsule, length=19.4803, radius=15, shapeName=spine_04_capsule, center=<2.7063 -1.3925 0>}]
 spine_05 [
  {shapeType=Capsule, length=15, radius=11.0622, shapeName=spine_05_capsule, center=<3.5154 -1.5455 0>}
  {shapeType=Capsule, length=15, radius=11.0622, shapeName=spine_05_capsule1, center=<-3.7131 2.4388 0>}
 ]
 upperarm_l [{shapeType=Capsule, length=22.8937, radius=7.1063, shapeName=upperarm_l_capsule, center=<13.8884 -0.4917 -0.4669>, rotation=<rot -90 0 0>}]
 upperarm_r [{
   shapeType=Capsule
   length=22.8937
   radius=7.1063
   shapeName=upperarm_r_capsule
   center=<-12.219 0.3599 -1.027>
   rotation=<rot -81.5625 0.0009 -0.0008>
  }]
}
constraints [
 {bone1Name=dyn_braid_r_05, bone2Name=dyn_braid_r_04, swing1LimitDegrees=15, swing2LimitDegrees=15, twistLimitDegrees=4}
 {bone1Name=dyn_braid_r_04, bone2Name=dyn_braid_r_03, swing1LimitDegrees=45, swing2LimitDegrees=45, twistLimitDegrees=5}
 {bone1Name=dyn_braid_r_03, bone2Name=dyn_braid_r_02, swing1LimitDegrees=45, swing2LimitDegrees=45, twistLimitDegrees=5}
 {bone1Name=dyn_braid_r_02, bone2Name=dyn_braid_r_01, swing1LimitDegrees=45, swing2LimitDegrees=45, twistLimitDegrees=5}
 {bone1Name=dyn_braid_r_01, bone2Name=dyn_braid_r_master, swing1LimitDegrees=45, swing2LimitDegrees=45, twistLimitDegrees=5}
 {bone1Name=dyn_braid_l_05, bone2Name=dyn_braid_l_04, swing1LimitDegrees=15, swing2LimitDegrees=15, twistLimitDegrees=4}
 {bone1Name=dyn_braid_l_04, bone2Name=dyn_braid_l_03, swing1LimitDegrees=45, swing2LimitDegrees=45, twistLimitDegrees=5}
 {bone1Name=dyn_braid_l_03, bone2Name=dyn_braid_l_02, swing1LimitDegrees=45, swing2LimitDegrees=45, twistLimitDegrees=5}
 {bone1Name=dyn_braid_l_02, bone2Name=dyn_braid_l_01, swing1LimitDegrees=55, swing2LimitDegrees=55, twistLimitDegrees=5}
 {bone1Name=dyn_braid_l_01, bone2Name=dyn_braid_l_master, swing1LimitDegrees=55, swing2LimitDegrees=55, twistLimitDegrees=5}
 {
  bone1Name=dyn_braid_r_05
  bone2Name=spine_05
  swing1LimitDegrees=45
  swing1Motion=Free
  swing2LimitDegrees=45
  swing2Motion=Free
  twistLimitDegrees=45
  twistMotion=Free
 }
 {
  bone1Name=dyn_braid_l_05
  bone2Name=spine_05
  swing1LimitDegrees=45
  swing1Motion=Free
  swing2LimitDegrees=45
  swing2Motion=Free
  twistLimitDegrees=45
  twistMotion=Free
 }
 {
  bone1Name=dyn_braid_l_05
  bone2Name=spine_05
  swing1LimitDegrees=45
  swing1Motion=Free
  swing2LimitDegrees=45
  swing2Motion=Free
  twistLimitDegrees=45
  twistMotion=Free
 }
 {
  bone1Name=dyn_braid_r_05
  bone2Name=spine_05
  swing1LimitDegrees=45
  swing1Motion=Free
  swing2LimitDegrees=45
  swing2Motion=Free
  twistLimitDegrees=45
  twistMotion=Free
 }
]
