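# Physics asset for the ElbowChat male medium character's face accessory part, covering two dangling braids plus chest skull and collar pieces.
# Capsule bodies cover both five-segment braids with their master bones, the accessories, and head, neck, upper spine, clavicles and arms.
# Constraints chain each braid segment to its master with wide swing and tight twist limits.
# Each braid tip also links to spine_05 with free angular motion; skull and collar bodies have no constraints.
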PhysicsAsset /BRCosmetics/Characters/Player/Male/Medium/Bodies/M_MED_ElbowChat/Meshes/Parts/M_MED_ElbowChat_FaceAcc_Physics_2
bodies {
 clavicle_l [{shapeType=Capsule, length=20.0499, radius=9.9501, shapeName=clavicle_l_capsule, center=<2.7106 0 -1.453>, rotation=<rot -90 0 -0.0002>}]
 clavicle_r [{
   shapeType=Capsule
   length=20.0499
   radius=9.9501
   shapeName=clavicle_r_capsule
   center=<-2.2023 0 0.919>
   rotation=<rot -90 179.4936 -179.4938>
  }]
 dyn_braid_l_01 [{
   shapeType=Capsule
   length=3.4085
   radius=1.1636
   shapeName=dyn_braid_l_01_capsule
   center=<-2.7162 0.2881 -0.1395>
   rotation=<rot 90 -179.7625 -179.7628>
  }]
 dyn_braid_l_02 [{shapeType=Capsule, length=4.5847, radius=1.2096, shapeName=dyn_braid_l_02_capsule, center=<-3.0644 0 0>, rotation=<rot 90 179.9998 179.9995>}]
 dyn_braid_l_03 [{shapeType=Capsule, length=4.5578, radius=1.4737, shapeName=dyn_braid_l_03_capsule, center=<-3.4405 0 0>, rotation=<rot 90 90 89.9998>}]
 dyn_braid_l_04 [{shapeType=Capsule, length=3.8829, radius=1.4737, shapeName=dyn_braid_l_04_capsule, center=<-3.3432 0 0>, rotation=<rot 90 -0.0002 -0.0004>}]
 dyn_braid_l_05 [{shapeType=Capsule, length=5.0034, radius=1.6431, shapeName=dyn_braid_l_05_capsule, center=<-3.9489 1.0747 0>, rotation=<rot 90 0 19.9996>}]
 dyn_braid_l_master [{
   shapeType=Capsule
   length=0.4481
   radius=1.4737
   shapeName=dyn_braid_l_master_capsule
   center=<0.4955 0 0>
   rotation=<rot 90 -179.9899 -179.9902>
  }]
 dyn_braid_r_01 [{shapeType=Capsule, length=3.35, radius=1.4737, shapeName=dyn_braid_r_01_capsule, center=<2.4142 0 0>, rotation=<rot 90 -179.9999 179.9998>}]
 dyn_braid_r_02 [{shapeType=Capsule, length=3.154, radius=1.6698, shapeName=dyn_braid_r_02_capsule, center=<3.0297 0.2613 0>, rotation=<rot 90 0 -0.0003>}]
 dyn_braid_r_03 [{shapeType=Capsule, length=3.029, radius=1.7948, shapeName=dyn_braid_r_03_capsule, center=<2.9069 0.1979 0>, rotation=<rot 90 180 -176.5469>}]
 dyn_braid_r_04 [{
   shapeType=Capsule
   length=3.2843
   radius=2.0595
   shapeName=dyn_braid_r_04_capsule
   center=<3.0191 0 0.1586>
   rotation=<rot 90 179.9704 179.9702>
  }]
 dyn_braid_r_05 [{
   shapeType=Capsule
   length=6.0857
   radius=1.9224
   shapeName=dyn_braid_r_05_capsule
   center=<4.1288 -1.3834 0.1517>
   rotation=<rot 90 0 14.4873>
  }]
 dyn_braid_r_master [{
   shapeType=Capsule
   length=0.2801
   radius=1.4737
   shapeName=dyn_braid_r_master_capsule
   center=<-0.8565 0 0>
   rotation=<rot 90 -160.5288 -160.5291>
  }]
 dyn_chestSkull_m_01 [{
   shapeType=Capsule
   length=4.4281
   radius=3.2906
   shapeName=dyn_chestSkull_m_01_capsule
   center=<-0.7644 -0.3684 -1.8199>
   rotation=<rot -10 0 0>
  }]
 dyn_collar_l_01 [{
   shapeType=Capsule
   length=17.2826
   radius=1.9249
   shapeName=dyn_collar_l_01_capsule
   center=<-6.6555 -2.2152 0.9174>
   rotation=<rot -86.628 -88.6506 72.3429>
  }]
 dyn_collar_r_01 [{
   shapeType=Capsule
   length=17.2826
   radius=1.9249
   shapeName=dyn_collar_r_01_capsule
   center=<7.9209 1.4898 -1.1387>
   rotation=<rot -90 -65.0262 50.0263>
  }]
 head [
  {shapeType=Capsule, length=7.6024, radius=4.233, shapeName=head_capsule, center=<1.3201 5.1651 0>, rotation=<rot 90 180 -149.9995>}
  {shapeType=Capsule, length=10.0485, radius=5.7089, shapeName=head_capsule, center=<10.6181 2.0344 0>}
 ]
 lowerarm_l [{
   shapeType=Capsule
   length=24.4948
   radius=5.5052
   shapeName=lowerarm_l_capsule
   center=<12.9037 0.0001 0>
   rotation=<rot 90 -173.2786 -173.2789>
  }]
 lowerarm_r [{
   shapeType=Capsule
   length=24.4948
   radius=5.5052
   shapeName=lowerarm_r_capsule
   center=<-10.6444 -0.0001 0>
   rotation=<rot 90 -160.5288 -160.5291>
  }]
 neck_01 [{shapeType=Capsule, length=10.4184, radius=4.8688, shapeName=neck_01_capsule, center=<4.4717 0 0>, rotation=<rot 90 179.9802 179.9803>}]
 spine_05 [
  {shapeType=Capsule, length=15, radius=11.0622, shapeName=spine_05_capsule, center=<3.5154 -1.5455 0>}
  {shapeType=Capsule, length=15, radius=11.0622, shapeName=spine_05_capsule1, center=<-3.7131 2.4388 0>}
  {shapeType=Capsule, length=15, radius=15, shapeName=spine_05_capsule, center=<13.6753 -9.8655 0>}
 ]
 upperarm_l [{shapeType=Capsule, length=22.8937, radius=7.1063, shapeName=upperarm_l_capsule, center=<13.8884 -0.4917 -0.4669>, rotation=<rot -90 0 0>}]
 upperarm_r [{
   shapeType=Capsule
   length=22.8937
   radius=7.1063
   shapeName=upperarm_r_capsule
   center=<-12.219 0.3599 0.1511>
   rotation=<rot -90 179.9975 -179.9975>
  }]
}
constraints [
 {bone1Name=dyn_braid_r_05, bone2Name=dyn_braid_r_04, swing1LimitDegrees=45, swing2LimitDegrees=45, twistLimitDegrees=5}
 {bone1Name=dyn_braid_r_04, bone2Name=dyn_braid_r_03, swing1LimitDegrees=45, swing2LimitDegrees=45, twistLimitDegrees=5}
 {bone1Name=dyn_braid_r_03, bone2Name=dyn_braid_r_02, swing1LimitDegrees=45, swing2LimitDegrees=45, twistLimitDegrees=5}
 {bone1Name=dyn_braid_r_02, bone2Name=dyn_braid_r_01, swing1LimitDegrees=45, swing2LimitDegrees=45, twistLimitDegrees=5}
 {bone1Name=dyn_braid_r_01, bone2Name=dyn_braid_r_master, swing1LimitDegrees=45, swing2LimitDegrees=45, twistLimitDegrees=5}
 {bone1Name=dyn_braid_l_05, bone2Name=dyn_braid_l_04, swing1LimitDegrees=45, swing2LimitDegrees=45, twistLimitDegrees=5}
 {bone1Name=dyn_braid_l_04, bone2Name=dyn_braid_l_03, swing1LimitDegrees=45, swing2LimitDegrees=45, twistLimitDegrees=5}
 {bone1Name=dyn_braid_l_03, bone2Name=dyn_braid_l_02, swing1LimitDegrees=45, swing2LimitDegrees=45, twistLimitDegrees=5}
 {bone1Name=dyn_braid_l_02, bone2Name=dyn_braid_l_01, swing1LimitDegrees=45, swing2LimitDegrees=45, twistLimitDegrees=5}
 {bone1Name=dyn_braid_l_01, bone2Name=dyn_braid_l_master, swing1LimitDegrees=45, swing2LimitDegrees=45, twistLimitDegrees=5}
 {
  bone1Name=dyn_braid_l_05
  bone2Name=spine_05
  swing1LimitDegrees=45
  swing1Motion=Free
  swing2LimitDegrees=45
  swing2Motion=Free
  twistLimitDegrees=45
  twistMotion=Free
 }
 {
  bone1Name=dyn_braid_r_05
  bone2Name=spine_05
  swing1LimitDegrees=45
  swing1Motion=Free
  swing2LimitDegrees=45
  swing2Motion=Free
  twistLimitDegrees=45
  twistMotion=Free
 }
]
