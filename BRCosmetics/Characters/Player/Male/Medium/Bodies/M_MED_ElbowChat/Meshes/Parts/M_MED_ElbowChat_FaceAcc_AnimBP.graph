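# Post-process style AnimBP for the ElbowChat face accessory part, adding secondary motion to hair and other dynamic bones.
# Copies the parent pose, layers Body mesh clavicles over it, runs a Control Rig, then blends two RigidBody simulations by DBNO state.
# Chained AnimDynamics springs then drive four hair bones.
# Updates derive RigidBody Z force from fall velocity, zeroed when skydiving; several constraint and physics nodes are unconnected.
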AnimBlueprint /BRCosmetics/Characters/Player/Male/Medium/Bodies/M_MED_ElbowChat/Meshes/Parts/M_MED_ElbowChat_FaceAcc_AnimBP
target_skeleton=&/M_MED_ElbowChat_FaceAcc_Skeleton
graphs [
 AnimGraph {nodes [
   n1 = Anim Dynamics - Bone: dyn_hair_front_01!(ComponentPose, Alpha=1, GravityOverride:Vector="0,0,0", ExternalForce="0,0,0") -> (Pose) {
    angularDampingOverride=0.999
    angularSpringConstant=500
    bAngularSpring=true
    bGravityOverrideInSimSpace=true
    bOverrideAngularDamping=true
    bUseGravityOverride=true
    bUsePlanarLimit=false
    simulationSpace=BoneRelative
    alphaBoolBlend {blendOption=Linear}
    boundBone {boneName=dyn_hair_front_01}
    componentLinearAccScale=<3 3 3>
    componentLinearVelScale=<3 3 3>
    gravityOverride {z=-100}
    physicsBodyDefinitions [{
      boundBone {boneName=dyn_hair_front_01}
      boxExtents=<15 10 10>
      constraintSetup {angularLimitsMax=<1 5 10>, angularLimitsMin=<-1 -5 -2>, angularTarget=<1 0 0>}
      localJointOffset=<0.05 0 0>
     }]
    relativeSpaceBone {boneName=head}
   }
   Anim Dynamics - Bone: dyn_hair_front_02!(ComponentPose=@n1, Alpha=1, GravityOverride:Vector="0,0,0", ExternalForce="0,0,0") -> (Pose) {
    angularDampingOverride=0.999
    angularSpringConstant=500
    bAngularSpring=true
    bGravityOverrideInSimSpace=true
    bOverrideAngularDamping=true
    bUseGravityOverride=true
    bUsePlanarLimit=false
    simulationSpace=BoneRelative
    alphaBoolBlend {blendOption=Linear}
    boundBone {boneName=dyn_hair_front_02}
    componentLinearAccScale=<3 3 3>
    componentLinearVelScale=<3 3 3>
    gravityOverride {z=-100}
    physicsBodyDefinitions [{
      boundBone {boneName=dyn_hair_front_02}
      boxExtents=<15 10 10>
      constraintSetup {angularLimitsMax=<1 5 10>, angularLimitsMin=<-1 -5 -2>, angularTarget=<1 0 0>}
      localJointOffset=<0.05 0 0>
     }]
    relativeSpaceBone {boneName=head}
   }
   n2 = Anim Dynamics - Bone: dyn_hair_l_01!(ComponentPose, Alpha=1, GravityOverride:Vector="0,0,0", ExternalForce="0,0,0") -> (Pose) {
    angularDampingOverride=0.9
    angularSpringConstant=100
    bAngularSpring=true
    bOverrideAngularDamping=true
    bUseGravityOverride=true
    bUsePlanarLimit=false
    simulationSpace=BoneRelative
    alphaBoolBlend {blendOption=Linear}
    boundBone {boneName=dyn_hair_l_01}
    componentAppliedLinearAccClamp=<10000 10000 10000>
    componentLinearAccScale=<5 5 5>
    componentLinearVelScale=<5 5 5>
    gravityOverride {z=-100}
    physicsBodyDefinitions [{
      boundBone {boneName=dyn_hair_l_01}
      boxExtents=<15 7 7>
      constraintSetup {angularLimitsMax=<15 45 45>, angularLimitsMin=<-15 -45 -45>, angularTarget=<1 0 0>}
      localJointOffset=<0.1 0 0>
     }]
    relativeSpaceBone {boneName=head}
   }
   Anim Dynamics - Bone: dyn_hair_r_01!(ComponentPose=@n2, Alpha=1, GravityOverride:Vector="0,0,0", ExternalForce="0,0,0") -> (Pose) {
    angularDampingOverride=0.999
    angularSpringConstant=100
    bAngularSpring=true
    bOverrideAngularDamping=true
    bUseGravityOverride=true
    bUsePlanarLimit=false
    simulationSpace=BoneRelative
    alphaBoolBlend {blendOption=Linear}
    boundBone {boneName=dyn_hair_r_01}
    componentAppliedLinearAccClamp=<10000 10000 10000>
    componentLinearAccScale=<5 5 5>
    componentLinearVelScale=<5 5 5>
    gravityOverride {z=-100}
    physicsBodyDefinitions [{
      boundBone {boneName=dyn_hair_r_01}
      boxExtents=<15 7 7>
      constraintSetup {angularLimitsMax=<15 45 45>, angularLimitsMin=<-15 -45 -45>, angularTarget=<1 0 0>}
      localJointOffset=<0.1 0 0>
     }]
    relativeSpaceBone {boneName=head}
   }
   n3 = "Constraint - dyn_collar_l_rban_01_master (PPPPPPP)"!(
    ComponentPose
    Alpha=1
    ConstraintWeights_0=0.0181
    ConstraintWeights_1:float=0.8387
    ConstraintWeights_2:float=0.081
    ConstraintWeights_3:float=0.0108
    ConstraintWeights_4:float=0.0043
    ConstraintWeights_5:float=0.0061
    ConstraintWeights_6:float=0.041
   ) -> (Pose) {
    alphaBoolBlend {blendOption=Linear}
    boneToModify {boneName=dyn_collar_l_rban_01_master}
    constraintSetup [
     {targetBone {boneName=spine_03}}
     {targetBone {boneName=spine_05}}
     {targetBone {boneName=clavicle_l}}
     {targetBone {boneName=clavicle_r}}
     {targetBone {boneName=deform_upperarm_up_r}}
     {targetBone {boneName=dyn_chestSkull_m_01}}
     {targetBone {boneName=spine_04}}
    ]
    constraintWeights [1,1,1,1,1,1,1]
   }
   n4 = "Constraint - dyn_collar_l_rban_02_master (PPPPPPP)"!(
    ComponentPose=@n3
    Alpha=1
    ConstraintWeights_0=0.7604
    ConstraintWeights_1:float=0.2277
    ConstraintWeights_2:float=0.0069
    ConstraintWeights_3:float=0.0005
    ConstraintWeights_4:float=0.0029
    ConstraintWeights_5:float=0.0008
    ConstraintWeights_6:float=0.0008
   ) -> (Pose) {
    alphaBoolBlend {blendOption=Linear}
    boneToModify {boneName=dyn_collar_l_rban_02_master}
    constraintSetup [
     {targetBone {boneName=spine_05}}
     {targetBone {boneName=clavicle_l}}
     {targetBone {boneName=dyn_collar_l_01}}
     {targetBone {boneName=dyn_collar_l_02}}
     {targetBone {boneName=dyn_collar_m_01}}
     {targetBone {boneName=dyn_shoulder_l_01}}
     {targetBone {boneName=spine_04}}
    ]
    constraintWeights [1,1,1,1,1,1,1]
   }
   "Constraint - dyn_collar_l_rban_03_master (PPPPPPPP)"!(
    ComponentPose=@n4
    Alpha=1
    ConstraintWeights_0=0.4894
    ConstraintWeights_1:float=0.3009
    ConstraintWeights_2:float=0.0001
    ConstraintWeights_3:float=0.1625
    ConstraintWeights_4:float=0.0359
    ConstraintWeights_5:float=0.0005
    ConstraintWeights_6:float=0.0001
    ConstraintWeights_7:float=0.0106
   ) -> (Pose) {
    alphaBoolBlend {blendOption=Linear}
    boneToModify {boneName=dyn_collar_l_rban_03_master}
    constraintSetup [
     {targetBone {boneName=spine_05}}
     {targetBone {boneName=clavicle_l}}
     {targetBone {boneName=dyn_collar_r_03}}
     {targetBone {boneName=dyn_collar_l_01}}
     {targetBone {boneName=dyn_collar_l_02}}
     {targetBone {boneName=dyn_collar_l_03}}
     {targetBone {boneName=dyn_chestSkull_m_01}}
     {targetBone {boneName=dyn_shoulder_l_01}}
    ]
    constraintWeights [1,1,1,1,1,1,1,1]
   }
   RigidBody!(ComponentPose, Alpha, OverrideWorldGravity:Vector="0,0,0", ExternalForce_X:double=0, ExternalForce_Y:double=0, ExternalForce_Z:double=0) -> (Pose) {
    bClampLinearTranslationLimitToRefPose=true
    bOverrideWorldGravity=true
    overridePhysicsAsset=&/M_MED_ElbowChat_FaceAcc_Physics
    simulationSpace=BaseBoneSpace
    alphaBoolBlend {blendOption=Linear}
    baseBoneRef {boneName=spine_05}
    componentLinearAccScale=<0.1 0.1>
    componentLinearVelScale=<0.1 0.1>
    simSpaceSettings {
     maxAngularAcceleration=2500
     maxAngularVelocity=15
     maxLinearAcceleration=3000
     maxLinearVelocity=1500
     worldAlpha=1
     externalLinearDragV=<0.05 0.05>
    }
   }
   n5 = RigidBody!(ComponentPose=Local To Component(LocalPose=Use cached pose 'BasePose'()), Alpha, OverrideWorldGravity="0,0,0", ExternalForce:Vector) -> (Pose) {
    bOverrideWorldGravity=true
    overridePhysicsAsset=&/M_MED_ElbowChat_FaceAcc_Physics_2
    alphaBoolBlend {blendOption=Linear}
    componentLinearAccScale=<0.4 0.4 0.4>
    componentLinearVelScale=<0.4 0.4 0.4>
    simSpaceSettings {
     maxAngularAcceleration=500
     maxAngularVelocity=3
     maxLinearAcceleration=400
     maxLinearVelocity=250
     velocityScaleZ=0
     worldAlpha=1
     externalLinearDragV=<0.1 0.1>
    }
   }
   n6 = Component To Local(ComponentPose=@n5)
   n7 = RigidBody!(
    ComponentPose=Local To Component(LocalPose=Use cached pose 'BasePose'())
    Alpha
    OverrideWorldGravity="0,0,0"
    ExternalForce_X:double=0
    ExternalForce_Y:double=0
    ExternalForce_Z:double=Get External Force Z() -> (External Force Z:double)
   ) -> (Pose) {
    bClampLinearTranslationLimitToRefPose=true
    bOverrideWorldGravity=true
    overridePhysicsAsset=&M_MED_ElbowChat_FaceAcc_Physics
    simulationSpace=BaseBoneSpace
    alphaBoolBlend {blendOption=Linear}
    baseBoneRef {boneName=spine_05}
    componentAppliedLinearAccClamp=<5000 5000 5000>
    componentLinearAccScale=<0.4 0.4 0.4>
    componentLinearVelScale=<0.4 0.4 0.4>
    simSpaceSettings {
     maxAngularAcceleration=250
     maxAngularVelocity=3
     maxLinearAcceleration=250
     maxLinearVelocity=250
     velocityScaleZ=0
     worldAlpha=1
     externalLinearDragV=<0.05 0.05>
    }
   }
   n8 = Component To Local(ComponentPose=@n7)
   n9 = Get bIsDBNO()
   n10 = Anim Dynamics - Bone: dyn_hair_front_01(
    ComponentPose=Local To Component(LocalPose=Blend Poses by bool(BlendPose_0=@n6, BlendPose_1=@n8, bActiveValue=@n9) {blendType=Linear})
    Alpha=1
    GravityOverride="0,0,0"
    ExternalForce="0,0,0"
   ) {
    angularDampingOverride=1
    angularSpringConstant=2000
    bAngularSpring=true
    bGravityOverrideInSimSpace=true
    bOverrideAngularDamping=true
    bUseGravityOverride=true
    bUsePlanarLimit=false
    simulationSpace=BoneRelative
    alphaBoolBlend {blendOption=Linear}
    boundBone {boneName=dyn_hair_front_01}
    componentLinearAccScale=<3 3 3>
    componentLinearVelScale=<3 3 3>
    gravityOverride {z=-100}
    physicsBodyDefinitions [{
      boundBone {boneName=dyn_hair_front_01}
      boxExtents=<10 5 5>
      constraintSetup {angularLimitsMax=<1 5 5>, angularLimitsMin=<-1 -5 -2>, angularTarget=<1 0 0>}
      localJointOffset=<0.2 0 0>
     }]
    relativeSpaceBone {boneName=root}
   }
   n11 = Anim Dynamics - Bone: dyn_hair_front_02(ComponentPose=@n10, Alpha=1, GravityOverride="0,0,0", ExternalForce="0,0,0") {
    angularDampingOverride=1
    angularSpringConstant=250
    bAngularSpring=true
    bGravityOverrideInSimSpace=true
    bOverrideAngularDamping=true
    bUseGravityOverride=true
    bUsePlanarLimit=false
    simulationSpace=BoneRelative
    alphaBoolBlend {blendOption=Linear}
    boundBone {boneName=dyn_hair_front_02}
    componentLinearAccScale=<3 3 3>
    componentLinearVelScale=<3 3 3>
    gravityOverride {z=-100}
    physicsBodyDefinitions [{
      boundBone {boneName=dyn_hair_front_02}
      boxExtents=<7 5 5>
      constraintSetup {angularLimitsMax=<1 5 5>, angularLimitsMin=<-1 -5 -2>, angularTarget=<1 0 0>}
      localJointOffset=<0.2 0 0>
     }]
    relativeSpaceBone {boneName=dyn_hair_front_01}
   }
   n12 = Anim Dynamics - Bone: dyn_hair_l_01(ComponentPose=@n11, Alpha=1, GravityOverride="0,0,0", ExternalForce="0,0,0") {
    angularDampingOverride=0.9999
    angularSpringConstant=3500
    bAngularSpring=true
    bGravityOverrideInSimSpace=true
    bOverrideAngularDamping=true
    bUseGravityOverride=true
    bUsePlanarLimit=false
    simulationSpace=BoneRelative
    alphaBoolBlend {blendOption=Linear}
    boundBone {boneName=dyn_hair_l_01}
    componentLinearAccScale=<3 3 3>
    componentLinearVelScale=<3 3 3>
    gravityOverride {z=-100}
    physicsBodyDefinitions [{
      boundBone {boneName=dyn_hair_l_01}
      boxExtents=<7 5 5>
      constraintSetup {angularLimitsMax=<1 5 5>, angularLimitsMin=<-1 -5 -2>, angularTarget=<1 0 0>}
      localJointOffset=<0.2 0 0>
     }]
    relativeSpaceBone {boneName=root}
   }
   n13 = Anim Dynamics - Bone: dyn_hair_r_01(ComponentPose=@n12, Alpha=1, GravityOverride="0,0,0", ExternalForce="0,0,0") {
    angularDampingOverride=0.9999
    angularSpringConstant=2800
    bAngularSpring=true
    bGravityOverrideInSimSpace=true
    bOverrideAngularDamping=true
    bUseGravityOverride=true
    bUsePlanarLimit=false
    simulationSpace=BoneRelative
    alphaBoolBlend {blendOption=Linear}
    boundBone {boneName=dyn_hair_r_01}
    componentLinearAccScale=<3 3 3>
    componentLinearVelScale=<3 3 3>
    gravityOverride {z=-100}
    physicsBodyDefinitions [{
      boundBone {boneName=dyn_hair_r_01}
      boxExtents=<7 5 5>
      constraintSetup {angularLimitsMax=<1 5 5>, angularLimitsMin=<-1 -5 -2>, angularTarget=<1 0 0>}
      localJointOffset=<0.2 0 0>
     }]
    relativeSpaceBone {boneName=root}
   }
   Output Pose(Result=Component To Local(ComponentPose=@n13)) {layerGroup=None}
   n14 = Layered blend per bone(
    BasePose=Copy Pose From Mesh() {bUseAttachedParent=true}
    BlendPoses_0=Copy Pose From Mesh(SourceMeshComponent=Get Body() -> (Body:Skeletal Mesh Component Budgeted Object Reference)) {bUseAttachedParent=true}
   ) {layerSetup [{branchFilters [{boneName=clavicle_r}, {boneName=clavicle_l}, {boneName=None}, {boneName=None}]}]}
   n15 = Control Rig!(Source=@n14, Alpha, isDBNO:bool=@n9) -> (Pose) {
    alphaBoolBlend {blendOption=Linear}
    controlRigAssetReference {blueprintRigClass=&/M_MED_ElbowChat_FaceAcc_CtrlRig.M_MED_ElbowChat_FaceAcc_CtrlRig_C}
   }
   Save cached pose 'BasePose'(Pose=@n15) {cacheName=BasePose}
  ]}
 EventGraph {nodes [
   n1 = Event BlueprintUpdateAnimation()
   > n2 = Sequence!(execute) -> (then_0, then_1, then_2:Exec, then_3:Exec)
   > n3 = Gravity Override State Function DEPRECATED(
    CharacterPartAnimInstance=Self-Reference()
    GameTimeInSeconds=GetGameTimeInSeconds()
    GravityOverrideParams=Get Gravity Override Params()
    RBAN Params=Get RBAN Params()
   )
   > Set Gravity Override(Gravity Override=@n3)
   n4 = Get PawnVelocityZ()
   Set Pawn Velocity Z interp(
    execute=@n2.then_1
    Pawn Velocity Z interp:float=FInterpTo(Current=Get Pawn Velocity Z interp() -> (Pawn Velocity Z interp:float), Target=@n4, DeltaTime=@n1.DeltaTimeX, InterpSpeed=15)
   ) -> (Output_Get:float)
   n5 = Branch(execute=@n2.then_2, Condition=Get bIsSkyDiving())
   > Set External Force Z(External Force Z:double=0) -> (Output_Get:double)
   Set External Force Z(execute=@n5.else, External Force Z=Remap Float(Input=@n4, OldRange=(X=0,Y=-300), NewRange=(X=0,Y=-5000), Clamp=true) -> (Output=~))
   n6 = Event OnCharacterCustomizationLoaded()
   > Set Body(Body:Skeletal Mesh Component Budgeted Object Reference=Get Array Item with EFort Custom Part Type(Array=@n6.CustomMeshParts)) -> (Output_Get:Skeletal Mesh Component Budgeted Object Reference)
  ]}
]
variables [
 Body {type=SkeletalMeshComponentBudgeted}
 Event {type=MidiEventInfo, default {trackIndex=0, timestamp {bar=0, beat=0}}}
 External Force Z {type=number}
 Gravity Override {type=Vector}
 Gravity Override Params {
  type=GravityOverrideParamsStruct
  default {
   dBNO {gravityOverride=<-600 980 0>}
   skydive_FloatingDown {gravityOverride=<0 -300 100>, windAmplitude=<50 -400 300>, windOffset=<0 -300 200>}
   skydive_GoingUp {gravityOverride=<0 0 -980>, windAmplitude=<0 0 0>}
   skydive_Parachuting {gravityOverride=<0 400 -980>, windAmplitude=<100 -200 200>, windFrequency=<1.5 3 3>}
  }
 }
 Pawn Velocity Z interp {type=number}
 RBAN Params {
  type=RBANParamsStruct
  default {
   dBNO {simSpaceSettings {
     maxAngularAcceleration=1000
     maxAngularVelocity=3
     maxLinearAcceleration=300
     maxLinearVelocity=700
     velocityScaleZ=0
     externalLinearDragV=<0 0.5 0>
    }}
   dBNO_Carried {simSpaceSettings {
     maxAngularAcceleration=1000
     maxAngularVelocity=3
     maxLinearAcceleration=300
     maxLinearVelocity=700
     velocityScaleZ=0
     externalLinearDragV=<0 0.5 0>
    }}
   emoteOrMelee {simSpaceSettings {
     maxAngularAcceleration=1000
     maxAngularVelocity=3
     maxLinearAcceleration=300
     maxLinearVelocity=700
     velocityScaleZ=0
     externalLinearDragV=<0 0.5 0>
    }}
   front_End {simSpaceSettings {
     maxAngularAcceleration=1000
     maxAngularVelocity=3
     maxLinearAcceleration=300
     maxLinearVelocity=700
     velocityScaleZ=0
     externalLinearDragV=<0 0.5 0>
    }}
   onGround_Crouch_Moving {simSpaceSettings {
     maxAngularAcceleration=1000
     maxAngularVelocity=3
     maxLinearAcceleration=300
     maxLinearVelocity=700
     velocityScaleZ=0
     externalLinearDragV=<0 0.5 0>
    }}
   onGround_Falling {simSpaceSettings {
     maxAngularAcceleration=1000
     maxAngularVelocity=3
     maxLinearAcceleration=300
     maxLinearVelocity=700
     velocityScaleZ=0
     externalLinearDragV=<0 0.5 0>
    }}
   onGround_Moving {simSpaceSettings {
     maxAngularAcceleration=1000
     maxAngularVelocity=3
     maxLinearAcceleration=300
     maxLinearVelocity=700
     velocityScaleZ=0
     externalLinearDragV=<0 0.5 0>
    }}
   onGround_Standing {simSpaceSettings {
     maxAngularAcceleration=1000
     maxAngularVelocity=3
     maxLinearAcceleration=300
     maxLinearVelocity=700
     velocityScaleZ=0
     externalLinearDragV=<0 0.5 0>
    }}
   rail_Grinding {simSpaceSettings {
     maxAngularAcceleration=1000
     maxAngularVelocity=3
     maxLinearAcceleration=300
     maxLinearVelocity=700
     velocityScaleZ=0
     worldAlpha=0.1
     externalLinearDragV=<0 0.5 0>
    }}
   riding_Moving {simSpaceSettings {
     maxAngularAcceleration=1000
     maxAngularVelocity=3
     maxLinearAcceleration=300
     maxLinearVelocity=700
     velocityScaleZ=0
     externalLinearDragV=<0 0.5 0>
    }}
   riding_Moving_Targeting {simSpaceSettings {
     maxAngularAcceleration=1000
     maxAngularVelocity=3
     maxLinearAcceleration=300
     maxLinearVelocity=700
     velocityScaleZ=0
     externalLinearDragV=<0 0.5 0>
    }}
   riding_Standing {simSpaceSettings {
     maxAngularAcceleration=1000
     maxAngularVelocity=3
     maxLinearAcceleration=300
     maxLinearVelocity=700
     velocityScaleZ=0
     externalLinearDragV=<0 0.5 0>
    }}
   skydive_DivingDown {simSpaceSettings {
     maxAngularAcceleration=1000
     maxAngularVelocity=3
     maxLinearAcceleration=300
     maxLinearVelocity=700
     velocityScaleZ=0
     externalLinearDragV=<0 0.5 0>
    }}
   skydive_FloatingDown {
    componentLinearAccScale=<0 0.5 0>
    componentLinearVelScale=<0 0.5 0>
    simSpaceSettings {
     maxAngularAcceleration=1000
     maxAngularVelocity=3
     maxLinearAcceleration=300
     maxLinearVelocity=700
     velocityScaleZ=0
     externalLinearDragV=<0 0.5 0>
    }
   }
   skydive_GoingUp {simSpaceSettings {
     maxAngularAcceleration=1000
     maxAngularVelocity=3
     maxLinearAcceleration=300
     maxLinearVelocity=700
     velocityScaleZ=0
     externalLinearDragV=<0 0.5 0>
    }}
   skydive_Parachuting {
    componentLinearAccScale=<0 0.1 0.3>
    componentLinearVelScale=<0 0.1 0.3>
    simSpaceSettings {
     maxAngularAcceleration=1000
     maxAngularVelocity=3
     maxLinearAcceleration=300
     maxLinearVelocity=700
     velocityScaleZ=0
     externalLinearDragV=<0 0.5 0>
    }
   }
   skydive_Surfing {simSpaceSettings {
     maxAngularAcceleration=1000
     maxAngularVelocity=3
     maxLinearAcceleration=300
     maxLinearVelocity=700
     velocityScaleZ=0
     externalLinearDragV=<0 0.5 0>
    }}
   swimming {simSpaceSettings {
     maxAngularAcceleration=1000
     maxAngularVelocity=3
     maxLinearAcceleration=300
     maxLinearVelocity=700
     velocityScaleZ=0
     externalLinearDragV=<0 0.5 0>
    }}
   tactical_Sprint {simSpaceSettings {
     maxAngularAcceleration=1000
     maxAngularVelocity=3
     maxLinearAcceleration=300
     maxLinearVelocity=700
     velocityScaleZ=0
     externalLinearDragV=<0 0.5 0>
    }}
   ziplining {simSpaceSettings {
     maxAngularAcceleration=1000
     maxAngularVelocity=3
     maxLinearAcceleration=300
     maxLinearVelocity=700
     velocityScaleZ=0
     externalLinearDragV=<0 0.5 0>
    }}
  }
 }
]
